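# Translucent UI material for the Island Details image, built around a Thumbnail texture parameter.
# MF_UI_UVScaling scales the UI UVs to fill using the texture size, and the thumbnail is sampled with clamping.
# A static switch, on by default, replaces the texture's real size with manual Media Width and Height, a hack to keep image and video the same size.
# An MF_UI_CoreUI_RoundedMask call provides a rounded-corner mask.
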Material /IslandDetails/Materials/M_UI_IslandDetails_Image
graph {
 properties {bAutomaticallySetUsageInEditor=false, blendMode=BLEND_Translucent, materialDomain=MD_UI, usedShadingModels=Default Lit}
 nodes [
  MaterialFunctionCall(Corner Radius=Constant() {r=10}, Radius in Pixels=Constant() {r=1}) {materialFunction=&MF_UI_CoreUI_RoundedMask}
  n1 = ScalarParameter() {defaultValue=2048, group=Thumbnail, parameterName=Media Width}
  n2 = ScalarParameter() {defaultValue=1024, group=Thumbnail, parameterName=Media Height, sortPriority=33}
  n3 = NamedRerouteUsage()
  NamedRerouteDeclaration(None=StaticSwitchParameter(
    True=AppendVector(A=@n1, B=@n2) {desc=HACK - Used to send the texture size manually to have both image and video to be the same size}
    False=TextureProperty(Tex=@n3)
   ) {defaultValue=true, group=Thumbnail, parameterName=Use Fixed Size, sortPriority=31}) {name=Texture Size, nodeColor=<rgba 0 0 0 1>}
  NamedRerouteDeclaration(None=TextureObjectParameter() {group=Thumbnail, parameterName=Thumbnail, sortPriority=30, texture=/MeshModelingToolsetExp/Textures/DefaultTexture}) {name=Texture, nodeColor=<rgba 0 0 0 1>}
  TextureSample(
   UVs=MaterialFunctionCall(
    UV=MaterialFunctionCall() {materialFunction=&GetUserInterfaceUV}.Normalized UV
    TextureSize=NamedRerouteUsage()
    ScalePivotPoint=Constant2Vector() {r=0.5}
   ) {materialFunction=&MF_UI_UVScaling}.Scaled To Fill
   Tex=NamedRerouteUsage()
  ) {samplerSource=SSM_Clamp_WorldGroupSettings}
 ]
}
