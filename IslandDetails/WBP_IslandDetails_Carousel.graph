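# Island details carousel widget: a clickable frame that hosts a media gallery inside an invisible-styled button.
# AddMediaGallery collapses the supplied gallery (IsExpanded false) and adds it to a named slot layered over a rounded UIKit outline block.
# Hover and unhover switch the outline between Focused and Unfocused states, and hovering also fires OnFocused when the local player uses a gamepad.
# Button clicks are forwarded through the OnClicked dispatcher.
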WidgetBlueprint /IslandDetails/WBP_IslandDetails_Carousel
graphs [
 AddMediaGallery {nodes [
   n1 = AddMediaGallery!() -> (then, InMediaGallery:WBP Island Details Media Gallery Object Reference) {is=title K2Node_FunctionEntry}
   > Set IsExpanded!(execute, IsExpanded:bool=false, self:WBP Island Details Media Gallery Object Reference=@n1) -> (then, Output_Get:bool)
   > AddChild(self=Get NamedSlot_MediaGallery() -> (NamedSlot_MediaGallery:Named Slot Object Reference), Content=@n1)
  ]}
 EventGraph {nodes [
   "On Hovered (Button)"() {is=title K2Node_ComponentBoundEvent}
   n1 = Get OutlineBlock() -> (OutlineBlock:WBP UIKit Block Outline Object Reference)
   > RequestTransition!(execute, self:UIKit Block Object Reference=@n1, State:EUIKitBlockVisualState Enum=Focused) -> (then)
   > Branch(Condition=IsUsingInputType!(
     InputType:ECommonInputType Enum=Gamepad
     PlayerController:Player Controller Object Reference=GetFirstLocalFortPlayerController!(bIncludeNonSpectators:bool=true, bIncludeSpectators:bool=false) -> (ReturnValue:Fort Player Controller Object Reference)
    ) -> (ReturnValue:bool))
   > Call On Focused!(execute:Exec, self:WBP Island Details Carousel Object Reference) -> (then:Exec) {is=title K2Node_CallDelegate}
   "On Unhovered (Button)"() {is=title K2Node_ComponentBoundEvent}
   > RequestTransition(self=@n1, State=Unfocused)
   "On Clicked (Button)"() {is=title K2Node_ComponentBoundEvent}
   > Call On Clicked!(execute:Exec, self:WBP Island Details Carousel Object Reference) -> (then:Exec) {is=title K2Node_CallDelegate}
  ]}
 OnClicked {nodes [OnClicked() {is=title K2Node_FunctionEntry}]}
 OnFocused {nodes [OnFocused() {is=title K2Node_FunctionEntry}]}
]
variables [
 OnClicked {}
 OnFocused {}
]
widgets {
 info {inheritedWidgetCount=0, parentClass=&UserWidget, rootWidgetClass=&Button, widgetCount=4}
 widgets [
  Button {
   widgetClassPath=&Button
   widget=Button
   properties {
    isFocusable=false
    touchMethod=PreciseTap
    visibility=SelfHitTestInvisible
    widgetStyle {
     clickedSlateSound {resourceObject=null}
     disabled {
      bIsDynamicallyLoaded=false
      drawAs=NoDrawType
      imageType=NoImage
      mirroring=NoMirror
      resourceName=None
      resourceObject=null
      tiling=NoTile
      imageSize=<0 0>
      margin=<ltrb 0 0 0 0>
      outlineSettings {
       bUseBrushTransparency=false
       roundingType=HalfHeightRadius
       width=0
       color {colorUseRule=UseColor_Specified, specifiedColor=<rgba 0 0 0 0>}
       cornerRadii=<0 0 0 0>
      }
      tintColor {colorUseRule=UseColor_Specified, specifiedColor=<rgba 1 1 1 1>}
      uVRegion {bIsValid=false, max=<0 0>, min=<0 0>}
     }
     disabledForeground {colorUseRule=UseColor_Specified, specifiedColor=<rgba 0.5271 0.5271 0.5271 1>}
     hovered {
      bIsDynamicallyLoaded=false
      drawAs=NoDrawType
      imageType=NoImage
      mirroring=NoMirror
      resourceName=None
      resourceObject=null
      tiling=NoTile
      imageSize=<32 32>
      margin=<ltrb 0 0 0 0>
      outlineSettings {
       bUseBrushTransparency=true
       roundingType=FixedRadius
       width=1
       color {colorUseRule=UseColor_Specified, specifiedColor=<rgba 0.7243 0.7243 0.7243 1>}
       cornerRadii=<4 4 4 4>
      }
      tintColor {colorUseRule=UseColor_Specified, specifiedColor=<rgba 0.7243 0.7243 0.7243 1>}
      uVRegion {bIsValid=false, max=<0 0>, min=<0 0>}
     }
     hoveredForeground {colorUseRule=UseColor_Specified, specifiedColor=<rgba 0 1 1 1>}
     hoveredSlateSound {resourceObject=/Game/Sounds/Fort_UI_Sounds/FrontEnd/DefaultSounds/PressedButtons/MSS_UI_Button_Hover_Default}
     normal {
      bIsDynamicallyLoaded=false
      drawAs=NoDrawType
      imageType=NoImage
      mirroring=NoMirror
      resourceName=None
      resourceObject=null
      tiling=NoTile
      imageSize=<32 32>
      margin=<ltrb 0 0 0 0>
      outlineSettings {
       bUseBrushTransparency=true
       roundingType=FixedRadius
       width=1
       color {colorUseRule=UseColor_Specified, specifiedColor=<rgba 0.6951 0.6951 0.6951 1>}
       cornerRadii=<4 4 4 4>
      }
      tintColor {colorUseRule=UseColor_Specified, specifiedColor=<rgba 0.4955 0.4955 0.4955 1>}
      uVRegion {bIsValid=false, max=<0 0>, min=<0 0>}
     }
     normalForeground {colorUseRule=UseColor_Specified, specifiedColor=<rgba 0 1 1 1>}
     normalPadding=<ltrb 0 0 0 0>
     pressed {
      bIsDynamicallyLoaded=false
      drawAs=NoDrawType
      imageType=NoImage
      mirroring=NoMirror
      resourceName=None
      resourceObject=null
      tiling=NoTile
      imageSize=<32 32>
      margin=<ltrb 0 0 0 0>
      outlineSettings {
       bUseBrushTransparency=true
       roundingType=FixedRadius
       width=1
       color {colorUseRule=UseColor_Specified, specifiedColor=<rgba 0.7243 0.7243 0.7243 1>}
       cornerRadii=<4 4 4 4>
      }
      tintColor {colorUseRule=UseColor_Specified, specifiedColor=<rgba 0.3843 0.3843 0.3843 1>}
      uVRegion {bIsValid=false, max=<0 0>, min=<0 0>}
     }
     pressedForeground {colorUseRule=UseColor_Specified, specifiedColor=<rgba 0 1 1 1>}
     pressedPadding=<ltrb 0 0 0 0>
     pressedSlateSound {resourceObject=/Game/Sounds/Fort_UI_Sounds/FrontEnd/DefaultSounds/PressedButtons/MSS_UI_Button_Pressed_Default}
    }
   }
  }
  Overlay_3 {
   widgetClassPath=&Overlay
   parent=Button
   widget=Overlay_3
   slot {class=ButtonSlot, properties {horizontalAlignment=HAlign_Fill, verticalAlignment=VAlign_Fill, padding=<ltrb 0 0 0 0>}}
  }
  OutlineBlock {
   widgetClassPath=/FortUILibrary/UIKit/Buttons/Blocks/WBP_UIKit_Block_Outline
   bIsVariable=true
   parent=Overlay_3
   widget=OutlineBlock
   properties {
    bIsFocusable=true
    bottomLeftCornerRadius=20
    bottomRightCornerRadius=20
    priority=100
    radiusInPixels=true
    topLeftCornerRadius=20
    topRightCornerRadius=20
    visibility=Visible
    transitionTiming {
     deselectingFocused=0
     deselectingIdle=0
     disabling=0
     enabling=0
     focusing=1
     hovering=0
     pressing=0
     releasing=0
     selecting=0
     unfocusing=0
     unhovering=0
    }
   }
   slot {class=OverlaySlot, properties {horizontalAlignment=HAlign_Fill, verticalAlignment=VAlign_Fill}}
  }
  NamedSlot_MediaGallery {
   widgetClassPath=&NamedSlot
   bIsVariable=true
   parent=Overlay_3
   widget=NamedSlot_MediaGallery
   slot {class=OverlaySlot, properties {horizontalAlignment=HAlign_Fill, verticalAlignment=VAlign_Fill}}
  }
 ]
}
