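# Island details description panel widget that shows an island's wrapping tag list above its description text on a translucent backplate.
# SetDescriptionTags rebuilds the dynamic entry box from activity browser tag data, collapsing it when empty and giving EpicOriginal tags their own tag type.
# An invisible root button's hover and unhover events switch the outline block between Focused and Unfocused states.
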WidgetBlueprint /IslandDetails/WBP_IslandDetails_Description
graphs [
 EventGraph {nodes [
   "On Hovered (Button_Details_Hover)"() {is=title K2Node_ComponentBoundEvent}
   n1 = Get WBP_Outline_Description() -> (WBP_Outline_Description:WBP UIKit Block Outline Object Reference)
   > RequestTransition!(execute, self:UIKit Block Object Reference=@n1, State:EUIKitBlockVisualState Enum=Focused) -> (then)
   "On Unhovered (Button_Details_Hover)"() {is=title K2Node_ComponentBoundEvent}
   > RequestTransition(self=@n1, State=Unfocused)
  ]}
 SetDescriptionTags {nodes [
   n1 = SetDescriptionTags!() -> (then, Tag Data:Array of Fort Activity Browser Tag Data Structures) {is=title K2Node_FunctionEntry}
   n2 = Get EntryBox_IslandTags() -> (EntryBox_IslandTags:Dynamic Entry Box Object Reference)
   > Reset!(execute, self:Dynamic Entry Box Object Reference=@n2, bDeleteWidgets:bool=true) -> (then)
   n3 = Length(TargetArray:Array of Fort Activity Browser Tag Data Structures=@n1)
   > SetVisibility(
    self=Get SizeBox_Tags() -> (SizeBox_Tags:Size Box Object Reference)
    InVisibility="To Visibility (Boolean)"!(
     bIsVisible:bool="integer > integer"(A=@n3)
     TrueVisibility:ESlateVisibility Enum=SelfHitTestInvisible
     FalseVisibility:ESlateVisibility Enum=Collapsed
    ) -> (ReturnValue:ESlateVisibility Enum=Visible)
   )
   > n4 = For Each Loop(Array:Array of Fort Activity Browser Tag Data Structures=@n1) -> (Array Element:Fort Activity Browser Tag Data Structure)
   > n5 = Create Entry of Class!(
    execute
    self:Dynamic Entry Box Object Reference=@n2
    EntryClass:User Widget Class Reference=/FortUILibrary/UIKit/Tag/WBP_UIKit_Tag.WBP_UIKit_Tag_C
   ) -> (then, ReturnValue:WBP UIKit Tag Object Reference)
   > Set Use Dark Background!(execute, Use Dark Background:bool=true, self:WBP UIKit Tag Base Object Reference=@n5) -> (then, Output_Get:bool)
   > InitializeTag(self:WBP UIKit Tag Base Object Reference=@n5)
   n6 = Break Fort Activity Browser Tag Data!(FortActivityBrowserTagData:"Fort Activity Browser Tag Data Structure (by ref)"=@n4) -> (TagType:EFortActivityBrowserTagType Enum, TagText:Text) {is=title K2Node_BreakStruct}
   > SetTagType!(
    execute
    self:WBP UIKit Tag Base Object Reference=@n5
    Type:E_UI_TagType Enum=Select(
     Option 0:E_UI_TagType Enum=NewEnumerator2
     Option 1:E_UI_TagType Enum=NewEnumerator1
     Index="Equal (Enum)"(A:EFortActivityBrowserTagType Enum=@n6, B:EFortActivityBrowserTagType Enum=EpicOriginal)
    ) -> (ReturnValue:E_UI_TagType Enum)
   ) -> (then)
   > SetTagText!(execute, self:WBP UIKit Tag Base Object Reference=@n5, Text:Text=@n6.TagText) -> (then)
   > SetLabelText!(execute, self:WBP UIKit Tag Base Object Reference=@n5, Text:Text) -> (then)
  ]}
]
variables []
widgets {
 info {inheritedWidgetCount=0, parentClass=&UserWidget, rootWidgetClass=&Button, widgetCount=11}
 widgets [
  Button_Details_Hover {
   widgetClassPath=&Button
   widget=Button_Details_Hover
   properties {
    touchMethod=PreciseTap
    widgetStyle {
     clickedSlateSound {resourceObject=null}
     disabled {
      bIsDynamicallyLoaded=false
      drawAs=NoDrawType
      imageType=NoImage
      mirroring=NoMirror
      resourceName=None
      resourceObject=null
      tiling=NoTile
      imageSize=<0 0>
      margin=<ltrb 0 0 0 0>
      outlineSettings {
       bUseBrushTransparency=false
       roundingType=HalfHeightRadius
       width=0
       color {colorUseRule=UseColor_Specified, specifiedColor=<rgba 0 0 0 0>}
       cornerRadii=<0 0 0 0>
      }
      tintColor {colorUseRule=UseColor_Specified, specifiedColor=<rgba 1 1 1 1>}
      uVRegion {bIsValid=false, max=<0 0>, min=<0 0>}
     }
     disabledForeground {colorUseRule=UseColor_Specified, specifiedColor=<rgba 0.5271 0.5271 0.5271 1>}
     hovered {
      bIsDynamicallyLoaded=false
      drawAs=NoDrawType
      imageType=NoImage
      mirroring=NoMirror
      resourceName=None
      resourceObject=null
      tiling=NoTile
      imageSize=<32 32>
      margin=<ltrb 0 0 0 0>
      outlineSettings {
       bUseBrushTransparency=true
       roundingType=FixedRadius
       width=1
       color {colorUseRule=UseColor_Specified, specifiedColor=<rgba 0.7243 0.7243 0.7243 1>}
       cornerRadii=<4 4 4 4>
      }
      tintColor {colorUseRule=UseColor_Specified, specifiedColor=<rgba 0.7243 0.7243 0.7243 1>}
      uVRegion {bIsValid=false, max=<0 0>, min=<0 0>}
     }
     hoveredForeground {colorUseRule=UseColor_Specified, specifiedColor=<rgba 0 1 1 1>}
     hoveredSlateSound {resourceObject=/Game/Sounds/Fort_UI_Sounds/FrontEnd/DefaultSounds/PressedButtons/MSS_UI_Button_Hover_Default}
     normal {
      bIsDynamicallyLoaded=false
      drawAs=NoDrawType
      imageType=NoImage
      mirroring=NoMirror
      resourceName=None
      resourceObject=null
      tiling=NoTile
      imageSize=<32 32>
      margin=<ltrb 0 0 0 0>
      outlineSettings {
       bUseBrushTransparency=true
       roundingType=FixedRadius
       width=1
       color {colorUseRule=UseColor_Specified, specifiedColor=<rgba 0.6951 0.6951 0.6951 1>}
       cornerRadii=<4 4 4 4>
      }
      tintColor {colorUseRule=UseColor_Specified, specifiedColor=<rgba 0.4955 0.4955 0.4955 1>}
      uVRegion {bIsValid=false, max=<0 0>, min=<0 0>}
     }
     normalForeground {colorUseRule=UseColor_Specified, specifiedColor=<rgba 0 1 1 1>}
     normalPadding=<ltrb 0 0 0 0>
     pressed {
      bIsDynamicallyLoaded=false
      drawAs=NoDrawType
      imageType=NoImage
      mirroring=NoMirror
      resourceName=None
      resourceObject=null
      tiling=NoTile
      imageSize=<32 32>
      margin=<ltrb 0 0 0 0>
      outlineSettings {
       bUseBrushTransparency=true
       roundingType=FixedRadius
       width=1
       color {colorUseRule=UseColor_Specified, specifiedColor=<rgba 0.7243 0.7243 0.7243 1>}
       cornerRadii=<4 4 4 4>
      }
      tintColor {colorUseRule=UseColor_Specified, specifiedColor=<rgba 0.3843 0.3843 0.3843 1>}
      uVRegion {bIsValid=false, max=<0 0>, min=<0 0>}
     }
     pressedForeground {colorUseRule=UseColor_Specified, specifiedColor=<rgba 0 1 1 1>}
     pressedPadding=<ltrb 0 0 0 0>
     pressedSlateSound {resourceObject=/Game/Sounds/Fort_UI_Sounds/FrontEnd/DefaultSounds/PressedButtons/MSS_UI_Button_Pressed_Default}
    }
   }
  }
  Overlay_2 {
   widgetClassPath=&Overlay
   parent=Button_Details_Hover
   widget=Overlay_2
   properties {visibility=HitTestInvisible}
   slot {class=ButtonSlot, properties {horizontalAlignment=HAlign_Fill, verticalAlignment=VAlign_Fill, padding=<ltrb 0 0 0 0>}}
  }
  WBP_CaptureForPostBufferUpdate {
   widgetClassPath=&WBP_CaptureForPostBufferUpdate
   bIsVariable=true
   parent=Overlay_2
   slot {class=OverlaySlot, properties {horizontalAlignment=HAlign_Fill, verticalAlignment=VAlign_Fill}}
  }
  Spacer_378 {widgetClassPath=&Spacer, parent=Overlay_2, widget=Spacer_378, properties {size=<1200 1>}, slot {class=OverlaySlot}}
  Spacer {widgetClassPath=&Spacer, parent=Overlay_2, widget=Spacer, properties {size=<1 100>}, slot {class=OverlaySlot}}
  WBP_UIKit_Backplate_Description {
   widgetClassPath=&WBP_UIKit_Backplate
   bIsVariable=true
   parent=Overlay_2
   widget=WBP_UIKit_Backplate_Description
   properties {brightness=Medium, colorAndOpacity=<rgba 1 1 1 0.5>}
   slot {class=OverlaySlot, properties {horizontalAlignment=HAlign_Fill, verticalAlignment=VAlign_Fill}}
  }
  VBox_Content {
   widgetClassPath=&VerticalBox
   parent=Overlay_2
   widget=VBox_Content
   slot {class=OverlaySlot, properties {horizontalAlignment=HAlign_Fill, verticalAlignment=VAlign_Fill, padding=<ltrb 32 24 32 24>}}
  }
  SizeBox_Tags {
   widgetClassPath=&SizeBox
   bIsVariable=true
   parent=VBox_Content
   widget=SizeBox_Tags
   properties {bOverride_WidthOverride=true, widthOverride=1300}
   slot {class=VerticalBoxSlot, properties {size {sizeRule=Fill, value=1}}}
  }
  EntryBox_IslandTags {
   widgetClassPath=&DynamicEntryBox
   bIsVariable=true
   parent=SizeBox_Tags
   widget=EntryBox_IslandTags
   properties {
    entryBoxType=Wrap
    entryHorizontalAlignment=HAlign_Right
    entryWidgetClass=&WBP_UIKit_Tag_C
    flowDirectionPreference=Culture
    numDesignerPreviewEntries=4
    entrySpacing=<8 8>
   }
   slot {class=SizeBoxSlot, properties {padding=<ltrb 0 8 0 24>}}
  }
  Text_IslandDescription {
   widgetClassPath=&CommonTextBlock
   bIsVariable=true
   parent=VBox_Content
   widget=Text_IslandDescription
   properties {
    applyLineHeightToBottomLine=false
    autoWrapText=true
    lineHeightPercentage=1.2
    mobileFontSizeMultiplier=1.2
    text="[Game Mode Description.  Wrapping is set to per character, text widget clips to bounds on itself. Lorem ipsum dolor sit amet, consectetur adipiscing elit."
    wrappingPolicy=AllowPerCharacterWrapping
    colorAndOpacity {colorUseRule=UseColor_Specified, specifiedColor=<rgba 0.5 0.5 0.5 1>}
    font {
     bForceMonospaced=false
     bMaterialIsStencil=false
     fontMaterial=null
     fontObject=&HeadingNow
     letterSpacing=0
     monospacedWidth=1
     size=15.75
     skewAmount=0
     typefaceFontName=74 Regular
     outlineSettings {
      bApplyOutlineToDropShadows=false
      bMiteredCorners=false
      bSeparateFillAlpha=false
      outlineMaterial=null
      outlineSize=0
      outlineColor=<rgba 0 0 0 1>
     }
    }
    shadowOffset=<0 0>
    shapedTextOptions {bOverride_TextFlowDirection=true, bOverride_TextShapingMethod=false, textFlowDirection=Culture, textShapingMethod=Auto}
    strikeBrush {
     bIsDynamicallyLoaded=false
     drawAs=NoDrawType
     imageType=NoImage
     mirroring=NoMirror
     resourceName=None
     resourceObject=null
     tiling=NoTile
     imageSize=<32 32>
     margin=<ltrb 0 0 0 0>
     outlineSettings {
      bUseBrushTransparency=false
      roundingType=HalfHeightRadius
      width=0
      color {colorUseRule=UseColor_Specified, specifiedColor=<rgba 0 0 0 0>}
      cornerRadii=<0 0 0 0>
     }
     tintColor {colorUseRule=UseColor_Specified, specifiedColor=<rgba 1 1 1 1>}
     uVRegion {bIsValid=false, max=<0 0>, min=<0 0>}
    }
   }
   slot {class=VerticalBoxSlot, properties {padding=<ltrb 0 0 0 4>}}
  }
  WBP_Outline_Description {
   widgetClassPath=/FortUILibrary/UIKit/Buttons/Blocks/WBP_UIKit_Block_Outline
   bIsVariable=true
   parent=Overlay_2
   widget=WBP_Outline_Description
   properties {
    bIsFocusable=true
    bottomLeftCornerRadius=20
    bottomRightCornerRadius=20
    navigation=&self:WidgetTree.WBP_Outline_Description.WidgetNavigation_0
    priority=100
    radiusInPixels=true
    topLeftCornerRadius=20
    topRightCornerRadius=20
    visibility=Visible
    transitionTiming {
     deselectingFocused=0
     deselectingIdle=0
     disabling=0
     enabling=0
     focusing=1
     hovering=0
     pressing=0
     releasing=0
     selecting=0
     unfocusing=0
     unhovering=0
    }
   }
   slot {class=OverlaySlot, properties {horizontalAlignment=HAlign_Fill, verticalAlignment=VAlign_Fill}}
  }
 ]
}
