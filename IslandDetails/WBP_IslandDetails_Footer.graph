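# Footer widget for the island details screen, showing View Developer, Attributions and Report Island buttons with icons and descriptions in a horizontal row.
# PreConstruct tightens button padding when compact UI is active, then styles each button's corner material parameters, secondary-text wrapping and tint.
# Focus on any of the buttons rebroadcasts through the OnFocused dispatcher.
# SetViewCreatorText formats a given creator name into the developer button's follow message.
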WidgetBlueprint /IslandDetails/WBP_IslandDetails_Footer
graphs [
 EventGraph {nodes [
   "On Focused (FooterButton_ViewCreator)"!() -> (OutputDelegate, then, Button:Common Button Base Object Reference) {is=title K2Node_ComponentBoundEvent}
   > Call On Focused!(execute:Exec, self:WBP Island Details Footer Object Reference) -> (then:Exec) {is=title K2Node_CallDelegate}
   "On Focused (FooterButton_ViewAttributions)"!() -> (OutputDelegate, then, Button:Common Button Base Object Reference) {is=title K2Node_ComponentBoundEvent}
   > Call On Focused()
   "On Focused (FooterButton_ReportIsland)"!() -> (OutputDelegate, then, Button:Common Button Base Object Reference) {is=title K2Node_ComponentBoundEvent}
   > Call On Focused()
   n1 = Event PreConstruct()
   > Parent: PreConstruct(IsDesignTime=@n1.IsDesignTime)
   > SetCompactUIStyle() {is=title K2Node_CallFunction}
   > SetupFooterButtonsVisuals()
  ]}
 OnFocused {nodes [OnFocused() {is=title K2Node_FunctionEntry}]}
 SetCompactUIStyle {nodes [
   SetCompactUIStyle() {is=title K2Node_FunctionEntry}
   > Branch(Condition=IsUsingCompactUI())
   > SetButtonPadding!(
    execute
    self:WBP UIKit Button Generic Base Object Reference=[
     Get FooterButton_ViewCreator() -> (FooterButton_ViewCreator:WBP UIKit Button Generic Object Reference)
     Get FooterButton_ViewAttributions() -> (FooterButton_ViewAttributions:WBP UIKit Button Generic Object Reference)
     Get FooterButton_ReportIsland() -> (FooterButton_ReportIsland:WBP UIKit Button Generic Object Reference)
    ]
    Padding_Left:float=24
    Padding_Top:float=24
    Padding_Right:float=24
    Padding_Bottom:float=24
   ) -> (then)
  ]}
 SetViewCreatorText {nodes [
   n1 = SetViewCreatorText!() -> (then, creator:Text) {is=title K2Node_FunctionEntry}
   > SetSecondText!(
    execute
    self:WBP UIKit Button Generic Base Object Reference=Get FooterButton_ViewCreator()
    SecondText:Text=Format Text!(Format="LOCTEXT(\"Follow {creator} to stay up to date with new islands and updates from their community.\")", creator:Text=@n1) -> (Result)
   ) -> (then)
  ]}
 SetupFooterButtonsVisuals {nodes [
   SetupFooterButtonsVisuals() {is=title K2Node_FunctionEntry}
   > SetupIndividualButtonVisual!(execute, self, Button:WBP UIKit Button Generic Base Object Reference=Get FooterButton_ViewCreator()) -> (then)
   > SetupIndividualButtonVisual(Button=Get FooterButton_ViewAttributions())
   > SetupIndividualButtonVisual(Button=Get FooterButton_ReportIsland())
  ]}
 SetupIndividualButtonVisual {nodes [
   n1 = SetupIndividualButtonVisual!() -> (then, Button:WBP UIKit Button Generic Base Object Reference) {is=title K2Node_FunctionEntry}
   n2 = Get Block_Background!(self:WBP UIKit Button Generic Base Object Reference=@n1) -> (Block_Background:WBP UIKit Block Background Base Object Reference)
   n3 = Get BackgroundDynamicMaterial!(self:WBP UIKit Block Background Base Object Reference=@n2) -> (BackgroundDynamicMaterial:Material Instance Dynamic Object Reference)
   > SetScalarParameterValue(self=@n3, ParameterName=Radius in Pixels, Value=1)
   > SetScalarParameterValue(self=@n3, ParameterName=Corner Radius, Value=16)
   n4 = Get Block_SecondText!(self:WBP UIKit Button Generic Base Object Reference=@n1) -> (Block_SecondText:WBP UIKit Block Text Object Reference)
   > SetWrapping!(execute, self:WBP UIKit Block Text Object Reference=@n4, IsAutoWrap:bool=true, WrapTextAt:int=0) -> (then)
   n5 = Get OutlineDynamicMaterial!(self:WBP UIKit Block Background Base Object Reference=@n2) -> (OutlineDynamicMaterial:Material Instance Dynamic Object Reference)
   > SetScalarParameterValue(self=@n5, ParameterName=Radius in Pixels, Value=1)
   > SetVectorParameterValue!(execute, self=@n5, ParameterName=Corner Radius, Value_R:float=24, Value_G:float=24, Value_B:float=24, Value_A:float=24) -> (then)
   > SetColorAndOpacity(self:User Widget Object Reference=@n4, InColorAndOpacity:Linear Color Structure=(R=0.5,G=0.5,B=0.5,A=1))
   > SetColorAndOpacity(
    self=Get Block_Background(self=Get Button() -> (Button:WBP UIKit Button Generic Base Object Reference))
    InColorAndOpacity=(R=1,G=1,B=1,A=0.5)
   )
  ]}
]
variables [OnFocused {}]
widgets {
 info {inheritedWidgetCount=0, parentClass=&UserWidget, rootWidgetClass=&HorizontalBox, widgetCount=4}
 widgets [
  HBox_Footer {widgetClassPath=&HorizontalBox, widget=HBox_Footer, properties {bOverride_Cursor=true}}
  FooterButton_ViewCreator {
   widgetClassPath=&WBP_UIKit_Button_Generic
   bIsVariable=true
   parent=HBox_Footer
   widget=FooterButton_ViewCreator
   properties {
    bHideInputAction=true
    bHideInputActionWithKeyboard=true
    contentHorizontalAlignment=HAlign_Fill
    contentVerticalAlignment=VAlign_Fill
    enableFocusedSelectedState=false
    iconHorizontalAlignment=HAlign_Left
    iconMaterial=/FortUILibrary/Textures/MaterialInstances/MI_UIKit_Icon_ItemShopCup
    iconVerticalAlignment=VAlign_Top
    maxWidth=570
    minHeight=0
    secondText=Follow Epic Games to stay up to date with new islands and updates from their community.
    secondTextBlockAlignment=Top
    secondTextTransformPolicy=None
    text=View Developer
    textBlockAlignment=Bottom
    textJustification=Left
    textSecondJustification=Left
    textTransformPolicy=None
    useSecondTextBlock=true
    visibility=Visible
    buttonPadding=<ltrb 32 32 32 32>
    font {
     bForceMonospaced=false
     bMaterialIsStencil=false
     fontMaterial=null
     fontObject=&HeadingNow
     letterSpacing=0
     monospacedWidth=1
     size=18
     skewAmount=0
     typefaceFontName=75 Medium
     outlineSettings {
      bApplyOutlineToDropShadows=false
      bMiteredCorners=false
      bSeparateFillAlpha=false
      outlineMaterial=null
      outlineSize=0
      outlineColor=<rgba 0 0 0 1>
     }
    }
    iconSize=<32 32>
    secondFont {
     bForceMonospaced=false
     bMaterialIsStencil=false
     fontMaterial=null
     fontObject=&HeadingNow
     letterSpacing=0
     monospacedWidth=1
     size=15.75
     skewAmount=0
     typefaceFontName=74 Regular
     outlineSettings {
      bApplyOutlineToDropShadows=false
      bMiteredCorners=false
      bSeparateFillAlpha=false
      outlineMaterial=null
      outlineSize=0
      outlineColor=<rgba 0 0 0 1>
     }
    }
    secondTextPadding=<ltrb 0 0 0 0>
    textPadding=<ltrb 0 12 0 12>
   }
   slot {class=HorizontalBoxSlot, properties {padding=<ltrb 0 0 24 0>}}
  }
  FooterButton_ViewAttributions {
   widgetClassPath=&WBP_UIKit_Button_Generic
   bIsVariable=true
   parent=HBox_Footer
   widget=FooterButton_ViewAttributions
   properties {
    bHideInputAction=true
    bHideInputActionWithKeyboard=true
    contentHorizontalAlignment=HAlign_Fill
    contentVerticalAlignment=VAlign_Fill
    enableFocusedSelectedState=false
    iconHorizontalAlignment=HAlign_Left
    iconMaterial=/FortUILibrary/Textures/MaterialInstances/MI_UIKit_Icon_Information
    iconVerticalAlignment=VAlign_Top
    maxWidth=570
    pressMethod=ButtonPress
    secondText=This island uses licensed content created by the Fortnite community..
    secondTextBlockAlignment=Top
    secondTextTransformPolicy=None
    text=Attributions
    textBlockAlignment=Bottom
    textJustification=Left
    textSecondJustification=Left
    textTransformPolicy=None
    useSecondTextBlock=true
    visibility=Visible
    buttonPadding=<ltrb 32 32 32 32>
    font {
     bForceMonospaced=false
     bMaterialIsStencil=false
     fontMaterial=null
     fontObject=&HeadingNow
     letterSpacing=0
     monospacedWidth=1
     size=18
     skewAmount=0
     typefaceFontName=75 Medium
     outlineSettings {
      bApplyOutlineToDropShadows=false
      bMiteredCorners=false
      bSeparateFillAlpha=false
      outlineMaterial=null
      outlineSize=0
      outlineColor=<rgba 0 0 0 1>
     }
    }
    iconSize=<32 32>
    secondFont {
     bForceMonospaced=false
     bMaterialIsStencil=false
     fontMaterial=null
     fontObject=&HeadingNow
     letterSpacing=0
     monospacedWidth=1
     size=15.75
     skewAmount=0
     typefaceFontName=74 Regular
     outlineSettings {
      bApplyOutlineToDropShadows=false
      bMiteredCorners=false
      bSeparateFillAlpha=false
      outlineMaterial=null
      outlineSize=0
      outlineColor=<rgba 0 0 0 1>
     }
    }
    secondTextPadding=<ltrb 0 0 0 0>
    textPadding=<ltrb 0 12 0 12>
   }
   slot {class=HorizontalBoxSlot, properties {padding=<ltrb 0 0 24 0>}}
  }
  FooterButton_ReportIsland {
   widgetClassPath=&WBP_UIKit_Button_Generic
   bIsVariable=true
   parent=HBox_Footer
   widget=FooterButton_ReportIsland
   properties {
    bHideInputAction=true
    bHideInputActionWithKeyboard=true
    contentHorizontalAlignment=HAlign_Fill
    contentVerticalAlignment=VAlign_Fill
    enableFocusedSelectedState=false
    iconHorizontalAlignment=HAlign_Left
    iconMaterial=/FortUILibrary/Textures/MaterialInstances/MI_UIKit_Icon_Alert
    iconVerticalAlignment=VAlign_Top
    maxWidth=570
    minHeight=0
    pressMethod=ButtonPress
    secondText="If you feel this island is violating community guidelines, please submit a report."
    secondTextBlockAlignment=Top
    secondTextTransformPolicy=None
    text=Report Island
    textBlockAlignment=Bottom
    textJustification=Left
    textSecondJustification=Left
    textTransformPolicy=None
    useSecondTextBlock=true
    visibility=Visible
    buttonPadding=<ltrb 32 32 32 32>
    font {
     bForceMonospaced=false
     bMaterialIsStencil=false
     fontMaterial=null
     fontObject=&HeadingNow
     letterSpacing=0
     monospacedWidth=1
     size=18
     skewAmount=0
     typefaceFontName=75 Medium
     outlineSettings {
      bApplyOutlineToDropShadows=false
      bMiteredCorners=false
      bSeparateFillAlpha=false
      outlineMaterial=null
      outlineSize=0
      outlineColor=<rgba 0 0 0 1>
     }
    }
    iconSize=<32 32>
    secondFont {
     bForceMonospaced=false
     bMaterialIsStencil=false
     fontMaterial=null
     fontObject=&HeadingNow
     letterSpacing=0
     monospacedWidth=1
     size=15.75
     skewAmount=0
     typefaceFontName=74 Regular
     outlineSettings {
      bApplyOutlineToDropShadows=false
      bMiteredCorners=false
      bSeparateFillAlpha=false
      outlineMaterial=null
      outlineSize=0
      outlineColor=<rgba 0 0 0 1>
     }
    }
    secondTextPadding=<ltrb 0 0 0 0>
    textPadding=<ltrb 0 12 0 12>
   }
   slot {class=HorizontalBoxSlot}
  }
 ]
}
