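# Header widget for an island details screen, showing a creator line above the island name.
# A vertical box stacks a "Created by · origin" row of small CommonTextBlocks over a large, bold italic, uppercase name.
# The origin and name texts are exposed as variables with placeholder content.
# The event graph holds only unconnected PreConstruct, Construct and Tick stubs, so the widget has no logic of its own.
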WidgetBlueprint /IslandDetails/WBP_IslandDetails_Header
graphs [EventGraph {nodes [
   Event PreConstruct()
   Event Construct()
   Event Tick()
  ]}]
variables []
widgets {
 info {inheritedWidgetCount=0, parentClass=&UserWidget, rootWidgetClass=&VerticalBox, widgetCount=6}
 widgets [
  VB_Header_Container {widgetClassPath=&VerticalBox, widget=VB_Header_Container}
  HB_CreatorLine {
   widgetClassPath=&HorizontalBox
   parent=VB_Header_Container
   widget=HB_CreatorLine
   properties {bOverride_Cursor=true}
   slot {class=VerticalBoxSlot}
  }
  Createdby {
   widgetClassPath=&CommonTextBlock
   parent=HB_CreatorLine
   widget=Createdby
   properties {
    mobileFontSizeMultiplier=1.2
    text=Created by
    colorAndOpacity {colorUseRule=UseColor_Specified, specifiedColor=<rgba 0.5 0.5 0.5 0.9>}
    font {
     bForceMonospaced=false
     bMaterialIsStencil=false
     fontMaterial=null
     fontObject=&HeadingNow
     letterSpacing=10
     monospacedWidth=1
     size=15.75
     skewAmount=0
     typefaceFontName=74 Regular
     outlineSettings {
      bApplyOutlineToDropShadows=false
      bMiteredCorners=false
      bSeparateFillAlpha=false
      outlineMaterial=null
      outlineSize=0
      outlineColor=<rgba 0 0 0 1>
     }
    }
    shadowOffset=<0 0>
   }
   slot {class=HorizontalBoxSlot, properties {verticalAlignment=VAlign_Center}}
  }
  Text_Divider {
   widgetClassPath=&CommonTextBlock
   parent=HB_CreatorLine
   widget=Text_Divider
   properties {
    mobileFontSizeMultiplier=1.2
    text="·"
    colorAndOpacity {colorUseRule=UseColor_Specified, specifiedColor=<rgba 0.5776 0.5776 0.5776 1>}
    font {
     bForceMonospaced=false
     bMaterialIsStencil=false
     fontMaterial=null
     fontObject=&HeadingNow
     letterSpacing=10
     monospacedWidth=1
     size=15.75
     skewAmount=0
     typefaceFontName=74 Regular
     outlineSettings {
      bApplyOutlineToDropShadows=false
      bMiteredCorners=false
      bSeparateFillAlpha=false
      outlineMaterial=null
      outlineSize=0
      outlineColor=<rgba 0 0 0 1>
     }
    }
    shadowOffset=<0 0>
   }
   slot {class=HorizontalBoxSlot, properties {verticalAlignment=VAlign_Center, padding=<ltrb 6 0 6 0>}}
  }
  Text_IslandOrigin {
   widgetClassPath=&CommonTextBlock
   bIsVariable=true
   parent=HB_CreatorLine
   widget=Text_IslandOrigin
   properties {
    mobileFontSizeMultiplier=1.2
    text=Ferins Bigwahed Grabber PlusUltra
    font {
     bForceMonospaced=false
     bMaterialIsStencil=false
     fontMaterial=null
     fontObject=&HeadingNow
     letterSpacing=10
     monospacedWidth=1
     size=15.75
     skewAmount=0
     typefaceFontName=74 Regular
     outlineSettings {
      bApplyOutlineToDropShadows=false
      bMiteredCorners=false
      bSeparateFillAlpha=false
      outlineMaterial=null
      outlineSize=0
      outlineColor=<rgba 0 0 0 1>
     }
    }
    shadowOffset=<0 0>
   }
   slot {class=HorizontalBoxSlot, properties {verticalAlignment=VAlign_Center}}
  }
  Text_IslandName {
   widgetClassPath=&CommonTextBlock
   bIsVariable=true
   parent=VB_Header_Container
   widget=Text_IslandName
   properties {
    bIsScrollingEnabled=false
    lineHeightPercentage=0.85
    text="2v2 Reload 2v2v2 realistic zonewar 문어임🎃👻"
    textTransformPolicy=ToUpper
    font {
     bForceMonospaced=false
     bMaterialIsStencil=false
     fontMaterial=null
     fontObject=&HeadingNow
     letterSpacing=0
     monospacedWidth=1
     size=32.25
     skewAmount=0
     typefaceFontName=86 Bold Italic
     outlineSettings {
      bApplyOutlineToDropShadows=false
      bMiteredCorners=false
      bSeparateFillAlpha=false
      outlineMaterial=null
      outlineSize=0
      outlineColor=<rgba 0 0 0 1>
     }
    }
    shadowOffset=<0 0>
   }
   slot {class=VerticalBoxSlot, properties {padding=<ltrb 0 0 0 20>}}
  }
 ]
}
