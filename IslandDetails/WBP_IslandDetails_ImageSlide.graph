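# Island Details image slide widget: a fixed 1080p SizeBox around a FortLazyImage that shows an image from a URL.
# SetImageURL stores the URL and starts a normal-priority SetBrushFromURL download into the image's material Thumbnail parameter, with a loading-background material.
# Destruct records whether a load is still in progress, and Construct re-requests the stored URL if one was and the URL is not empty, restarting interrupted loads.
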WidgetBlueprint /IslandDetails/WBP_IslandDetails_ImageSlide
graphs [
 EventGraph {nodes [
   Event Construct()
   > Branch(Condition=GetIsLoadRequired!(self) -> (IsLoadRequired:bool))
   > SetImageURL!(execute, self, InImageURL:String=Get ImageURL() -> (ImageURL:String)) -> (then)
   Event Destruct()
   > Set WasLoadingBeforeDestruct(WasLoadingBeforeDestruct:bool=IsLoading!(self:Common Lazy Image Object Reference=Get FortLazyImage() -> (FortLazyImage:Fort Lazy Image Object Reference)) -> (ReturnValue:bool)) -> (Output_Get:bool)
  ]}
 GetIsLoadRequired {nodes [
   GetIsLoadRequired() {is=title K2Node_FunctionEntry}
   n1 = IsEmpty(InString=Get ImageURL())
   > Return Node!(execute, IsLoadRequired:bool=AND Boolean(A=Get WasLoadingBeforeDestruct() -> (WasLoadingBeforeDestruct:bool), B=NOT Boolean(A=@n1)))
  ]}
 SetImageURL {nodes [
   n1 = SetImageURL!() -> (then, InImageURL:String) {is=title K2Node_FunctionEntry}
   > n2 = Set ImageURL(ImageURL:String=@n1) -> (Output_Get:String)
   > SetBrushFromURL!(
    execute
    self:Fort Lazy Image Object Reference=Get FortLazyImage()
    CMSImageURL:String=@n2
    bMatchSize:bool=false
    Delay:float=0
    DownloadPriority:ECMSDownloadPriority Enum=Normal
   ) -> (then)
  ]}
]
variables [
 ImageURL {type=string, default=""}
 WasLoadingBeforeDestruct {type=boolean}
]
widgets {
 info {inheritedWidgetCount=0, parentClass=&UserWidget, rootWidgetClass=&SizeBox, widgetCount=2}
 widgets [
  SizeBox_0 {
   widgetClassPath=&SizeBox
   widget=SizeBox_0
   properties {bOverride_HeightOverride=true, bOverride_WidthOverride=true, heightOverride=1080, widthOverride=1920}
  }
  FortLazyImage {
   widgetClassPath=&FortLazyImage
   bIsVariable=true
   parent=SizeBox_0
   widget=FortLazyImage
   properties {
    materialTextureParamName=Thumbnail
    brush {
     bIsDynamicallyLoaded=false
     drawAs=Image
     imageType=FullColor
     mirroring=NoMirror
     resourceName=None
     resourceObject=/IslandDetails/Materials/MI_UI_IslandDetails_Image
     tiling=NoTile
     imageSize=<1920 1080>
     margin=<ltrb 0 0 0 0>
     outlineSettings {
      bUseBrushTransparency=false
      roundingType=HalfHeightRadius
      width=0
      color {colorUseRule=UseColor_Specified, specifiedColor=<rgba 0 0 0 0>}
      cornerRadii=<0 0 0 0>
     }
     tintColor {colorUseRule=UseColor_Specified, specifiedColor=<rgba 1 1 1 1>}
     uVRegion {bIsValid=false, max=<0 0>, min=<0 0>}
    }
    loadingBackgroundBrush {
     bIsDynamicallyLoaded=false
     drawAs=Image
     imageType=FullColor
     mirroring=NoMirror
     resourceName=None
     resourceObject=/IslandDetails/Materials/M_UI_IslandDetails_Loading
     tiling=NoTile
     imageSize=<1920 1080>
     margin=<ltrb 0 0 0 0>
     outlineSettings {
      bUseBrushTransparency=false
      roundingType=HalfHeightRadius
      width=0
      color {colorUseRule=UseColor_Specified, specifiedColor=<rgba 0 0 0 0>}
      cornerRadii=<0 0 0 0>
     }
     tintColor {colorUseRule=UseColor_Specified, specifiedColor=<rgba 1 1 1 1>}
     uVRegion {bIsValid=false, max=<0 0>, min=<0 0>}
    }
    loadingThrobberBrush {
     bIsDynamicallyLoaded=false
     drawAs=Image
     imageType=FullColor
     mirroring=NoMirror
     resourceName=None
     resourceObject=/Game/Global/Textures/Default/Blanks/T_Blank_White_ZeroAlpha_sRGBColor
     tiling=NoTile
     imageSize=<1920 1080>
     margin=<ltrb 0 0 0 0>
     outlineSettings {
      bUseBrushTransparency=false
      roundingType=HalfHeightRadius
      width=0
      color {colorUseRule=UseColor_Specified, specifiedColor=<rgba 0 0 0 0>}
      cornerRadii=<0 0 0 0>
     }
     tintColor {colorUseRule=UseColor_Specified, specifiedColor=<rgba 1 1 1 1>}
     uVRegion {bIsValid=false, max=<0 0>, min=<0 0>}
    }
   }
   slot {class=SizeBoxSlot}
  }
 ]
}
