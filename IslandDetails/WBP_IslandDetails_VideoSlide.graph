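# WBP_IslandDetails_VideoSlide is a widget for a full-HD video slide in the island details UI.
# A fixed 1920x1080 SizeBox holds an Overlay that layers an obscuring UIKit backplate beneath an ActivityVideoCycleWidget, both filling the frame.
# The video widget renders through the M_UI_IslandDetails_Video2 material and is hit-test invisible.
# The event graph is empty, so the asset itself only defines layout.
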WidgetBlueprint /IslandDetails/WBP_IslandDetails_VideoSlide
graphs [EventGraph {links [], nodes []}]
variables []
widgets {
 info {inheritedWidgetCount=0, parentClass=&UserWidget, rootWidgetClass=&SizeBox, widgetCount=4}
 widgets [
  SizeBox_0 {
   widgetClassPath=&SizeBox
   widget=SizeBox_0
   properties {bOverride_HeightOverride=true, bOverride_WidthOverride=true, heightOverride=1080, widthOverride=1920}
  }
  Overlay_31 {widgetClassPath=&Overlay, parent=SizeBox_0, widget=Overlay_31, slot {class=SizeBoxSlot}}
  Backplate {
   widgetClassPath=&WBP_UIKit_Backplate
   bIsVariable=true
   parent=Overlay_31
   widget=Backplate
   properties {brightness=Obscure}
   slot {class=OverlaySlot, properties {horizontalAlignment=HAlign_Fill, verticalAlignment=VAlign_Fill}}
  }
  VideoWidget {
   widgetClassPath=/FortniteCoreUI/UI/DiscoveryShared/GameDetails/ActivityVideoCycleWidget
   bIsVariable=true
   parent=Overlay_31
   widget=VideoWidget
   properties {videoMaterial=/IslandDetails/Materials/M_UI_IslandDetails_Video2, visibility=HitTestInvisible}
   slot {class=OverlaySlot, properties {horizontalAlignment=HAlign_Fill, verticalAlignment=VAlign_Fill}}
  }
 ]
}
