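# Island Details media gallery widget showing an island's video and images in a carousel with arrows and pips.
# SetMedia rebuilds slides, leading with video unless absent, unplayable, locked or restricted, then one slide per image URL.
# Page changes pause the previous video, update pips, then play video or auto-scroll images when controls show and IsExpanded is false.
# Ended videos replay if expanded, otherwise advance via GoRight, which calls PreviousPage (GoLeft calls NextPage).
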WidgetBlueprint /IslandDetails/WBP_IslandDetails_MediaGallery
graphs [
 AddImage {nodes [
   n1 = AddImage!() -> (then, ImageURL:String) {is=title K2Node_FunctionEntry}
   > n2 = Construct Object from Class(Class=&/WBP_IslandDetails_ImageSlide.WBP_IslandDetails_ImageSlide_C) -> (ReturnValue:WBP Island Details Image Slide Object Reference)
   > AddChild(self=Get Carousel() -> (Carousel:Fort Widget Carousel Object Reference), Content=@n2)
   > SetImageURL!(execute, self:WBP Island Details Image Slide Object Reference=@n2, InImageURL:String=@n1) -> (then)
  ]}
 AddVideo {nodes [
   n1 = AddVideo!() -> (then, ViewModel:Discover Activity View Model Object Reference) {is=title K2Node_FunctionEntry}
   n2 = IsActivityRestricted!(self:Discover Activity View Model Object Reference=@n1) -> (ReturnValue:bool)
   n3 = GetRuntimeOptions()
   n4 = IsActivityLocked!(self:Fort Discover Link Entry Tile Item VM Object Reference=@n1) -> (ReturnValue:bool)
   > n5 = Branch(Condition=AND Boolean!(
     A=HasVideo!(self:Fort Discover Link Entry Tile Item VM Object Reference=@n1) -> (ReturnValue:bool)
     B=CanPlayVideosOnCurrentPlatform!(self:Fort Runtime Options Object Reference=@n3) -> (ReturnValue:bool)
     C:bool=NOT Boolean(A=@n4)
     D:bool=NOT Boolean(A=@n2)
    ) -> (ReturnValue))
   > n6 = Construct Object from Class(Class=&/WBP_IslandDetails_VideoSlide.WBP_IslandDetails_VideoSlide_C) -> (ReturnValue:WBP Island Details Video Slide Object Reference)
   > AddChild(self=Get Carousel(), Content=@n6)
   > Bind Event to Video Ended(
    self:Activity Video Cycle Widget Object Reference=Get VideoWidget!(self:WBP Island Details Video Slide Object Reference=@n6) -> (VideoWidget:Activity Video Cycle Widget Object Reference)
    Delegate=Create Event()
   ) {is=title K2Node_AddDelegate}
   > Return Node!(execute, IsVideoAdded:bool=true)
   Return Node(execute=@n5.else, IsVideoAdded=false)
  ]}
 ClearSlides {nodes [
   ClearSlides() {is=title K2Node_FunctionEntry}
   > n1 = For Each Loop(Array:Array of Widget Object References=GetAllChildren(self=Get Carousel())) -> (Array Element:Widget Object Reference)
   ClearChildren(execute=@n1.Completed, self=Get Carousel())
   > Set LastActiveSlideIndex(LastActiveSlideIndex:int=-1) -> (Output_Get:int=0)
   n2 = Cast To WBP_IslandDetails_VideoSlide!(execute=@n1.LoopBody, Object=@n1) -> (then, CastFailed, AsWBP Island Details Video Slide:WBP Island Details Video Slide Object Reference) {is=title K2Node_DynamicCast}
   > Unbind all Events from Video Ended!(execute:Exec, self:Activity Video Cycle Widget Object Reference=Get VideoWidget(self=@n2)) -> (then:Exec) {is=title K2Node_ClearDelegate}
  ]}
 EventGraph {nodes [
   "On Clicked (ArrowButton_Left)"!() -> (OutputDelegate, then, Button:Common Button Base Object Reference) {is=title K2Node_ComponentBoundEvent}
   > GoLeft()
   "On Clicked (ArrowButton_Right)"!() -> (OutputDelegate, then, Button:Common Button Base Object Reference) {is=title K2Node_ComponentBoundEvent}
   > GoRight()
   n1 = OnCarouselIndexChange!() -> (OutputDelegate, then, CarouselWidget:Common Widget Carousel Object Reference, CurrentPageIndex:int) {is=title K2Node_CustomEvent}
   > UpdateActiveSlide()
   Event Construct()
   n2 = Get LastActiveSlideIndex() -> (LastActiveSlideIndex:int=0)
   > Initialize!(execute, self, Initial Slide Index:int=SelectInt(A=@n2, B=0, bPickA="Not Equal (Integer)"(A=@n2, B=-1))) -> (then)
   > Bind Event to On Current Page Index Changed(self:Common Widget Carousel Object Reference=Get Carousel(), Delegate=@n1) {is=title K2Node_AddDelegate}
   Event Destruct()
   > Unbind Event from On Current Page Index Changed(self:Common Widget Carousel Object Reference=Get Carousel(), Delegate=@n1) {is=title K2Node_RemoveDelegate}
  ]}
 GetVideoSlideAt {nodes [
   n1 = Inputs!() -> (In:Exec, Index:int)
   > n2 = Cast To WBP_IslandDetails_VideoSlide(Object=GetChildAt!(self:Panel Widget Object Reference=Get Carousel(), Index:int=@n1) -> (ReturnValue:Widget Object Reference))
   > Outputs!(Is Valid:Exec, Is Not Valid:Exec=@n2.CastFailed, Video Slide:WBP Island Details Video Slide Object Reference=@n2)
  ]}
 GoLeft {nodes [
   GoLeft() {is=title K2Node_FunctionEntry}
   > NextPage(self:Common Widget Carousel Object Reference=Get Carousel())
  ]}
 GoRight {nodes [
   GoRight() {is=title K2Node_FunctionEntry}
   > PreviousPage(self:Common Widget Carousel Object Reference=Get Carousel())
  ]}
 Initialize {nodes [
   Initialize!() -> (then, Initial Slide Index:int) {is=title K2Node_FunctionEntry}
   > Branch(Condition=HasAnyChildren!(self:Panel Widget Object Reference=Get Carousel()) -> (ReturnValue:bool))
   > InitializePips()
   > SnapActiveWidgetIndex!(execute, self:Fort Widget Carousel Object Reference=Get Carousel(), Index:int=Get Initial Slide Index() -> (Initial Slide Index:int=0)) -> (then)
   > UpdateActiveSlide()
  ]}
 InitializePips {nodes [
   InitializePips() {is=title K2Node_FunctionEntry}
   n1 = GetChildrenCount!(self:Panel Widget Object Reference=Get Carousel()) -> (ReturnValue:int=0)
   > SetPipsCount!(
    execute
    self:WBP Island Details Pip List Object Reference=Get WBP_IslandDetails_Pip_List() -> (WBP_IslandDetails_Pip_List:WBP Island Details Pip List Object Reference)
    Count:int=@n1
    Duration Value:double=0
    Anim Pip:bool=true
    Show Progress:bool=false
   ) -> (then)
   > SetVisibility(
    self=Get HB_Controls() -> (HB_Controls:Horizontal Box Object Reference)
    InVisibility=Select(Option 0=Collapsed, Option 1=SelfHitTestInvisible, Index="integer > integer"(A=@n1, B=1))
   )
  ]}
 OnVideoEnded {nodes [
   OnVideoEnded() {is=title K2Node_FunctionEntry}
   > n1 = Get Video Slide At!(In:Exec, Index:int=GetActiveWidgetIndex!(self:Common Widget Carousel Object Reference=Get Carousel()) -> (ReturnValue:int=0)) -> (Is Valid, Is Not Valid, Video Slide:WBP Island Details Video Slide Object Reference) {is=title K2Node_MacroInstance}
   > n2 = Branch(Condition=NOT Boolean(A=Get IsExpanded() -> (IsExpanded:bool)))
   > GoRight() {is=title K2Node_CallFunction}
   PlayVideo(execute=@n2.else, self:Fort Activity Video Cycle Object Reference=Get VideoWidget(self=@n1))
  ]}
 SetMedia {nodes [
   n1 = SetMedia!() -> (then, ImageURLs:Array of Strings, InViewModel:Discover Activity View Model Object Reference) {is=title K2Node_FunctionEntry}
   > Set ViewModel(ViewModel:Discover Activity View Model Object Reference=@n1.InViewModel) -> (Output_Get:Discover Activity View Model Object Reference)
   > ClearSlides() {is=title K2Node_CallFunction}
   > n2 = Sequence()
   > AddVideo!(
    execute
    self
    ViewModel:Discover Activity View Model Object Reference=Get InViewModel() -> (InViewModel:Discover Activity View Model Object Reference)
   ) -> (then, IsVideoAdded:bool) {is=title K2Node_CallFunction}
   > n3 = For Each Loop(Array:Array of Strings=Get ImageURLs() -> (ImageURLs:Array of Strings)) -> (Array Element:String)
   > AddImage!(execute, self, ImageURL:String=@n3) -> (then) {is=title K2Node_CallFunction}
   Initialize(execute=@n2.then_1, Initial Slide Index=0) {is=title K2Node_CallFunction}
  ]}
 StartAutoScrolling {nodes [
   StartAutoScrolling() {is=title K2Node_FunctionEntry}
   n1 = Get HB_Controls()
   n2 = Get IsExpanded()
   > Branch(Condition=AND Boolean!(A=IsVisible(self=@n1), B=NOT Boolean(A=@n2)) -> (ReturnValue))
   > BeginAutoScrolling!(
    execute
    self:Common Widget Carousel Object Reference=Get Carousel()
    ScrollInterval:float=Get Autoscroll Interval() -> (Autoscroll Interval:float)
   ) -> (then)
  ]}
 StopAutoScrolling {nodes [
   StopAutoScrolling() {is=title K2Node_FunctionEntry}
   > EndAutoScrolling(self:Common Widget Carousel Object Reference=Get Carousel())
  ]}
 UpdateActiveSlide {nodes [
   UpdateActiveSlide() {is=title K2Node_FunctionEntry}
   > n1 = Sequence!(execute) -> (then_0, then_1, then_2:Exec, then_3:Exec)
   > Branch(Condition="Not Equal (Integer)"(A=Get LastActiveSlideIndex() -> (LastActiveSlideIndex=0), B=-1))
   > n2 = Get Video Slide At(Index=Get LastActiveSlideIndex() -> (LastActiveSlideIndex=0))
   > PauseVideo(self:Fort Activity Video Cycle Object Reference=Get VideoWidget(self=@n2))
   SetActivePip!(
    execute=@n1.then_1
    self:WBP Island Details Pip List Object Reference=Get WBP_IslandDetails_Pip_List()
    Index:int=GetActiveWidgetIndex(self=Get Carousel()) -> (ReturnValue=0)
   ) -> (then)
   n3 = Get Video Slide At(In=@n1.then_2, Index=GetActiveWidgetIndex(self=Get Carousel()) -> (ReturnValue=0))
   StartAutoScrolling(execute=@n3.Is Not Valid) {is=title K2Node_CallFunction}
   StopAutoScrolling(execute=@n3.Is Valid) {is=title K2Node_CallFunction}
   n4 = Get VideoWidget(self=@n3)
   > n5 = IsVideoCycleInProgress!(execute, self:Fort Activity Video Cycle Object Reference=@n4) -> (then, ReturnValue:bool)
   > n6 = Branch(Condition=@n5)
   > PlayVideo(self=@n4)
   n7 = Get ViewModel() -> (ViewModel:Discover Activity View Model Object Reference)
   StopVideoCycleFromViewModel!(execute=@n6.else, self:Fort Activity Video Cycle Object Reference=@n4, ActivityViewModel:Discover Activity View Model Object Reference=@n7) -> (then)
   > StartVideoCycleFromViewModel!(execute, self:Fort Activity Video Cycle Object Reference=@n4, ActivityViewModel:Discover Activity View Model Object Reference=@n7) -> (then)
   Set LastActiveSlideIndex(execute=@n1.then_3, LastActiveSlideIndex=GetActiveWidgetIndex(self=Get Carousel()) -> (ReturnValue=0)) -> (Output_Get=0)
  ]}
]
variables [
 Autoscroll Interval {type=number, default=8}
 IsExpanded {type=boolean}
 LastActiveSlideIndex {type=integer, default=-1}
 ViewModel {type=FortActivityViewModel, default=null}
]
widgets {
 info {inheritedWidgetCount=0, parentClass=&UserWidget, rootWidgetClass=&Overlay, widgetCount=9}
 widgets [
  Overlay_3 {widgetClassPath=&Overlay, widget=Overlay_3}
  ScaleBox_Carousel {
   widgetClassPath=&ScaleBox
   parent=Overlay_3
   widget=ScaleBox_Carousel
   slot {class=OverlaySlot, properties {horizontalAlignment=HAlign_Center, verticalAlignment=VAlign_Center}}
  }
  Carousel {
   widgetClassPath=/Script/FortniteUI.FortWidgetCarousel
   bIsVariable=true
   parent=ScaleBox_Carousel
   widget=Carousel
   properties {bOverride_Cursor=true}
   slot {class=ScaleBoxSlot, properties {horizontalAlignment=HAlign_Fill, verticalAlignment=VAlign_Fill}}
  }
  ImageSlide_Placeholder {
   widgetClassPath=&/WBP_IslandDetails_ImageSlide
   bIsVariable=true
   parent=Carousel
   widget=ImageSlide_Placeholder
   slot {class=PanelSlot}
  }
  CaptureForPostBufferUpdate {
   widgetClassPath=&WBP_CaptureForPostBufferUpdate
   parent=Overlay_3
   widget=CaptureForPostBufferUpdate
   slot {class=OverlaySlot, properties {horizontalAlignment=HAlign_Fill, verticalAlignment=VAlign_Fill}}
  }
  HB_Controls {
   widgetClassPath=&HorizontalBox
   bIsVariable=true
   parent=Overlay_3
   widget=HB_Controls
   slot {class=OverlaySlot, properties {horizontalAlignment=HAlign_Center, verticalAlignment=VAlign_Bottom, padding=<ltrb 80 10 80 32>}}
  }
  ArrowButton_Left {
   widgetClassPath=&/WBP_IslandDetails_ArrowButton
   bIsVariable=true
   parent=HB_Controls
   widget=ArrowButton_Left
   properties {
    bHideInputActionWithKeyboard=true
    bIsFocusable=false
    inverseArrow=true
    minWidth=48
    mobileMinHeight=88
    triggeringInputAction {dataTable=&InputActionDataTable, rowName=GenericLeftTrigger}
   }
   slot {class=HorizontalBoxSlot}
  }
  WBP_IslandDetails_Pip_List {
   widgetClassPath=&/WBP_IslandDetails_Pip_List
   bIsVariable=true
   parent=HB_Controls
   widget=WBP_IslandDetails_Pip_List
   properties {pipSize=<12 12>}
   slot {class=HorizontalBoxSlot, properties {horizontalAlignment=HAlign_Center, verticalAlignment=VAlign_Center, padding=<ltrb 24 0 24 0>}}
  }
  ArrowButton_Right {
   widgetClassPath=&WBP_IslandDetails_ArrowButton
   bIsVariable=true
   parent=HB_Controls
   widget=ArrowButton_Right
   properties {bHideInputActionWithKeyboard=true, bIsFocusable=false, mobileMinHeight=88}
   slot {class=HorizontalBoxSlot}
  }
 ]
}
