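# Activatable modal that shows the island details media gallery expanded over a click-catching scrim.
# AddMediaGallery stores the passed-in gallery, sets it expanded and places it in a named slot inside an invisible FocusButton, which serves as the desired focus target.
# NavigateSlides turns left and right navigation into GoLeft and GoRight on the gallery, returning focus to that button.
# Activation plays a short scale-up animation, and the Back button deactivates the modal.
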WidgetBlueprint /IslandDetails/WBP_IslandDetails_MediaGallery_Modal
animations [Anim_Expand {
  binding_count=1
  length_seconds=0.25
  object=&self:Anim_Expand
  sequence {bindings [NamedSlot_MediaGallery {
     id=#1
     possessed_class=NamedSlot
     tracks [{
       type=MovieScene2DTransformTrack
       sections [{
         type=MovieScene2DTransformSection
         end_frame=4
         start_frame=-1
         channels [
          Scale.X {type=MovieSceneScriptingFloatChannel, keys [[-1,0.8],[4,1],[6,1]]}
          Scale.Y {type=MovieSceneScriptingFloatChannel, keys [[-1,0.8],[4,1],[6,1]]}
         ]
        }]
      }]
    }]}
 }]
graphs [
 AddMediaGallery {nodes [
   n1 = AddMediaGallery!() -> (then, InMediaGallery:WBP Island Details Media Gallery Object Reference) {is=title K2Node_FunctionEntry}
   > n2 = Set MediaGallery(MediaGallery:WBP Island Details Media Gallery Object Reference=@n1) -> (Output_Get:WBP Island Details Media Gallery Object Reference)
   > Set IsExpanded!(execute, IsExpanded:bool=true, self:WBP Island Details Media Gallery Object Reference=@n2) -> (then, Output_Get:bool)
   > AddChild(self=Get NamedSlot_MediaGallery() -> (NamedSlot_MediaGallery:Named Slot Object Reference), Content=@n2)
  ]}
 BP_GetDesiredFocusTarget {nodes [
   Get Desired Focus Target()
   > Return Node!(execute, ReturnValue:Widget Object Reference=Get FocusButton() -> (FocusButton:Button Object Reference))
  ]}
 EventGraph {nodes [
   "On Clicked (Button_Back)"!() -> (OutputDelegate, then, Button:Common Button Base Object Reference) {is=title K2Node_ComponentBoundEvent}
   > DeactivateWidget()
   Event On Activated()
   > PlayAnimation(InAnimation=Get Anim_Expand() -> (Anim_Expand:Widget Animation Object Reference))
  ]}
 NavigateSlides {nodes [
   n1 = NavigateSlides!() -> (then, Navigation:EUINavigation Enum) {is=title K2Node_FunctionEntry}
   > n2 = Switch on EUINavigation!(execute, Selection:EUINavigation Enum=@n1) -> (Left:Exec, Right:Exec, Up:Exec, Down:Exec, Next:Exec, Previous:Exec, Invalid:Exec) {is=title K2Node_SwitchEnum}
   > n3 = GoLeft(self:WBP Island Details Media Gallery Object Reference=Get MediaGallery() -> (MediaGallery:WBP Island Details Media Gallery Object Reference))
   Return Node(execute=[@n3.then, @n4.then], ReturnValue=Get FocusButton())
   n4 = GoRight(execute=@n2.Right, self:WBP Island Details Media Gallery Object Reference=Get MediaGallery())
  ]}
]
variables [MediaGallery {type=WBP_IslandDetails_MediaGallery_C, default=null}]
widgets {
 info {
  inheritedWidgetCount=0
  namedSlotCount=1
  parentClass=&CommonActivatableWidget
  rootWidgetClass=&Overlay
  widgetCount=8
 }
 widgets [
  Overlay {widgetClassPath=&Overlay, widget=Overlay}
  WBP_UIKit_Scrim {
   widgetClassPath=/FortUILibrary/Systems/Dialog/WBP_UIKit_Scrim
   bIsVariable=true
   parent=Overlay
   widget=WBP_UIKit_Scrim
   properties {useClickCatcher=true, visibility=Visible}
   slot {class=OverlaySlot, properties {horizontalAlignment=HAlign_Fill, verticalAlignment=VAlign_Fill}}
  }
  FocusButton {
   widgetClassPath=&Button
   parent=Overlay
   widget=FocusButton
   properties {
    navigation=&self:WidgetTree.FocusButton.WidgetNavigation_0
    visibility=SelfHitTestInvisible
    widgetStyle {
     clickedSlateSound {resourceObject=null}
     disabled {
      bIsDynamicallyLoaded=false
      drawAs=NoDrawType
      imageType=NoImage
      mirroring=NoMirror
      resourceName=None
      resourceObject=null
      tiling=NoTile
      imageSize=<0 0>
      margin=<ltrb 0 0 0 0>
      outlineSettings {
       bUseBrushTransparency=false
       roundingType=HalfHeightRadius
       width=0
       color {colorUseRule=UseColor_Specified, specifiedColor=<rgba 0 0 0 0>}
       cornerRadii=<0 0 0 0>
      }
      tintColor {colorUseRule=UseColor_Specified, specifiedColor=<rgba 1 1 1 1>}
      uVRegion {bIsValid=false, max=<0 0>, min=<0 0>}
     }
     disabledForeground {colorUseRule=UseColor_Specified, specifiedColor=<rgba 0.5271 0.5271 0.5271 1>}
     hovered {
      bIsDynamicallyLoaded=false
      drawAs=NoDrawType
      imageType=NoImage
      mirroring=NoMirror
      resourceName=None
      resourceObject=null
      tiling=NoTile
      imageSize=<32 32>
      margin=<ltrb 0 0 0 0>
      outlineSettings {
       bUseBrushTransparency=true
       roundingType=FixedRadius
       width=1
       color {colorUseRule=UseColor_Specified, specifiedColor=<rgba 0.7243 0.7243 0.7243 1>}
       cornerRadii=<4 4 4 4>
      }
      tintColor {colorUseRule=UseColor_Specified, specifiedColor=<rgba 0.7243 0.7243 0.7243 1>}
      uVRegion {bIsValid=false, max=<0 0>, min=<0 0>}
     }
     hoveredForeground {colorUseRule=UseColor_Specified, specifiedColor=<rgba 0 1 1 1>}
     hoveredSlateSound {resourceObject=null}
     normal {
      bIsDynamicallyLoaded=false
      drawAs=NoDrawType
      imageType=NoImage
      mirroring=NoMirror
      resourceName=None
      resourceObject=null
      tiling=NoTile
      imageSize=<32 32>
      margin=<ltrb 0 0 0 0>
      outlineSettings {
       bUseBrushTransparency=true
       roundingType=FixedRadius
       width=1
       color {colorUseRule=UseColor_Specified, specifiedColor=<rgba 0.6951 0.6951 0.6951 1>}
       cornerRadii=<4 4 4 4>
      }
      tintColor {colorUseRule=UseColor_Specified, specifiedColor=<rgba 0.4955 0.4955 0.4955 1>}
      uVRegion {bIsValid=false, max=<0 0>, min=<0 0>}
     }
     normalForeground {colorUseRule=UseColor_Specified, specifiedColor=<rgba 0 1 1 1>}
     normalPadding=<ltrb 0 0 0 0>
     pressed {
      bIsDynamicallyLoaded=false
      drawAs=NoDrawType
      imageType=NoImage
      mirroring=NoMirror
      resourceName=None
      resourceObject=null
      tiling=NoTile
      imageSize=<32 32>
      margin=<ltrb 0 0 0 0>
      outlineSettings {
       bUseBrushTransparency=true
       roundingType=FixedRadius
       width=1
       color {colorUseRule=UseColor_Specified, specifiedColor=<rgba 0.7243 0.7243 0.7243 1>}
       cornerRadii=<4 4 4 4>
      }
      tintColor {colorUseRule=UseColor_Specified, specifiedColor=<rgba 0.3843 0.3843 0.3843 1>}
      uVRegion {bIsValid=false, max=<0 0>, min=<0 0>}
     }
     pressedForeground {colorUseRule=UseColor_Specified, specifiedColor=<rgba 0 1 1 1>}
     pressedPadding=<ltrb 0 0 0 0>
     pressedSlateSound {resourceObject=null}
    }
   }
   slot {class=OverlaySlot, properties {horizontalAlignment=HAlign_Fill, verticalAlignment=VAlign_Fill}}
  }
  NamedSlot_MediaGallery {
   widgetClassPath=&NamedSlot
   bIsVariable=true
   parent=FocusButton
   widget=NamedSlot_MediaGallery
   slot {class=ButtonSlot, properties {horizontalAlignment=HAlign_Fill, verticalAlignment=VAlign_Fill, padding=<ltrb 0 0 0 0>}}
  }
  SafeZone_Prompts {
   widgetClassPath=&SafeZone
   bIsVariable=true
   parent=Overlay
   widget=SafeZone_Prompts
   properties {padTop=false}
   slot {class=OverlaySlot, properties {horizontalAlignment=HAlign_Right, verticalAlignment=VAlign_Bottom, padding=<ltrb 80 0 80 32>}}
  }
  WBP_ActionBarContainer {
   widgetClassPath=/Game/UI/Frontend/BottomBar/WBP_ActionBarContainer
   bIsVariable=true
   parent=SafeZone_Prompts
   widget=WBP_ActionBarContainer
   slot {class=SafeZoneSlot}
  }
  HorizontalBox_1 {
   widgetClassPath=&HorizontalBox
   namedSlotHost=WBP_ActionBarContainer
   widget=HorizontalBox_1
   properties {flowDirectionPreference=Culture}
  }
  Button_Back {
   widgetClassPath=/FortUILibrary/UIKit/Buttons/WBP_UIKit_Button_Quiet
   bIsVariable=true
   parent=HorizontalBox_1
   widget=Button_Back
   properties {bIsFocusable=false, text=Back, triggeringInputAction {dataTable=&InputActionDataTable, rowName=GenericBack}}
   slot {class=HorizontalBoxSlot, properties {horizontalAlignment=HAlign_Right, verticalAlignment=VAlign_Bottom}}
  }
 ]
}
