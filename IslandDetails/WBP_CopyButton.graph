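# WBP_CopyButton is a generic UIKit button subclass in IslandDetails that copies its text to the clipboard when clicked.
# SetTextToCopy stores the text and also sets it as the button label.
# On click, it calls the parent handler, then, if the text is non-empty, copies it and plays ShowCopied.
# That animation slides and fades a green "Copied!" label in from a right-aligned FortVisualAttachment, then fades it out.
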WidgetBlueprint /IslandDetails/WBP_CopyButton
animations [ShowCopied {
  binding_count=1
  length_seconds=1.55
  object=&self:ShowCopied
  sequence {bindings [CommonTextBlock_76 {
     id=#1
     possessed_class=CommonTextBlock
     tracks [
      {
       type=MovieScene2DTransformTrack
       sections [{
         type=MovieScene2DTransformSection
         end_frame=31
         start_frame=0
         channels [
          Translation.X {type=MovieSceneScriptingFloatChannel, keys [[0,-100],[6.2727,0],[25,5]]}
          Translation.Y {type=MovieSceneScriptingFloatChannel, const [0,1]}
          Angle {type=MovieSceneScriptingFloatChannel, const [0,1]}
          Scale.X {type=MovieSceneScriptingFloatChannel, const [1,1]}
          Scale.Y {type=MovieSceneScriptingFloatChannel, keys [[0,0.6],[6.2727,1]]}
          Shear.X {type=MovieSceneScriptingFloatChannel, keys [[0,-20],[6.2727,0]]}
          Shear.Y {type=MovieSceneScriptingFloatChannel, const [0,1]}
         ]
        }]
      }
      {
       type=MovieSceneFloatTrack
       display_name=Render Opacity
       sections [{
         type=MovieSceneFloatSection
         end_frame=31
         start_frame=0
         channels [None {type=MovieSceneScriptingFloatChannel, thinned_from=5, keys [[0,0],[3.1363,1],[25,1],[30,0]]}]
        }]
      }
     ]
    }]}
 }]
graphs [
 EventGraph {nodes [
   Event On Clicked()
   > Parent: On Clicked() {is=title K2Node_CallParentFunction}
   n1 = Get TextToCopy() -> (TextToCopy:Text)
   > Branch(Condition=NOT Boolean(A=TextIsEmpty(InText=@n1)))
   > CopyToClipboard!(execute, StringToCopy:String="To String (Text)"(InText=Get TextToCopy())) -> (then)
   > PlayAnimationForward(InAnimation=Get ShowCopied() -> (ShowCopied:Widget Animation Object Reference))
  ]}
 SetTextToCopy {nodes [
   n1 = SetTextToCopy!() -> (then, StringToCopy:Text) {is=title K2Node_FunctionEntry}
   > n2 = Set TextToCopy(TextToCopy:Text=@n1) -> (Output_Get:Text)
   > SetText!(execute, self:WBP UIKit Button Generic Base Object Reference, Text:Text=@n2) -> (then)
  ]}
]
variables [TextToCopy {type=string, default=""}]
widgets {
 info {
  inheritedWidgetCount=0
  namedSlotCount=1
  parentClass=/FortUILibrary/UIKit/Buttons/WBP_UIKit_Button_Generic.WBP_UIKit_Button_Generic_C
  rootWidgetClass=null
  widgetCount=3
 }
 widgets [
  Overlay_56 {widgetClassPath=&Overlay, widget=Overlay_56}
  FortVisualAttachment_46 {
   widgetClassPath=&FortVisualAttachment
   parent=Overlay_56
   widget=FortVisualAttachment_46
   properties {contentAnchor=<0 0.5>}
   slot {class=OverlaySlot, properties {horizontalAlignment=HAlign_Right, verticalAlignment=VAlign_Center}}
  }
  CommonTextBlock_76 {
   widgetClassPath=&CommonTextBlock
   parent=FortVisualAttachment_46
   widget=CommonTextBlock_76
   properties {
    renderOpacity=0
    text="Copied!"
    colorAndOpacity {colorUseRule=UseColor_Specified, specifiedColor=<rgba 0.0369 0.807 0.2159 1>}
    font {
     bForceMonospaced=false
     bMaterialIsStencil=false
     fontMaterial=null
     fontObject=&HeadingNow
     letterSpacing=0
     monospacedWidth=1
     size=14.25
     skewAmount=0
     typefaceFontName=74 Regular
     outlineSettings {
      bApplyOutlineToDropShadows=false
      bMiteredCorners=false
      bSeparateFillAlpha=false
      outlineMaterial=null
      outlineSize=0
      outlineColor=<rgba 0 0 0 1>
     }
    }
    shadowOffset=<0 0>
   }
   slot {class=SizeBoxSlot, properties {padding=<ltrb 10 0 0 0>}}
  }
 ]
}
