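# Island details stat row widget showing an icon, a formatted stat text and an optional copy button.
# PreConstruct creates the button in a named slot when ShowCopyIcon is set, applies the icon, formats the text and passes the raw Stat for copying.
# With a button present, input-method changes show it for mouse/touch or Text_Stat for gamepad, adjusting padding.
# SetDisplayed collapses the row when requested or when its text is empty.
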WidgetBlueprint /IslandDetails/WBP_IslandDetails_Stats_Entry
graphs [
 CreateCopyButton {nodes [
   CreateCopyButton() {is=title K2Node_FunctionEntry}
   > n1 = Branch(Condition=Get ShowCopyIcon() -> (ShowCopyIcon:bool))
   > n2 = Create Widget!(
    execute:Exec
    Class:User Widget Class Reference=&/WBP_CopyButton.WBP_CopyButton_C
    OwningPlayer:Player Controller Object Reference
    Style:Common Button Style Class Reference=/FortUILibrary/UIKit/Buttons/Blocks/ButtonStyle_UIKit_Empty.ButtonStyle_UIKit_Empty_C
    bApplyAlphaOnDisable:bool=False
    bLocked:bool=False
    bSelectable:bool=False
    bShouldSelectUponReceivingFocus:bool=False
    bInteractableWhenSelected:bool=False
    bToggleable:bool=False
    bTriggerClickedAfterSelection:bool=False
    bDisplayInputActionWhenNotInteractable:bool=False
    bHideInputActionWithKeyboard:bool=False
    bShouldUseFallbackDefaultInputAction:bool=True
    bRequiresHold:bool=False
    HoldData:Common UIHold Data Class Reference=&CommonUI_HoldData_Base_C
    ClickMethod:EButtonClickMethod Enum=DownAndUp
    InputPriority:int=0
    TriggeringInputAction:Data Table Row Handle Structure
   ) -> (then:Exec, ReturnValue:WBP Copy Button Object Reference)
   > AddChild(self=Get Slot_CopyButton() -> (Slot_CopyButton:Named Slot Object Reference), Content=@n2)
   > SetVisibility(self=Get Slot_CopyButton(), InVisibility=SelfHitTestInvisible)
   ClearChildren(execute=@n1.else, self=Get Slot_CopyButton())
   > SetVisibility(self=Get Slot_CopyButton(), InVisibility=Collapsed)
  ]}
 EventGraph {nodes [
   n1 = OnInputMethodChanged!() -> (OutputDelegate, then, bNewInputType:ECommonInputType Enum) {is=title K2Node_CustomEvent}
   > UpdateCopyButtonVisibility!(execute, self, Index:ECommonInputType Enum=@n1.bNewInputType) -> (then)
   n2 = Event PreConstruct()
   > CreateCopyButton() {is=title K2Node_CallFunction}
   > SetBrush(
    self=Get Image_Icon() -> (Image_Icon:Image Object Reference)
    InBrush=Make Slate Brush!(
     TintColor:Slate Color Structure=(SpecifiedColor=(R=1,G=1,B=1,A=1),ColorUseRule=UseColor_Specified)
     DrawAs:ESlateBrushDrawType Enum=Image
     Tiling:ESlateBrushTileType Enum=NoTile
     Mirroring:ESlateBrushMirrorType Enum=NoMirror
     ImageSize:Vector 2D Structure=Get IconSize() -> (IconSize:Vector 2D Structure)
     Margin:Margin Structure=(Left=0,Top=0,Right=0,Bottom=0)
     ResourceObject:Object Reference=Get Icon() -> (Icon:Object Reference)
     OutlineSettings:Slate Brush Outline Settings Structure=(CornerRadii=(X=0,Y=0,Z=0,W=0),Color=(SpecifiedColor=(R=0,G=0,B=0,A=0),ColorUseRule=UseColor_Specified),Width=0,RoundingType=HalfHeightRadius,bUseBrushTransparency=False)
    ) -> (SlateBrush:Slate Brush Structure) {is=title K2Node_MakeStruct}
   )
   > FormatAndSetText!(execute, self, Stat:Text=Get Stat() -> (Stat:Text)) -> (then)
   > Branch(Condition=@n2.IsDesignTime)
   > UpdateCopyButtonVisibility(Index=MouseAndKeyboard)
   Event Construct()
   > Bind Event to On Input Method Changed(self=Get CommonInputSubsystem(), Delegate=@n1)
   > UpdateCopyButtonVisibility(Index=GetCurrentInputType(self=Get CommonInputSubsystem()))
   Event Tick()
  ]}
 FormatAndSetText {nodes [
   n1 = FormatAndSetText!() -> (then, Stat:Text) {is=title K2Node_FunctionEntry}
   > Set Text(
    Text=Format Text!(Format=Get TextFormat() -> (TextFormat:Text), stat:Text=@n1) -> (Result)
    self=Get Text_Stat() -> (Text_Stat:Common Text Object Reference)
   )
   n2 = GetChildAt!(self:Panel Widget Object Reference=Get Slot_CopyButton(), Index:int=0) -> (ReturnValue:Widget Object Reference)
   > Is Valid(InputObject=@n2)
   > n3 = Cast To WBP_CopyButton!(execute, Object=@n2) -> (then, CastFailed, AsWBP Copy Button:WBP Copy Button Object Reference) {is=title K2Node_DynamicCast}
   > SetTextToCopy!(execute, self:WBP Copy Button Object Reference=@n3, StringToCopy:Text=Get Stat()) -> (then)
  ]}
 SetDisplayed {nodes [
   n1 = SetDisplayed!() -> (then, isDisplayed:bool=true) {is=title K2Node_FunctionEntry}
   n2 = Get Text_Stat()
   n3 = NOT Boolean(A=TextIsEmpty(InText=Get Text(self:Text Object Reference=@n2)))
   > SetVisibility(InVisibility=Select(Option 0=Collapsed, Option 1=SelfHitTestInvisible, Index=AND Boolean(A=@n1, B=@n3)))
  ]}
 UpdateCopyButtonVisibility {nodes [
   n1 = UpdateCopyButtonVisibility!() -> (then, Index:ECommonInputType Enum) {is=title K2Node_FunctionEntry}
   > Branch(Condition=HasAnyChildren!(self:Panel Widget Object Reference=Get Slot_CopyButton()) -> (ReturnValue:bool))
   > SetVisibility(
    self=GetChildAt(self=Get Slot_CopyButton(), Index=0)
    InVisibility=Select!(
     MouseAndKeyboard:ESlateVisibility Enum=SelfHitTestInvisible
     Gamepad:ESlateVisibility Enum=Collapsed
     Touch:ESlateVisibility Enum=SelfHitTestInvisible
     Count:ESlateVisibility Enum=Collapsed
     Index:ECommonInputType Enum=@n1
    ) -> (ReturnValue)
   )
   > SetVisibility(
    self=Get Text_Stat()
    InVisibility=Select(MouseAndKeyboard=Collapsed, Gamepad=SelfHitTestInvisible, Touch=Collapsed, Count=SelfHitTestInvisible, Index=@n1)
   )
   > SetPadding!(
    execute
    self:User Widget Object Reference
    InPadding_Left:float=0
    InPadding_Top:float=Select(MouseAndKeyboard:float=10, Gamepad:float=0, Touch:float=10, Count:float=0, Index=@n1) -> (ReturnValue:float)
    InPadding_Right:float=0
    InPadding_Bottom:float=12
   ) -> (then)
  ]}
]
variables [
 Icon {type=Object, default=null}
 IconSize {type=Vector2D, default=<36 36>}
 ShowCopyIcon {type=boolean}
 Stat {type=string, default=1234}
 TextFormat {type=string, default="Test Format Number {stat}"}
]
widgets {
 info {inheritedWidgetCount=0, parentClass=&UserWidget, rootWidgetClass=&HorizontalBox, widgetCount=4}
 widgets [
  HBox_Root {widgetClassPath=&HorizontalBox, bIsVariable=true, widget=HBox_Root}
  Image_Icon {
   widgetClassPath=&Image
   bIsVariable=true
   parent=HBox_Root
   widget=Image_Icon
   properties {brush {
     bIsDynamicallyLoaded=false
     drawAs=Image
     imageType=FullColor
     mirroring=NoMirror
     resourceName=None
     resourceObject=/FortUILibrary/Textures/MaterialInstances/MI_UIKit_Icon_LegoIsland
     tiling=NoTile
     imageSize=<32 32>
     margin=<ltrb 0 0 0 0>
     outlineSettings {
      bUseBrushTransparency=false
      roundingType=HalfHeightRadius
      width=0
      color {colorUseRule=UseColor_Specified, specifiedColor=<rgba 0 0 0 0>}
      cornerRadii=<0 0 0 0>
     }
     tintColor {colorUseRule=UseColor_Specified, specifiedColor=<rgba 1 1 1 1>}
     uVRegion {bIsValid=false, max=<0 0>, min=<0 0>}
    }}
   slot {class=HorizontalBoxSlot, properties {horizontalAlignment=HAlign_Left, verticalAlignment=VAlign_Center}}
  }
  Text_Stat {
   widgetClassPath=&CommonTextBlock
   bIsVariable=true
   parent=HBox_Root
   widget=Text_Stat
   properties {
    bIsScrollingEnabled=false
    mobileFontSizeMultiplier=1.3
    text="Released Dec 7, 2023"
    font {
     bForceMonospaced=false
     bMaterialIsStencil=false
     fontMaterial=null
     fontObject=&HeadingNow
     letterSpacing=0
     monospacedWidth=1
     size=14.25
     skewAmount=0
     typefaceFontName=74 Regular
     outlineSettings {
      bApplyOutlineToDropShadows=false
      bMiteredCorners=false
      bSeparateFillAlpha=false
      outlineMaterial=null
      outlineSize=0
      outlineColor=<rgba 0 0 0 1>
     }
    }
    shadowOffset=<0 0>
    strikeBrush {
     bIsDynamicallyLoaded=false
     drawAs=NoDrawType
     imageType=NoImage
     mirroring=NoMirror
     resourceName=None
     resourceObject=null
     tiling=NoTile
     imageSize=<32 32>
     margin=<ltrb 0 0 0 0>
     outlineSettings {
      bUseBrushTransparency=false
      roundingType=HalfHeightRadius
      width=0
      color {colorUseRule=UseColor_Specified, specifiedColor=<rgba 0 0 0 0>}
      cornerRadii=<0 0 0 0>
     }
     tintColor {colorUseRule=UseColor_Specified, specifiedColor=<rgba 1 1 1 1>}
     uVRegion {bIsValid=false, max=<0 0>, min=<0 0>}
    }
   }
   slot {class=HorizontalBoxSlot, properties {verticalAlignment=VAlign_Center, padding=<ltrb 8 0 8 0>, size {sizeRule=Fill, value=1}}}
  }
  Slot_CopyButton {
   widgetClassPath=&NamedSlot
   bIsVariable=true
   parent=HBox_Root
   widget=Slot_CopyButton
   slot {class=HorizontalBoxSlot, properties {padding=<ltrb 5 0 0 0>}}
  }
 ]
}
