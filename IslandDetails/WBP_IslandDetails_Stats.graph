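# Scrollable "About Island" panel on the island details screen, showing stats like likes, player count, dates, island code, features and IP branding.
# Entries start collapsed in two groups; SetStat functions fill and reveal them, some consulting FortIslandDetailsVM.
# A divider shows only when a lower-group entry is visible.
# Input type toggles the hover button and outline for gamepad focus, where hover fires OnFocused; touch scrolling follows scrollbar visibility.
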WidgetBlueprint /IslandDetails/WBP_IslandDetails_Stats
graphs [
 CF_IP_Text {nodes [
   n1 = CF_IP_Text!() -> (then, IP Enum:EFortActivityLinkCategory Enum=None, IP Name:Text) {is=title K2Node_FunctionEntry}
   n2 = Get TextFormat!(self:WBP Island Details Stats Entry Object Reference=Get Stats_Entry_IP() -> (Stats_Entry_IP:WBP Island Details Stats Entry Object Reference)) -> (TextFormat:Text)
   > Return Node!(
    execute
    Display Text:Text=Format Text!(
     Format=Select!(
      None:Text
      LEGO:Text="LOCTEXT(\"{stat}\")"
      RocketRacing:Text=@n2
      FallGuys:Text=@n2
      TMNT:Text=@n2
      TheWalkingDeadUniverse:Text=@n2
      SolidWave:Text=@n2
      Akita:Text=@n2
      KPOPDemonHunters:Text=@n2
      EpicGames:Text=@n2
      Creatures:Text
      Index:EFortActivityLinkCategory Enum=@n1
     ) -> (ReturnValue:Text)
     stat:Text=@n1.IP Name
    ) -> (Result)
   )
  ]}
 CF_StatLikePercentVisibility {nodes [
   n1 = CF_StatLikePercentVisibility!() -> (then, LikePercent:double=~, ShowLikeTag:bool=~) {is=title K2Node_FunctionEntry}
   n2 = GetRuntimeOptions()
   n3 = Round(A=@n1) {is=title K2Node_CallFunction}
   > Return Node!(
    execute
    IsDisplayed:bool=AND Boolean!(
     A="integer > integer"(A=@n3)
     B=@n1.ShowLikeTag
     C:bool=GetIsLikeDislikeIslandOptionEnabled!(self:Fort Runtime Options Object Reference=@n2) -> (ReturnValue:bool)
    ) -> (ReturnValue)
   )
  ]}
 Debug_AddStatsForTesting {nodes [
   Debug_AddStatsForTesting() {is=title K2Node_FunctionEntry}
   > For Loop(LastIndex=100)
   > n1 = Construct Object from Class(Class=&/WBP_IslandDetails_Stats_Entry.WBP_IslandDetails_Stats_Entry_C) -> (ReturnValue:WBP Island Details Stats Entry Object Reference)
   > AddChildToVerticalBox(self=Get VB_AboveLine() -> (VB_AboveLine:Vertical Box Object Reference), Content=@n1)
   > SetVisibility(self=@n1, InVisibility=SelfHitTestInvisible)
  ]}
 EventGraph {nodes [
   n1 = "On Scroll Bar Visibility Changed (ScrollBox)"!() -> (OutputDelegate, then, NewVisibility:ESlateVisibility Enum=Visible) {is=title K2Node_ComponentBoundEvent}
   > Set bEnableTouchScrolling!(
    execute
    bEnableTouchScrolling:bool="Equal (Enum)"(A:ESlateVisibility Enum=@n1.NewVisibility, B:ESlateVisibility Enum=Visible)
    self:Scroll Box Object Reference=Get ScrollBox() -> (ScrollBox:Scroll Box Object Reference)
   ) -> (then, Output_Get:bool)
   "On Pressed (Button_Stats_Hover)"() {is=title K2Node_ComponentBoundEvent}
   > PlaySound2D(Sound=Get Pressed Sound() -> (Pressed Sound:Sound Base Object Reference))
   "On Hovered (Button_Stats_Hover)"() {is=title K2Node_ComponentBoundEvent}
   n2 = Get WBP_Outline_Stats() -> (WBP_Outline_Stats:WBP UIKit Block Outline Object Reference)
   > RequestTransition!(execute, self:UIKit Block Object Reference=@n2, State:EUIKitBlockVisualState Enum=Focused) -> (then)
   > PlaySound2D(Sound=Get Hovered Sound() -> (Hovered Sound:Sound Base Object Reference))
   > Branch(Condition=IsUsingInputType!(
     InputType:ECommonInputType Enum=Gamepad
     PlayerController:Player Controller Object Reference=GetFirstLocalFortPlayerController!(bIncludeNonSpectators:bool=true, bIncludeSpectators:bool=false) -> (ReturnValue:Fort Player Controller Object Reference)
    ) -> (ReturnValue:bool))
   > Call On Focused!(execute:Exec, self:WBP Island Details Stats Object Reference) -> (then:Exec) {is=title K2Node_CallDelegate}
   "On Unhovered (Button_Stats_Hover)"() {is=title K2Node_ComponentBoundEvent}
   > RequestTransition(self=@n2, State=Unfocused)
   n3 = OnInputDeviceChanged!() -> (OutputDelegate, then, bNewInputType:ECommonInputType Enum) {is=title K2Node_CustomEvent}
   > UpdateInputDeviceConfig!(execute, self, Index:ECommonInputType Enum=@n3.bNewInputType) -> (then)
   Event OnInitialized()
   > n4 = For Each Loop(Array:Array of Widget Object References=GetAllChildren(self=Get VB_BelowLine() -> (VB_BelowLine:Vertical Box Object Reference))) -> (Array Element:Widget Object Reference)
   > n5 = Cast To WBP_IslandDetails_Stats_Entry!(execute, Object=@n4) -> (then, CastFailed, AsWBP Island Details Stats Entry:WBP Island Details Stats Entry Object Reference) {is=title K2Node_DynamicCast}
   > Bind Event to On Visibility Changed(self:User Widget Object Reference=@n5, Delegate=Create Event()) {is=title K2Node_AddDelegate}
   Event Construct()
   > InitializeVisibility()
   > Bind Event to On Input Method Changed(self=Get CommonInputSubsystem(), Delegate=@n3)
   > UpdateInputDeviceConfig(Index=GetCurrentInputType(self=Get CommonInputSubsystem()))
  ]}
 InitializeVisibility {nodes [
   InitializeVisibility() {is=title K2Node_FunctionEntry}
   > n1 = For Each Loop(Array=GetAllChildren(self=Get VB_AboveLine()))
   n2 = For Each Loop(Exec=@n1.Completed, Array=GetAllChildren(self=Get VB_BelowLine()))
   > SetVisibility(self=@n2, InVisibility=Collapsed)
   SetVisibility(execute=@n1.LoopBody, self=@n1, InVisibility=Collapsed)
  ]}
 IsDateTimeValid {nodes [
   n1 = IsDateTimeValid!() -> (then, DateTime:Date Time Structure) {is=title K2Node_FunctionEntry}
   > Return Node!(
    execute
    ReturnValue:bool="DateTime > DateTime"!(
     A:Date Time Structure=@n1
     B:Date Time Structure=MakeDateTime!(Year:int=1994, Month:int=1, Day:int=1, Hour:int=1, Minute:int=1, Second:int=1, Millisecond:int=1) -> (ReturnValue:Date Time Structure)
    ) -> (ReturnValue:bool)
   )
  ]}
 OnFocused {nodes [OnFocused() {is=title K2Node_FunctionEntry}]}
 SetHorizontalLineVisibility {nodes [
   SetHorizontalLineVisibility!() -> (then, InVisibility:ESlateVisibility Enum=Visible) {is=title K2Node_FunctionEntry}
   > Set FoundVisibleEntry(FoundVisibleEntry:bool=false) -> (Output_Get:bool)
   > n1 = For Each Loop with Break!(Exec:Exec, Array:Array of Widget Object References=GetAllChildren(self=Get VB_BelowLine()), Break:Exec=@n2.then) -> (LoopBody:Exec, Array Element:Widget Object Reference, Array Index:int, Completed:Exec)
   SetVisibility(
    execute=@n1.Completed
    self=Get Image_line() -> (Image_line:Image Object Reference)
    InVisibility=Select!(
     Option 0:ESlateVisibility Enum=Collapsed
     Option 1:ESlateVisibility Enum=HitTestInvisible
     Index:bool=Get FoundVisibleEntry() -> (FoundVisibleEntry:bool)
    ) -> (ReturnValue:ESlateVisibility Enum)
   )
   Branch(execute=@n1.LoopBody, Condition=IsVisible(self=@n1))
   > n2 = Set FoundVisibleEntry(FoundVisibleEntry=true)
  ]}
 SetStat_Build {nodes [
   n1 = SetStat_Build!() -> (then, IsBuildModeEnabled:bool=~) {is=title K2Node_FunctionEntry}
   > SetDisplayed!(
    execute
    self:WBP Island Details Stats Entry Object Reference=Get Stats_Entry_Build() -> (Stats_Entry_Build:WBP Island Details Stats Entry Object Reference)
    isDisplayed:bool=@n1
   ) -> (then)
  ]}
 SetStat_CCU {nodes [
   n1 = SetStat_CCU!() -> (then, CCU:int) {is=title K2Node_FunctionEntry}
   n2 = Get Stats_Entry_CCU() -> (Stats_Entry_CCU:WBP Island Details Stats Entry Object Reference)
   > FormatAndSetText!(
    execute
    self:WBP Island Details Stats Entry Object Reference=@n2
    Stat:Text=ConvertQuantityToShorthandText!(Quantity:int=@n1) -> (Text:Text)
   ) -> (then)
   > SetDisplayed(self=@n2, isDisplayed="integer > integer"(A=@n1))
  ]}
 SetStat_Category {nodes [
   n1 = SetStat_Category!() -> (then, FortGameActivity:Fort Game Activity Object Reference) {is=title K2Node_FunctionEntry}
   n2 = Get Stats_Entry_Category() -> (Stats_Entry_Category:WBP Island Details Stats Entry Object Reference)
   FormatAndSetText(self=@n2, Stat="To Text (String)"(InString=GetLinkCategoryString!(self:Fort Game Activity Object Reference=@n1) -> (ReturnValue:String)))
   n3 = GetLinkCategory!(self:Fort Game Activity Object Reference=@n1) -> (ReturnValue:EFortActivityLinkCategory Enum=None)
   n4 = Get Image_Icon!(self:WBP Island Details Stats Entry Object Reference=@n2) -> (Image_Icon:Image Object Reference)
   n5 = Get!(self:Image Object Reference=@n4) -> (
    Brush_TintColor:Slate Color Structure
    Brush_DrawAs:ESlateBrushDrawType Enum
    Brush_Tiling:ESlateBrushTileType Enum
    Brush_Mirroring:ESlateBrushMirrorType Enum
    Brush_ImageSize:Vector 2D Structure
    Brush_Margin:Margin Structure
    Brush_ResourceObject:Object Reference
    Brush_OutlineSettings:Slate Brush Outline Settings Structure
   )
   > SetBrush(
    self=@n4
    InBrush=Make Slate Brush!(
     TintColor:Slate Color Structure=@n5
     DrawAs:ESlateBrushDrawType Enum=@n5.Brush_DrawAs
     Tiling:ESlateBrushTileType Enum=@n5.Brush_Tiling
     Mirroring:ESlateBrushMirrorType Enum=@n5.Brush_Mirroring
     ImageSize:Vector 2D Structure=@n5.Brush_ImageSize
     Margin:Margin Structure=@n5.Brush_Margin
     ResourceObject:Object Reference=Select!(
      None:Object Reference
      LEGO:Object Reference=/FortUILibrary/Textures/MaterialInstances/MI_UIKit_Icon_LegoIsland
      RocketRacing:Object Reference=/FortUILibrary/Textures/MaterialInstances/MI_UIKit_Icon_RocketRacingIsland
      FallGuys:Object Reference=/FortUILibrary/Textures/MaterialInstances/MI_UIKit_Icon_Brand_FallGuys
      TMNT:Object Reference=/FortUILibrary/Textures/MaterialInstances/MI_UIKit_Icon_Brand_TMNT
      TheWalkingDeadUniverse:Object Reference=/FortUILibrary/Textures/MaterialInstances/MI_UIKit_Icon_Brand_TWDU
      SolidWave:Object Reference=/FortUILibrary/Textures/MaterialInstances/MI_UIKit_Icon_Brand_SquidGame
      Akita:Object Reference=/FortUILibrary/Textures/MaterialInstances/MI_UIKit_Icon_Brand_Akita
      KPOPDemonHunters:Object Reference
      EpicGames:Object Reference
      Creatures:Object Reference
      Index:EFortActivityLinkCategory Enum=@n3
     ) -> (ReturnValue:Object Reference)
     OutlineSettings:Slate Brush Outline Settings Structure=@n5.Brush_OutlineSettings
    ) -> (SlateBrush:Slate Brush Structure) {is=title K2Node_MakeStruct}
   )
   > SetVisibility(self=@n2, InVisibility=HitTestInvisible)
  ]}
 SetStat_IP_Icon {nodes [
   n1 = SetStat_IP_Icon!() -> (then, IP_Enum:EFortActivityLinkCategory Enum=None) {is=title K2Node_FunctionEntry}
   > n2 = Branch(Condition="Not Equal (Enum)"!(A:EFortActivityLinkCategory Enum=@n1, B:EFortActivityLinkCategory Enum=None) -> (ReturnValue:bool))
   n3 = Get Map_IPIcons() -> (Map_IPIcons:Map of EFortActivityLinkCategory Enums to Object References=None)
   n4 = Get IP_Enum() -> (IP_Enum:EFortActivityLinkCategory Enum=None)
   n5 = Get Image_Icon(self=Get Stats_Entry_IP())
   n6 = Get(self=@n5)
   > SetBrush(
    self=@n5
    InBrush=Make Slate Brush(
     TintColor=@n6
     DrawAs=@n6.Brush_DrawAs
     Tiling=@n6.Brush_Tiling
     Mirroring=@n6.Brush_Mirroring
     ImageSize=@n6.Brush_ImageSize
     Margin=@n6.Brush_Margin
     ResourceObject=Find(TargetMap:Map of EFortActivityLinkCategory Enums to Object References=@n3, Key:"EFortActivityLinkCategory Enum (by ref)"=@n4) -> (Value:Object Reference)
     OutlineSettings=@n6.Brush_OutlineSettings
    )
   )
   > SetDisplayed(self=Get Stats_Entry_IP(), isDisplayed=true)
   SetDisplayed(execute=@n2.else, self=Get Stats_Entry_IP(), isDisplayed=false)
  ]}
 SetStat_IP_Text {nodes [
   n1 = SetStat_IP_Text!() -> (then, Display Name:Text) {is=title K2Node_FunctionEntry}
   > Set Text(
    Text=@n1
    self=Get Text_Stat!(self:WBP Island Details Stats Entry Object Reference=Get Stats_Entry_IP()) -> (Text_Stat:Common Text Object Reference)
   )
   > SetDisplayed(self=Get Stats_Entry_IP(), isDisplayed=NOT Boolean(A=TextIsEmpty(InText=@n1)))
  ]}
 SetStat_IsEpicCreated {nodes [
   n1 = SetStat_IsEpicCreated!() -> (then, IsEpicCreated:bool=~) {is=title K2Node_FunctionEntry}
   > SetDisplayed(self=Get Stats_Entry_IsEpicCreated() -> (Stats_Entry_IsEpicCreated:WBP Island Details Stats Entry Object Reference), isDisplayed=@n1)
  ]}
 SetStat_IslandCode {nodes [
   n1 = SetStat_IslandCode!() -> (then, IslandCode:String) {is=title K2Node_FunctionEntry}
   n2 = Get Stats_Entry_IslandCode() -> (Stats_Entry_IslandCode:WBP Island Details Stats Entry Object Reference)
   > FormatAndSetText(self=@n2, Stat="To Text (String)"(InString=@n1))
   n3 = Get FortIslandDetailsVM() -> (FortIslandDetailsVM:Fort Island Details VM Object Reference)
   > SetDisplayed(self=@n2, isDisplayed=NOT Boolean(A=IsEpicCreated!(self:Fort Discover Link Entry Tile Item VM Object Reference=@n3) -> (ReturnValue:bool)))
  ]}
 SetStat_LikePercent {nodes [
   n1 = SetStat_LikePercent!() -> (then, LikePercent:double=~) {is=title K2Node_FunctionEntry}
   > FormatAndSetText(
    self=Get Stats_Entry_LikePercent() -> (Stats_Entry_LikePercent:WBP Island Details Stats Entry Object Reference)
    Stat="To Text (Integer)"(Value=Round(A=@n1))
   )
  ]}
 SetStat_MaxPlayers {nodes [
   n1 = SetStat_MaxPlayers!() -> (then, MaxPlayers:int) {is=title K2Node_FunctionEntry}
   n2 = Get Stats_Entry_MaxPlayers() -> (Stats_Entry_MaxPlayers:WBP Island Details Stats Entry Object Reference)
   > FormatAndSetText(self=@n2, Stat="To Text (Integer)"(Value=@n1))
   > SetDisplayed(self=@n2, isDisplayed="integer > integer"(A=@n1))
  ]}
 SetStat_ProximityChat {nodes [
   n1 = SetStat_ProximityChat!() -> (then, HasProximityChat:bool=~) {is=title K2Node_FunctionEntry}
   > SetDisplayed(self=Get Stats_Entry_ProximityChat() -> (Stats_Entry_ProximityChat:WBP Island Details Stats Entry Object Reference), isDisplayed=@n1)
  ]}
 SetStat_Purchases {nodes [
   n1 = SetStat_Purchases!() -> (then, HasPurchases:bool=~) {is=title K2Node_FunctionEntry}
   > SetDisplayed(self=Get Stats_Entry_Purchases() -> (Stats_Entry_Purchases:WBP Island Details Stats Entry Object Reference), isDisplayed=@n1)
  ]}
 SetStat_RankedAvailable {nodes [
   n1 = SetStat_RankedAvailable!() -> (then, isRankedAvailable:bool=~) {is=title K2Node_FunctionEntry}
   > SetDisplayed(self=Get Stats_Entry_RankedAvailable() -> (Stats_Entry_RankedAvailable:WBP Island Details Stats Entry Object Reference), isDisplayed=@n1)
  ]}
 SetStat_Released {nodes [
   n1 = SetStat_Released!() -> (then, DateTime:Date Time Structure) {is=title K2Node_FunctionEntry}
   n2 = Now!() -> (ReturnValue:Date Time Structure)
   n3 = Get Stats_Entry_Released() -> (Stats_Entry_Released:WBP Island Details Stats Entry Object Reference)
   > Set TextFormat!(
    execute
    TextFormat:Text=SelectText(A="LOCTEXT(\"Releases {stat}\")", B="LOCTEXT(\"Released {stat}\")", bPickA="DateTime > DateTime"(A=@n1, B=@n2))
    self:WBP Island Details Stats Entry Object Reference=@n3
   ) -> (then, Output_Get:Text)
   > FormatAndSetText(self=@n3, Stat=AsMediumDate!(InDateTime:"Date Time Structure (by ref)"=@n1) -> (ReturnValue:Text))
   > SetDisplayed(self=@n3, isDisplayed=IsDateTimeValid!(DateTime:Date Time Structure=@n1) -> (ReturnValue:bool) {is=title K2Node_CallFunction})
  ]}
 SetStat_Saves {nodes [
   n1 = SetStat_Saves!() -> (then, HasSaves:bool=~) {is=title K2Node_FunctionEntry}
   > SetDisplayed(self=Get Stats_Entry_Saves() -> (Stats_Entry_Saves:WBP Island Details Stats Entry Object Reference), isDisplayed=@n1)
  ]}
 SetStat_Updated {nodes [
   n1 = SetStat_Updated!() -> (then, DateTime:Date Time Structure) {is=title K2Node_FunctionEntry}
   n2 = Get Stats_Entry_Updated() -> (Stats_Entry_Updated:WBP Island Details Stats Entry Object Reference)
   > FormatAndSetText(self=@n2, Stat=AsMediumDate(InDateTime=@n1))
   n3 = GetReleaseDate!(self:Fort Island Details VM Object Reference=Get FortIslandDetailsVM()) -> (ReturnValue:Date Time Structure)
   > SetDisplayed(self=@n2, isDisplayed=AND Boolean!(A=IsDateTimeValid(DateTime=@n1), B="DateTime > DateTime"(A=@n1, B=@n3)) -> (ReturnValue))
  ]}
 UpdateInputDeviceConfig {nodes [
   n1 = UpdateInputDeviceConfig!() -> (then, Index:ECommonInputType Enum) {is=title K2Node_FunctionEntry}
   > SetVisibility(
    self=Get Button_Stats_Hover() -> (Button_Stats_Hover:Button Object Reference)
    InVisibility=Select!(
     MouseAndKeyboard:ESlateVisibility Enum=SelfHitTestInvisible
     Gamepad:ESlateVisibility Enum=Visible
     Touch:ESlateVisibility Enum=SelfHitTestInvisible
     Count:ESlateVisibility Enum=SelfHitTestInvisible
     Index:ECommonInputType Enum=@n1
    ) -> (ReturnValue:ESlateVisibility Enum)
   )
   > SetBlockVisibility!(
    execute
    self:WBP UIKit Block Base Object Reference=Get WBP_Outline_Stats()
    Visible:bool=Select(MouseAndKeyboard:bool=false, Gamepad:bool=true, Touch:bool=false, Count:bool=false, Index=@n1) -> (ReturnValue:bool)
   ) -> (then)
  ]}
]
variables [
 Hovered Sound {type=SoundBase, default=/Game/Sounds/Fort_UI_Sounds/FrontEnd/DefaultSounds/PressedButtons/MSS_UI_Button_Hover_Default}
 Map_IPIcons {default {
   akita=&MI_UIKit_Icon_Brand_Akita
   epicGames=/FortniteCoreUI/Frontend/StoreMain/Materials/Textures/T_Icon_EpicStoreWhite
   fallGuys=&MI_UIKit_Icon_Brand_FallGuys
   kPOPDemonHunters=/FortUILibrary/Textures/MaterialInstances/MI_UIKit_Icon_Refrigerator
   lEGO=&MI_UIKit_Icon_LegoIsland
   none=null
   rocketRacing=&MI_UIKit_Icon_RocketRacingIsland
   solidWave=&MI_UIKit_Icon_Brand_SquidGame
   tMNT=&MI_UIKit_Icon_Brand_TMNT
   theWalkingDeadUniverse=&MI_UIKit_Icon_Brand_TWDU
  }}
 OnFocused {}
 Pressed Sound {type=SoundBase, default=/Game/Sounds/Fort_UI_Sounds/FrontEnd/DefaultSounds/PressedButtons/MSS_UI_Button_Pressed_Default}
]
widgets {
 info {inheritedWidgetCount=0, parentClass=&UserWidget, rootWidgetClass=&Button, widgetCount=30}
 widgets [
  Button_Stats_Hover {
   widgetClassPath=&Button
   widget=Button_Stats_Hover
   properties {widgetStyle {
     clickedSlateSound {resourceObject=null}
     disabled {
      bIsDynamicallyLoaded=false
      drawAs=NoDrawType
      imageType=NoImage
      mirroring=NoMirror
      resourceName=None
      resourceObject=null
      tiling=NoTile
      imageSize=<0 0>
      margin=<ltrb 0 0 0 0>
      outlineSettings {
       bUseBrushTransparency=false
       roundingType=HalfHeightRadius
       width=0
       color {colorUseRule=UseColor_Specified, specifiedColor=<rgba 0 0 0 0>}
       cornerRadii=<0 0 0 0>
      }
      tintColor {colorUseRule=UseColor_Specified, specifiedColor=<rgba 1 1 1 1>}
      uVRegion {bIsValid=false, max=<0 0>, min=<0 0>}
     }
     disabledForeground {colorUseRule=UseColor_Specified, specifiedColor=<rgba 0.5271 0.5271 0.5271 1>}
     hovered {
      bIsDynamicallyLoaded=false
      drawAs=NoDrawType
      imageType=NoImage
      mirroring=NoMirror
      resourceName=None
      resourceObject=null
      tiling=NoTile
      imageSize=<32 32>
      margin=<ltrb 0 0 0 0>
      outlineSettings {
       bUseBrushTransparency=true
       roundingType=FixedRadius
       width=1
       color {colorUseRule=UseColor_Specified, specifiedColor=<rgba 0.7243 0.7243 0.7243 1>}
       cornerRadii=<4 4 4 4>
      }
      tintColor {colorUseRule=UseColor_Specified, specifiedColor=<rgba 0.7243 0.7243 0.7243 1>}
      uVRegion {bIsValid=false, max=<0 0>, min=<0 0>}
     }
     hoveredForeground {colorUseRule=UseColor_Specified, specifiedColor=<rgba 0 1 1 1>}
     hoveredSlateSound {resourceObject=null}
     normal {
      bIsDynamicallyLoaded=false
      drawAs=NoDrawType
      imageType=NoImage
      mirroring=NoMirror
      resourceName=None
      resourceObject=null
      tiling=NoTile
      imageSize=<32 32>
      margin=<ltrb 0 0 0 0>
      outlineSettings {
       bUseBrushTransparency=true
       roundingType=FixedRadius
       width=1
       color {colorUseRule=UseColor_Specified, specifiedColor=<rgba 0.6951 0.6951 0.6951 1>}
       cornerRadii=<4 4 4 4>
      }
      tintColor {colorUseRule=UseColor_Specified, specifiedColor=<rgba 0.4955 0.4955 0.4955 1>}
      uVRegion {bIsValid=false, max=<0 0>, min=<0 0>}
     }
     normalForeground {colorUseRule=UseColor_Specified, specifiedColor=<rgba 0 1 1 1>}
     normalPadding=<ltrb 0 0 0 0>
     pressed {
      bIsDynamicallyLoaded=false
      drawAs=NoDrawType
      imageType=NoImage
      mirroring=NoMirror
      resourceName=None
      resourceObject=null
      tiling=NoTile
      imageSize=<32 32>
      margin=<ltrb 0 0 0 0>
      outlineSettings {
       bUseBrushTransparency=true
       roundingType=FixedRadius
       width=1
       color {colorUseRule=UseColor_Specified, specifiedColor=<rgba 0.7243 0.7243 0.7243 1>}
       cornerRadii=<4 4 4 4>
      }
      tintColor {colorUseRule=UseColor_Specified, specifiedColor=<rgba 0.3843 0.3843 0.3843 1>}
      uVRegion {bIsValid=false, max=<0 0>, min=<0 0>}
     }
     pressedForeground {colorUseRule=UseColor_Specified, specifiedColor=<rgba 0 1 1 1>}
     pressedPadding=<ltrb 0 0 0 0>
     pressedSlateSound {resourceObject=null}
    }}
  }
  Overlay_3 {
   widgetClassPath=&Overlay
   parent=Button_Stats_Hover
   widget=Overlay_3
   slot {class=ButtonSlot, properties {horizontalAlignment=HAlign_Fill, verticalAlignment=VAlign_Fill, padding=<ltrb 0 0 0 0>}}
  }
  WBP_CaptureForPostBufferUpdate {
   widgetClassPath=&WBP_CaptureForPostBufferUpdate
   bIsVariable=true
   parent=Overlay_3
   slot {class=OverlaySlot, properties {horizontalAlignment=HAlign_Fill, verticalAlignment=VAlign_Fill}}
  }
  Spacer_1 {
   widgetClassPath=&Spacer
   parent=Overlay_3
   widget=Spacer_1
   properties {visibility=Collapsed, size=<1200 1>}
   slot {class=OverlaySlot}
  }
  Spacer_2 {widgetClassPath=&Spacer, parent=Overlay_3, widget=Spacer_2, properties {size=<1 100>}, slot {class=OverlaySlot}}
  WBP_UIKit_Backplate_Stats {
   widgetClassPath=&WBP_UIKit_Backplate
   bIsVariable=true
   parent=Overlay_3
   widget=WBP_UIKit_Backplate_Stats
   properties {brightness=Medium, colorAndOpacity=<rgba 1 1 1 0.5>}
   slot {class=OverlaySlot, properties {horizontalAlignment=HAlign_Fill, verticalAlignment=VAlign_Fill}}
  }
  Overlay_Content {
   widgetClassPath=&Overlay
   bIsVariable=true
   parent=Overlay_3
   widget=Overlay_Content
   properties {bOverride_Cursor=true}
   slot {class=OverlaySlot, properties {horizontalAlignment=HAlign_Fill, verticalAlignment=VAlign_Fill, padding=<ltrb 32 32 32 32>}}
  }
  VBox_Stats {
   widgetClassPath=&VerticalBox
   parent=Overlay_Content
   widget=VBox_Stats
   slot {class=OverlaySlot, properties {horizontalAlignment=HAlign_Fill, verticalAlignment=VAlign_Fill}}
  }
  Text_AboutIsland {
   widgetClassPath=&CommonTextBlock
   parent=VBox_Stats
   widget=Text_AboutIsland
   properties {
    mobileFontSizeMultiplier=1.2
    text=About Island
    font {
     bForceMonospaced=false
     bMaterialIsStencil=false
     fontMaterial=null
     fontObject=&HeadingNow
     letterSpacing=0
     monospacedWidth=1
     size=15.75
     skewAmount=0
     typefaceFontName=75 Medium
     outlineSettings {
      bApplyOutlineToDropShadows=false
      bMiteredCorners=false
      bSeparateFillAlpha=false
      outlineMaterial=null
      outlineSize=0
      outlineColor=<rgba 0 0 0 1>
     }
    }
    shadowOffset=<0 0>
   }
   slot {class=VerticalBoxSlot, properties {horizontalAlignment=HAlign_Center, verticalAlignment=VAlign_Top, padding=<ltrb 0 0 0 24>}}
  }
  SizeBox_2 {
   widgetClassPath=&SizeBox
   parent=VBox_Stats
   widget=SizeBox_2
   properties {bOverride_HeightOverride=true}
   slot {class=VerticalBoxSlot, properties {size {sizeRule=Fill, value=1}}}
  }
  ScrollBox {
   widgetClassPath=&ScrollBox
   bIsVariable=true
   parent=SizeBox_2
   widget=ScrollBox
   properties {
    scrollWhenFocusChanges=AnimatedScroll
    widgetStyle {
     barThickness=9
     bottomShadowBrush {
      bIsDynamicallyLoaded=false
      drawAs=NoDrawType
      imageType=FullColor
      mirroring=NoMirror
      resourceName=../../../Engine/Content/Slate/Common/ScrollBoxShadowBottom.png
      resourceObject=null
      tiling=NoTile
      imageSize=<16 8>
      margin=<ltrb 0.5 0 0.5 1>
      outlineSettings {
       bUseBrushTransparency=false
       roundingType=HalfHeightRadius
       width=0
       color {colorUseRule=UseColor_Specified, specifiedColor=<rgba 0 0 0 0>}
       cornerRadii=<0 0 0 0>
      }
      tintColor {colorUseRule=UseColor_Specified, specifiedColor=<rgba 1 1 1 1>}
      uVRegion {bIsValid=false, max=<0 0>, min=<0 0>}
     }
     horizontalScrolledContentPadding=<ltrb 0 0 1 0>
     leftShadowBrush {
      bIsDynamicallyLoaded=false
      drawAs=Box
      imageType=FullColor
      mirroring=NoMirror
      resourceName=../../../Engine/Content/Slate/Common/ScrollBoxShadowLeft.png
      resourceObject=null
      tiling=NoTile
      imageSize=<8 16>
      margin=<ltrb 1 0.5 0 0.5>
      outlineSettings {
       bUseBrushTransparency=false
       roundingType=HalfHeightRadius
       width=0
       color {colorUseRule=UseColor_Specified, specifiedColor=<rgba 0 0 0 0>}
       cornerRadii=<0 0 0 0>
      }
      tintColor {colorUseRule=UseColor_Specified, specifiedColor=<rgba 1 1 1 1>}
      uVRegion {bIsValid=false, max=<0 0>, min=<0 0>}
     }
     rightShadowBrush {
      bIsDynamicallyLoaded=false
      drawAs=Box
      imageType=FullColor
      mirroring=NoMirror
      resourceName=../../../Engine/Content/Slate/Common/ScrollBoxShadowRight.png
      resourceObject=null
      tiling=NoTile
      imageSize=<8 16>
      margin=<ltrb 0 0.5 1 0.5>
      outlineSettings {
       bUseBrushTransparency=false
       roundingType=HalfHeightRadius
       width=0
       color {colorUseRule=UseColor_Specified, specifiedColor=<rgba 0 0 0 0>}
       cornerRadii=<0 0 0 0>
      }
      tintColor {colorUseRule=UseColor_Specified, specifiedColor=<rgba 1 1 1 1>}
      uVRegion {bIsValid=false, max=<0 0>, min=<0 0>}
     }
     topShadowBrush {
      bIsDynamicallyLoaded=false
      drawAs=NoDrawType
      imageType=FullColor
      mirroring=NoMirror
      resourceName=../../../Engine/Content/Slate/Common/ScrollBoxShadowTop.png
      resourceObject=null
      tiling=NoTile
      imageSize=<16 8>
      margin=<ltrb 0.5 1 0.5 0>
      outlineSettings {
       bUseBrushTransparency=false
       roundingType=HalfHeightRadius
       width=0
       color {colorUseRule=UseColor_Specified, specifiedColor=<rgba 0 0 0 0>}
       cornerRadii=<0 0 0 0>
      }
      tintColor {colorUseRule=UseColor_Specified, specifiedColor=<rgba 1 1 1 1>}
      uVRegion {bIsValid=false, max=<0 0>, min=<0 0>}
     }
     verticalScrolledContentPadding=<ltrb 0 0 0 1>
    }
   }
   slot {class=SizeBoxSlot}
  }
  VerticalBox {widgetClassPath=&VerticalBox, parent=ScrollBox, widget=VerticalBox, slot {class=ScrollBoxSlot}}
  VB_AboveLine {widgetClassPath=&VerticalBox, bIsVariable=true, parent=VerticalBox, widget=VB_AboveLine, slot {class=VerticalBoxSlot}}
  Stats_Entry_LikePercent {
   widgetClassPath=&/WBP_IslandDetails_Stats_Entry
   bIsVariable=true
   parent=VB_AboveLine
   widget=Stats_Entry_LikePercent
   properties {icon=/FortUILibrary/Textures/MaterialInstances/MI_UIKit_Icon_Like, textFormat="{stat}% liked", padding=<ltrb 0 0 0 12>}
   slot {class=VerticalBoxSlot}
  }
  Stats_Entry_CCU {
   widgetClassPath=&WBP_IslandDetails_Stats_Entry
   bIsVariable=true
   parent=VB_AboveLine
   widget=Stats_Entry_CCU
   properties {icon=/FortUILibrary/Textures/MaterialInstances/MI_UIKit_Icon_PlayerController, textFormat="{stat} Playing Now", padding=<ltrb 0 0 0 12>}
   slot {class=VerticalBoxSlot}
  }
  Stats_Entry_Updated {
   widgetClassPath=&WBP_IslandDetails_Stats_Entry
   bIsVariable=true
   parent=VB_AboveLine
   widget=Stats_Entry_Updated
   properties {icon=/FortUILibrary/Textures/MaterialInstances/MI_UIKit_Icon_Update, textFormat="Updated {stat}", padding=<ltrb 0 0 0 12>}
   slot {class=VerticalBoxSlot}
  }
  Stats_Entry_Released {
   widgetClassPath=&WBP_IslandDetails_Stats_Entry
   bIsVariable=true
   parent=VB_AboveLine
   widget=Stats_Entry_Released
   properties {icon=/FortUILibrary/Textures/MaterialInstances/MI_UIKit_Icon_UpdatePending, textFormat="Released {stat}", padding=<ltrb 0 0 0 12>}
   slot {class=VerticalBoxSlot}
  }
  Stats_Entry_IslandCode {
   widgetClassPath=&WBP_IslandDetails_Stats_Entry
   bIsVariable=true
   parent=VB_AboveLine
   widget=Stats_Entry_IslandCode
   properties {icon=/FortUILibrary/Textures/MaterialInstances/MI_UIKit_Icon_Island, showCopyIcon=true, textFormat="{stat}", padding=<ltrb 0 0 0 12>}
   slot {class=VerticalBoxSlot, properties {horizontalAlignment=HAlign_Left}}
  }
  Image_line {
   widgetClassPath=&Image
   bIsVariable=true
   parent=VerticalBox
   widget=Image_line
   properties {
    brush {
     bIsDynamicallyLoaded=false
     drawAs=Image
     imageType=NoImage
     mirroring=NoMirror
     resourceName=None
     resourceObject=null
     tiling=NoTile
     imageSize=<0 2>
     margin=<ltrb 0 0 0 0>
     outlineSettings {
      bUseBrushTransparency=false
      roundingType=HalfHeightRadius
      width=0
      color {colorUseRule=UseColor_Specified, specifiedColor=<rgba 0 0 0 0>}
      cornerRadii=<0 0 0 0>
     }
     tintColor {colorUseRule=UseColor_Specified, specifiedColor=<rgba 1 1 1 1>}
     uVRegion {bIsValid=false, max=<0 0>, min=<0 0>}
    }
    colorAndOpacity=<rgba 1 1 1 0.25>
   }
   slot {class=VerticalBoxSlot, properties {verticalAlignment=VAlign_Center, padding=<ltrb 0 6 0 12>}}
  }
  VB_BelowLine {widgetClassPath=&VerticalBox, bIsVariable=true, parent=VerticalBox, widget=VB_BelowLine, slot {class=VerticalBoxSlot}}
  Stats_Entry_RankedAvailable {
   widgetClassPath=&WBP_IslandDetails_Stats_Entry
   bIsVariable=true
   parent=VB_BelowLine
   widget=Stats_Entry_RankedAvailable
   properties {icon=/FortUILibrary/Textures/MaterialInstances/MI_UIKit_Icon_RankedCup, textFormat=Ranked Available, padding=<ltrb 0 0 0 12>}
   slot {class=VerticalBoxSlot}
  }
  Stats_Entry_Build {
   widgetClassPath=&WBP_IslandDetails_Stats_Entry
   bIsVariable=true
   parent=VB_BelowLine
   widget=Stats_Entry_Build
   properties {icon=/FortUILibrary/Textures/MaterialInstances/MI_UIKit_Icon_BuildFighting, textFormat="Build, Zero Build", padding=<ltrb 0 0 0 12>}
   slot {class=VerticalBoxSlot}
  }
  Stats_Entry_IsEpicCreated {
   widgetClassPath=&WBP_IslandDetails_Stats_Entry
   bIsVariable=true
   parent=VB_BelowLine
   widget=Stats_Entry_IsEpicCreated
   properties {icon=&T_Icon_EpicStoreWhite, textFormat=By Epic, padding=<ltrb 0 0 0 12>}
   slot {class=VerticalBoxSlot}
  }
  Stats_Entry_Purchases {
   widgetClassPath=&WBP_IslandDetails_Stats_Entry
   bIsVariable=true
   parent=VB_BelowLine
   widget=Stats_Entry_Purchases
   properties {icon=/FortUILibrary/Textures/MaterialInstances/MI_UIKit_Icon_InGamePurchase, textFormat=In-Island Transactions, padding=<ltrb 0 0 0 12>}
   slot {class=VerticalBoxSlot}
  }
  Stats_Entry_ProximityChat {
   widgetClassPath=&WBP_IslandDetails_Stats_Entry
   bIsVariable=true
   parent=VB_BelowLine
   widget=Stats_Entry_ProximityChat
   properties {icon=/FortUILibrary/Textures/MaterialInstances/MI_UIKit_Icon_AudioSettings, textFormat=Proximity chat, padding=<ltrb 0 0 0 12>}
   slot {class=VerticalBoxSlot}
  }
  Stats_Entry_Saves {
   widgetClassPath=&WBP_IslandDetails_Stats_Entry
   bIsVariable=true
   parent=VB_BelowLine
   widget=Stats_Entry_Saves
   properties {
    icon=/FortUILibrary/Textures/MaterialInstances/MI_UIKit_Icon_Save
    textFormat=Saves progress
    foregroundColor {colorUseRule=UseColor_Specified, specifiedColor=<rgba 1 0 1 1>}
    padding=<ltrb 0 0 0 12>
   }
   slot {class=VerticalBoxSlot}
  }
  Stats_Entry_IP {
   widgetClassPath=&WBP_IslandDetails_Stats_Entry
   bIsVariable=true
   parent=VB_BelowLine
   widget=Stats_Entry_IP
   properties {textFormat="{stat} Island", padding=<ltrb 0 0 0 12>}
   slot {class=VerticalBoxSlot}
  }
  Stats_Entry_Category {
   widgetClassPath=&WBP_IslandDetails_Stats_Entry
   bIsVariable=true
   parent=VB_BelowLine
   widget=Stats_Entry_Category
   properties {textFormat="{stat}", padding=<ltrb 0 0 0 12>}
   slot {class=VerticalBoxSlot}
  }
  Stats_Entry_MaxPlayers {
   widgetClassPath=&WBP_IslandDetails_Stats_Entry
   bIsVariable=true
   parent=VB_BelowLine
   widget=Stats_Entry_MaxPlayers
   properties {icon=/FortUILibrary/Textures/MaterialInstances/MI_UIKit_Icon_Squads, textFormat="Up to {stat} Players", padding=<ltrb 0 0 0 12>}
   slot {class=VerticalBoxSlot}
  }
  WBP_Outline_Stats {
   widgetClassPath=/FortUILibrary/UIKit/Buttons/Blocks/WBP_UIKit_Block_Outline
   bIsVariable=true
   parent=Overlay_3
   widget=WBP_Outline_Stats
   properties {
    bIsFocusable=true
    bottomLeftCornerRadius=20
    bottomRightCornerRadius=20
    navigation=&self:WidgetTree.WBP_Outline_Stats.WidgetNavigation_0
    priority=100
    radiusInPixels=true
    topLeftCornerRadius=20
    topRightCornerRadius=20
    transitionTiming {
     deselectingFocused=0
     deselectingIdle=0
     disabling=0
     enabling=0
     focusing=1
     hovering=0
     pressing=0
     releasing=0
     selecting=0
     unfocusing=0
     unhovering=0
    }
   }
   slot {class=OverlaySlot, properties {horizontalAlignment=HAlign_Fill, verticalAlignment=VAlign_Fill}}
  }
 ]
}
