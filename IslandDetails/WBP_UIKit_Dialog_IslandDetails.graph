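# Island details dialog, a CommonActivatableWidget wrapping the UIKit internal dialog, with Private and Team Fill toggles.
# SetToggleStateAndVisibility applies passed settings to each toggle's selection and row visibility.
# Button 0 broadcasts OnConfirm_Event with both toggle values, then deactivates; other buttons, focus loss and the touch-only close button just deactivate.
# PreConstruct recolors the switchers' selected states through their dynamic materials, and the Private toggle is the default focus target.
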WidgetBlueprint /IslandDetails/WBP_UIKit_Dialog_IslandDetails
graphs [
 BP_GetDesiredFocusTarget {nodes [
   Get Desired Focus Target()
   > Return Node!(execute, ReturnValue:Widget Object Reference=Get ToggleButton_Private() -> (ToggleButton_Private:WBP UIKit Switcher Button Object Reference))
  ]}
 EventGraph {nodes [
   n1 = "On Button Clicked (Button_BackTouch)"!() -> (OutputDelegate, then, Button:Common Button Legacy Object Reference) {is=title K2Node_ComponentBoundEvent}
   DeactivateWidget(execute=[@n4.then, @n1.then, @n8.then, @n3.1, @n3.Default])
   n2 = OnButtonClicked!() -> (OutputDelegate, then, ButtonIndex:int) {is=title K2Node_CustomEvent}
   > n3 = Switch on Int!(execute:Exec, Selection:int=@n2.ButtonIndex) -> (Default:Exec, 0:Exec, 1:Exec)
   n4 = Call On Confirm Event!(
    execute:Exec=@n3.0
    self:WBP UIKit Dialog Island Details Object Reference
    PrivacyAndFill_bPrivacyValue:bool=GetSelected!(self:Common Button Base Object Reference=Get ToggleButton_Private()) -> (ReturnValue:bool)
    PrivacyAndFill_bPrivacyAvailable:bool=False
    PrivacyAndFill_bFillValue:bool=GetSelected(self=Get ToggleButton_TeamFill() -> (ToggleButton_TeamFill:WBP UIKit Switcher Button Object Reference))
    PrivacyAndFill_bFillAvailable:bool=False
   ) -> (then:Exec) {is=title K2Node_CallDelegate}
   n5 = Event PreConstruct()
   > Parent: PreConstruct(IsDesignTime=@n5.IsDesignTime)
   > SetSwitcherButtonMIDParameter!(execute, self, Target:WBP UIKit Switcher Button Object Reference=Get ToggleButton_Private()) -> (then)
   > SetSwitcherButtonMIDParameter(Target=Get ToggleButton_TeamFill())
   Event Construct()
   > Parent: Construct()
   n6 = Create Event()
   n7 = Get Dialog_Internal() -> (Dialog_Internal:WBP UIKit Dialog Internal Object Reference)
   > Bind Event to Dialog Button Clicked Dispatcher(self:WBP UIKit Dialog Internal Object Reference=@n7, Delegate=@n6) {is=title K2Node_AddDelegate}
   n8 = Event OnFocusLost!() -> (OutputDelegate, then, InFocusEvent:Focus Event Structure) {is=title K2Node_Event}
   Event Destruct()
   > Parent: Destruct()
   > Unbind Event from Dialog Button Clicked Dispatcher(self:WBP UIKit Dialog Internal Object Reference=@n7, Delegate=@n6) {is=title K2Node_RemoveDelegate}
  ]}
 OnConfirm_Event {nodes [OnConfirm_Event!() -> (then, PrivacyAndFill:Privacy And Fill Toggle Settings Structure) {is=title K2Node_FunctionEntry}]}
 SetSwitcherButtonMIDParameter {nodes [
   n1 = SetSwitcherButtonMIDParameter!() -> (then, Target:WBP UIKit Switcher Button Object Reference) {is=title K2Node_FunctionEntry}
   n2 = Get ImageMID!(self:WBP UIKit Block Switcher Object Reference=Get Block_Switcher!(self:WBP UIKit Switcher Button Object Reference=@n1) -> (Block_Switcher:WBP UIKit Block Switcher Object Reference)) -> (ImageMID:Material Instance Dynamic Object Reference)
   > SetVectorParameterValue(self=@n2, ParameterName=Color_SelectedIdle, Value=Get MI_ColorParameter() -> (MI_ColorParameter:Linear Color Structure))
   > SetVectorParameterValue(self=@n2, ParameterName=Color_SelectedHovered, Value=Get MI_ColorParameter())
   > SetVectorParameterValue(self=@n2, ParameterName=Color_SelectedFocused, Value=Get MI_ColorParameter())
   > SetVectorParameterValue(self=@n2, ParameterName=Color_SelectedPressed, Value=Get MI_ColorParameter())
   > SetVectorParameterValue(self=@n2, ParameterName=Color_SelectedDisabled, Value=Get MI_ColorParameter())
  ]}
 SetToggleStateAndVisibility {nodes [
   n1 = SetToggleStateAndVisibility!() -> (
    then
    PrivacyAndFillToggleSettings_bPrivacyValue:bool=~
    PrivacyAndFillToggleSettings_bPrivacyAvailable:bool=~
    PrivacyAndFillToggleSettings_bFillValue:bool=~
    PrivacyAndFillToggleSettings_bFillAvailable:bool=~
   ) {is=title K2Node_FunctionEntry}
   > n2 = Sequence()
   > SetIsSelected!(execute, self:Common Button Base Object Reference=Get ToggleButton_Private(), InSelected:bool=@n1, bGiveClickFeedback:bool=false) -> (then)
   > SetIsSelected(self=Get ToggleButton_TeamFill(), InSelected=@n1.PrivacyAndFillToggleSettings_bFillValue, bGiveClickFeedback=false)
   SetVisibility(
    execute=@n2.then_1
    self=Get HBox_Toggle_Private() -> (HBox_Toggle_Private:Horizontal Box Object Reference)
    InVisibility=Select(Option 0=Collapsed, Option 1=SelfHitTestInvisible, Index=@n1.PrivacyAndFillToggleSettings_bPrivacyAvailable)
   )
   > SetVisibility(
    self=Get HBox_Toggle_TeamFill() -> (HBox_Toggle_TeamFill:Horizontal Box Object Reference)
    InVisibility=Select(Option 0=Collapsed, Option 1=SelfHitTestInvisible, Index=@n1.PrivacyAndFillToggleSettings_bFillAvailable)
   )
  ]}
]
variables [OnConfirm_Event {}]
widgets {
 info {
  inheritedWidgetCount=0
  namedSlotCount=2
  parentClass=&CommonActivatableWidget
  rootWidgetClass=/FortUILibrary/Systems/Dialog/WBP_UIKit_Dialog_Internal.WBP_UIKit_Dialog_Internal_C
  widgetCount=11
 }
 widgets [
  Dialog_Internal {
   widgetClassPath=/FortUILibrary/Systems/Dialog/WBP_UIKit_Dialog_Internal
   bIsVariable=true
   widget=Dialog_Internal
   properties {
    bIsFocusable=true
    designTimePreview_DialogVM=/FortniteCoreUI/UI/DiscoveryShared/GameDetails/DialogVM_ActivityDetails_Subscribe.DialogVM_ActivityDetails_Subscribe_C
    useClickCatch=true
    useFooter=true
    buttons {
     each Button Horizontal Fill=false
     each Button Min Width=0
     each Button Size=Medium
     focuse Button=0
     space Between Buttons=30
     cTA Buttons Types [Quiet]
     group Placement {
      horizontal Alignment=HAlign_Fill
      vertical Alignment=VAlign_Bottom
      grid {column=1, column Span=0, layer=0, row=2, row Span=0, nudge=<0 0>}
      padding=<ltrb 80 0 80 -45>
     }
    }
    dialogSIze {
     height Override {enable=false, value=800}
     max Aspect Ratio {enable=false, value=0}
     max Desired Height {enable=false, value=800}
     max Desired Width {enable=false, value=0}
     min Aspect Ratio {enable=false, value=0}
     min Desired Height {enable=true, value=552}
     min Desired Width {enable=false, value=0}
     width Override {enable=true, value=961}
    }
    footerTextProperties {
     auto Wrap Text=true
     justification=Left
     line Height Percentage=1.35
     mobile Font Size Multiplier=1.3
     transform Policy=None
     wrap Text At=0
     color {colorUseRule=UseColor_Specified, specifiedColor=<rgba 0.5776 0.5776 0.5776 1>}
     font {
      bForceMonospaced=false
      bMaterialIsStencil=false
      fontMaterial=null
      fontObject=&HeadingNow
      letterSpacing=1
      monospacedWidth=1
      size=11.25
      skewAmount=0
      typefaceFontName=74 Regular
      outlineSettings {
       bApplyOutlineToDropShadows=false
       bMiteredCorners=false
       bSeparateFillAlpha=false
       outlineMaterial=null
       outlineSize=0
       outlineColor=<rgba 0 0 0 1>
      }
     }
     placement {
      horizontal Alignment=HAlign_Fill
      vertical Alignment=VAlign_Fill
      grid {column=1, column Span=0, layer=0, row=3, row Span=0, nudge=<0 0>}
      padding=<ltrb 80 0 160 68>
     }
    }
   }
  }
  Content {widgetClassPath=&Overlay, namedSlotHost=Dialog_Internal, widget=Content}
  VBox_Toggles {
   widgetClassPath=&VerticalBox
   bIsVariable=true
   parent=Content
   widget=VBox_Toggles
   slot {class=OverlaySlot, properties {horizontalAlignment=HAlign_Fill, verticalAlignment=VAlign_Fill, padding=<ltrb 80 0 80 0>}}
  }
  HBox_Toggle_Private {
   widgetClassPath=&HorizontalBox
   bIsVariable=true
   parent=VBox_Toggles
   widget=HBox_Toggle_Private
   properties {bOverride_Cursor=true, navigation=&self:WidgetTree.HBox_Toggle_Private.WidgetNavigation_0}
   slot {class=VerticalBoxSlot, properties {padding=<ltrb 0 0 0 10>}}
  }
  CommonTextBlock_137 {
   widgetClassPath=&CommonTextBlock
   parent=HBox_Toggle_Private
   widget=CommonTextBlock_137
   properties {
    bIsScrollingEnabled=false
    text=Private
    font {
     bForceMonospaced=false
     bMaterialIsStencil=false
     fontMaterial=null
     fontObject=&NotoSans
     letterSpacing=0
     monospacedWidth=1
     size=24
     skewAmount=0
     typefaceFontName=NotoSans-Regular
     outlineSettings {
      bApplyOutlineToDropShadows=false
      bMiteredCorners=false
      bSeparateFillAlpha=false
      outlineMaterial=null
      outlineSize=0
      outlineColor=<rgba 0 0 0 1>
     }
    }
    shadowOffset=<0 0>
   }
   slot {class=HorizontalBoxSlot, properties {padding=<ltrb 0 0 8 0>, size {sizeRule=Fill, value=1}}}
  }
  ToggleButton_Private {
   widgetClassPath=/FortUILibrary/UIKit/Buttons/WBP_UIKit_SwitcherButton
   bIsVariable=true
   parent=HBox_Toggle_Private
   widget=ToggleButton_Private
   properties {
    bHideInputAction=true
    bIsFocusable=true
    navigation=&self:WidgetTree.ToggleButton_Private.WidgetNavigation_1
    useTextBlock=false
    buttonPadding=<ltrb 0 0 0 0>
    triggeringInputAction {dataTable=null, rowName=None}
   }
   slot {class=HorizontalBoxSlot}
  }
  HBox_Toggle_TeamFill {
   widgetClassPath=&HorizontalBox
   bIsVariable=true
   parent=VBox_Toggles
   widget=HBox_Toggle_TeamFill
   properties {bOverride_Cursor=true, navigation=&self:WidgetTree.HBox_Toggle_TeamFill.WidgetNavigation_0}
   slot {class=VerticalBoxSlot}
  }
  CommonTextBlock {
   widgetClassPath=&CommonTextBlock
   parent=HBox_Toggle_TeamFill
   widget=CommonTextBlock
   properties {
    bIsScrollingEnabled=false
    text=Team Fill
    font {
     bForceMonospaced=false
     bMaterialIsStencil=false
     fontMaterial=null
     fontObject=&NotoSans
     letterSpacing=0
     monospacedWidth=1
     size=24
     skewAmount=0
     typefaceFontName=NotoSans-Regular
     outlineSettings {
      bApplyOutlineToDropShadows=false
      bMiteredCorners=false
      bSeparateFillAlpha=false
      outlineMaterial=null
      outlineSize=0
      outlineColor=<rgba 0 0 0 1>
     }
    }
    shadowOffset=<0 0>
   }
   slot {class=HorizontalBoxSlot, properties {padding=<ltrb 0 0 8 0>, size {sizeRule=Fill, value=1}}}
  }
  ToggleButton_TeamFill {
   widgetClassPath=&WBP_UIKit_SwitcherButton
   bIsVariable=true
   parent=HBox_Toggle_TeamFill
   widget=ToggleButton_TeamFill
   properties {
    bHideInputAction=true
    bIsFocusable=true
    navigation=&self:WidgetTree.ToggleButton_TeamFill.WidgetNavigation_1
    useTextBlock=false
    buttonPadding=<ltrb 0 0 0 0>
    triggeringInputAction {dataTable=null, rowName=None}
   }
   slot {class=HorizontalBoxSlot}
  }
  VisiblityBorder_Touch {
   widgetClassPath=&CommonHardwareVisibilityBorder
   namedSlotHost=Dialog_Internal
   widget=VisiblityBorder_Touch
   properties {
    horizontalAlignment=HAlign_Right
    verticalAlignment=VAlign_Top
    padding=<ltrb 0 32 32 0>
    visibilityQuery {
     autoDescription=" ANY( Input.Touch )"
     tokenStreamVersion=0
     userDescription=""
     queryTokenStream [0,1,1,1,0]
     tagDictionary [{tagName=Input.Touch}]
    }
   }
  }
  Button_BackTouch {
   widgetClassPath=/FortUILibrary/Foundation/Buttons/CloseButton
   bIsVariable=true
   parent=VisiblityBorder_Touch
   widget=Button_BackTouch
   properties {
    onButtonBaseClicked=()
    onButtonBaseDoubleClicked=()
    onButtonBaseHovered=()
    onButtonBaseUnhovered=()
    onSelectedChangedBase=()
    renderTransform {angle=0, scale=<1 1>, shear=<0 0>, translation=<0 10>}
   }
   slot {class=BorderSlot, properties {horizontalAlignment=HAlign_Right, verticalAlignment=VAlign_Top, padding=<ltrb 0 32 32 0>}}
  }
 ]
}
