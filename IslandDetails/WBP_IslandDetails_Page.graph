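# Island details page widget, a modal showing an island's media, stats, description, community highlights and footer, with select, favorite and privacy/fill option actions.
# On construct it creates an island details view model from the mnemonic in UI state tree parameters, then initializes the carousel, context menu and footer from it.
# One media gallery moves between carousel and expanded modal.
# Custom navigation functions route focus around hidden elements, and mouse wheel scrolls child by child.
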WidgetBlueprint /IslandDetails/WBP_IslandDetails_Page
animations [DetailsIntro {
  binding_count=6
  length_seconds=0.48
  object=&self:DetailsIntro
  sequence {bindings [
    VBox_Content {
     id=#1
     possessed_class=VerticalBox
     tracks [{
       type=MovieScene2DTransformTrack
       sections [{type=MovieScene2DTransformSection, channels [Translation.Y {type=MovieSceneScriptingFloatChannel, keys [[0,4],[8,0]]}]}]
      }]
    }
    Button_ContextMenu {
     id=#2
     possessed_class=WBP_UIKit_ContextMenuAnchorButton_C
     tracks [
      {
       type=MovieSceneFloatTrack
       display_name=Render Opacity
       sections [{type=MovieSceneFloatSection, channels [None {type=MovieSceneScriptingFloatChannel, keys [[6.4,0],[9.5993,1]]}]}]
      }
      {
       type=MovieScene2DTransformTrack
       sections [{
         type=MovieScene2DTransformSection
         channels [
          Scale.X {type=MovieSceneScriptingFloatChannel, keys [[6.4,0.875],[9.5997,1]]}
          Scale.Y {type=MovieSceneScriptingFloatChannel, keys [[6.4,0.875],[9.5997,1]]}
         ]
        }]
      }
     ]
    }
    DisabledWarningHB {
     id=#3
     possessed_class=HorizontalBox
     tracks [{
       type=MovieSceneFloatTrack
       display_name=Render Opacity
       sections [{type=MovieSceneFloatSection, channels [None {type=MovieSceneScriptingFloatChannel, keys [[3.2,0.125],[7.9997,1]]}]}]
      }]
    }
    Button_Favorite {
     id=#4
     possessed_class=WBP_GameDetails_Toggle_C
     tracks [
      {
       type=MovieSceneFloatTrack
       display_name=Render Opacity
       sections [{type=MovieSceneFloatSection, channels [None {type=MovieSceneScriptingFloatChannel, keys [[6.4,0],[9.5993,1]]}]}]
      }
      {
       type=MovieScene2DTransformTrack
       sections [{
         type=MovieScene2DTransformSection
         channels [
          Scale.X {type=MovieSceneScriptingFloatChannel, keys [[6.4,0.875],[9.5997,1]]}
          Scale.Y {type=MovieSceneScriptingFloatChannel, keys [[6.4,0.875],[9.5997,1]]}
         ]
        }]
      }
     ]
    }
    WBP_IslandDetails_Header {
     id=#5
     possessed_class=WBP_IslandDetails_Header_C
     tracks [{
       type=MovieSceneFloatTrack
       display_name=Render Opacity
       sections [{type=MovieSceneFloatSection, channels [None {type=MovieSceneScriptingFloatChannel, keys [[3.2,0.125],[7.9993,1]]}]}]
      }]
    }
    Button_Select {
     id=#6
     possessed_class=WBP_UIKit_Button_Loud_C
     tracks [
      {
       type=MovieSceneFloatTrack
       display_name=Render Opacity
       sections [{type=MovieSceneFloatSection, channels [None {type=MovieSceneScriptingFloatChannel, keys [[4.7997,0],[9.5993,1]]}]}]
      }
      {
       type=MovieScene2DTransformTrack
       sections [{type=MovieScene2DTransformSection, channels [Translation.Y {type=MovieSceneScriptingFloatChannel, keys [[4.8,8],[9.5997,0]]}]}]
      }
     ]
    }
   ]}
 }]
graphs [
 BP_GetDesiredFocusTarget {nodes [
   Get Desired Focus Target()
   > n1 = Get!(execute:Exec) -> (then:Exec, else:Exec, IsNextFocusReturningFromExpandedMediaGallery:bool)
   > Return Node!(
    execute
    ReturnValue:Widget Object Reference=Get WBP_IslandDetails_Carousel() -> (WBP_IslandDetails_Carousel:WBP Island Details Carousel Object Reference)
   )
   n2 = GetActiveWidget!(self:Widget Switcher Object Reference=Get WidgetSwitcher_Buttons() -> (WidgetSwitcher_Buttons:Common Activatable Widget Switcher Object Reference)) -> (ReturnValue:Widget Object Reference)
   n3 = Cast To FortPluginWidgetInjectionSlot!(execute=@n1.else, Object=@n2) -> (then, CastFailed, AsFort Plugin Widget Injection Slot:Fort Plugin Widget Injection Slot Object Reference) {is=title K2Node_DynamicCast}
   > n4 = GetDesiredFocusTarget!(execute, self:Fort Plugin Widget Injection Slot Object Reference=@n3) -> (then, ReturnValue:Widget Object Reference)
   > Return Node(ReturnValue=@n4)
   n5 = Get Button_Select() -> (Button_Select:WBP UIKit Button Loud Object Reference)
   n6 = Branch(execute=@n3.CastFailed, Condition=NOT Boolean(A=GetLocked!(self:Common Button Base Object Reference=@n5) -> (ReturnValue:bool)))
   > Return Node(ReturnValue=@n2)
   Return Node(execute=@n6.else, ReturnValue=Get Button_Favorite() -> (Button_Favorite:WBP Game Details Toggle Object Reference))
  ]}
 BP_OnHandleBackAction {nodes [
   On Handle Back Action() {is=title K2Node_FunctionEntry}
   > HandleBackButtonClicked(self:Fort Activity Details Modal V2 Object Reference)
   > Return Node(ReturnValue:bool=false)
  ]}
 ConvertLockedStateToIndex {nodes [
   n1 = ConvertLockedStateToIndex!() -> (then, IsLocked:bool=~, IsRestricted:bool=~, IsTimegated:bool=~) {is=title K2Node_FunctionEntry}
   n2 = OR Boolean(A=@n1, B=@n1.IsRestricted)
   > Return Node!(execute, Index:int=SelectInt(A=2, B=SelectInt(A=1, B=0, bPickA=@n2), bPickA=@n1.IsTimegated))
  ]}
 EventGraph {nodes [
   "On Clicked (Button_Select)"!() -> (OutputDelegate, then, Button:Common Button Base Object Reference) {is=title K2Node_ComponentBoundEvent}
   > HandleAcceptButtonClicked!(
    execute
    self:Fort Activity Details Modal V2 Object Reference
    SelectedGameActivity:Fort Game Activity Object Reference=GetGameActivity!(self:Discover Activity View Model Object Reference=Get FortIslandDetailsVM() -> (FortIslandDetailsVM:Fort Island Details VM Object Reference)) -> (ReturnValue:Fort Game Activity Object Reference)
   ) -> (then)
   n1 = "On Focused (WBP_IslandDetails_Carousel)"() {is=title K2Node_ComponentBoundEvent}
   ScrollWidgetIntoView!(
    execute=[@n1.then, @n2.then]
    self:Scroll Box Object Reference=Get ScrollBox() -> (ScrollBox:Scroll Box Object Reference)
    WidgetToFind:Widget Object Reference=Get WBP_IslandDetails_Header() -> (WBP_IslandDetails_Header:WBP Island Details Header Object Reference)
    AnimateScroll:bool=true
    ScrollDestination:EDescendantScrollDestination Enum=Center
    Padding:float=0
   ) -> (then)
   "On Focused (WBP_IslandDetails_Footer)"() {is=title K2Node_ComponentBoundEvent}
   > ScrollWidgetIntoView(
    self=Get ScrollBox()
    WidgetToFind=Get WBP_IslandDetails_Footer() -> (WBP_IslandDetails_Footer:WBP Island Details Footer Object Reference)
    AnimateScroll=true
    ScrollDestination=Center
    Padding=0
   )
   n2 = "On Focused (WBP_IslandDetails_Stats)"() {is=title K2Node_ComponentBoundEvent}
   "On Widget Injected (SqueegeeSlot_LockInfo)"!() -> (OutputDelegate, then, Widget:Widget Object Reference) {is=title K2Node_ComponentBoundEvent}
   > RequestRefreshFocus(self:Common Activatable Widget Object Reference)
   "On Clicked (Button_CreatorPage)"!() -> (OutputDelegate, then, Button:Common Button Base Object Reference) {is=title K2Node_ComponentBoundEvent}
   > HandleCreatorPageButtonClicked(self:Fort Activity Details Modal V2 Object Reference)
   "On Clicked (Button_Back)"!() -> (OutputDelegate, then, Button:Common Button Base Object Reference) {is=title K2Node_ComponentBoundEvent}
   > HandleBackButtonClicked()
   "On Clicked (WBP_IslandDetails_Carousel)"() {is=title K2Node_ComponentBoundEvent}
   > ActivateWidget(self=Get MediaGallery_Expanded() -> (MediaGallery_Expanded:WBP Island Details Media Gallery Modal Object Reference))
   "On Clicked (Button_Favorite)"!() -> (OutputDelegate, then, Button:Common Button Base Object Reference) {is=title K2Node_ComponentBoundEvent}
   > ToggleFavorite(self:Fort Discover Link Entry Tile Item VM Object Reference=Get FortIslandDetailsVM())
   n3 = Get bIsFavorite!(self:Fort Discover Link Entry Tile Item VM Object Reference=Get FortIslandDetailsVM()) -> (bIsFavorite:bool)
   > PlaySound2D(Sound=Select(
     Option 0:Sound Base Object Reference=Get Unfavorite_Sound() -> (Unfavorite_Sound:Sound Base Object Reference)
     Option 1:Sound Base Object Reference=Get Favorite_Sound() -> (Favorite_Sound:Sound Base Object Reference)
     Index=@n3
    ) -> (ReturnValue:Sound Base Object Reference))
   > n4 = Branch(Condition=@n3)
   n5 = Get Favorite Count() -> (Favorite Count:int=0)
   > n6 = Increment Int(Value=@n5)
   SetFavoriteCount!(execute=[@n6."  ", @n7."  "], self, FavoriteCount:int=@n5) -> (then)
   n7 = Decrement Int(" "=@n4.else, Value=@n5)
   "On Widget Activated (MediaGallery_Expanded)"() {is=title K2Node_ComponentBoundEvent}
   > AddMediaGallery!(
    execute
    self:WBP Island Details Media Gallery Modal Object Reference=Get MediaGallery_Expanded()
    InMediaGallery:WBP Island Details Media Gallery Object Reference=Get MediaGallery() -> (MediaGallery:WBP Island Details Media Gallery Object Reference)
   ) -> (then)
   > Set IsNextFocusReturningFromExpandedMediaGallery(IsNextFocusReturningFromExpandedMediaGallery:bool=true) -> (Output_Get:bool)
   "On Widget Deactivated (MediaGallery_Expanded)"() {is=title K2Node_ComponentBoundEvent}
   > AddMediaGallery(self:WBP Island Details Carousel Object Reference=Get WBP_IslandDetails_Carousel(), InMediaGallery=Get MediaGallery())
   > Set IsNextFocusReturningFromExpandedMediaGallery(IsNextFocusReturningFromExpandedMediaGallery=false)
   "On Widget Injected (CommunityRow)"!() -> (OutputDelegate, then, Widget:Widget Object Reference) {is=title K2Node_ComponentBoundEvent}
   > SetupCommunityHighlights()
   "On Active Widget Index Changed BP (WidgetSwitcher_Buttons)"!() -> (OutputDelegate, then, ActiveWidget:Widget Object Reference, ActiveIndex:int=0) {is=title K2Node_ComponentBoundEvent}
   > RequestRefreshFocus()
   "On Clicked (Button_IslandOptions)"!() -> (OutputDelegate, then, Button:Common Button Base Object Reference) {is=title K2Node_ComponentBoundEvent}
   > n8 = Construct Object from Class(Class=&/DialogVM_IslandDetails_IslandOptions.DialogVM_IslandDetails_IslandOptions_C) -> (ReturnValue:Dialog VM Island Details Island Options Object Reference)
   > n9 = Create Widget!(
    execute:Exec
    Class:User Widget Class Reference=&/WBP_UIKit_Dialog_IslandDetails.WBP_UIKit_Dialog_IslandDetails_C
    OwningPlayer:Player Controller Object Reference
    UIKitDialogViewModel:UI Kit Dialog Viewmodel Object Reference=@n8
   ) -> (then:Exec, ReturnValue:WBP UIKit Dialog Island Details Object Reference)
   > n10 = Set Modal_IslandOptions(Modal_IslandOptions:WBP UIKit Dialog Island Details Object Reference=@n9) -> (Output_Get:WBP UIKit Dialog Island Details Object Reference)
   > TryPushWidgetAsModalPanel!(execute, PushWidget:User Widget Object Reference=@n10) -> (then)
   > n11 = GetPrivacyAndFillToggleSettings!(execute, self:Fort Activity Details Modal V2 Object Reference) -> (then, ReturnValue:Privacy And Fill Toggle Settings Structure)
   n12 = Get Modal_IslandOptions() -> (Modal_IslandOptions:WBP UIKit Dialog Island Details Object Reference)
   > n13 = SetToggleStateAndVisibility!(
    execute
    self:WBP UIKit Dialog Island Details Object Reference=@n12
    PrivacyAndFillToggleSettings:Privacy And Fill Toggle Settings Structure=@n11
   ) -> (then)
   n14 = OnIslandOptionsConfirm!() -> (
    OutputDelegate
    then
    PrivacyAndFill_bPrivacyValue:bool=~
    PrivacyAndFill_bPrivacyAvailable:bool=~
    PrivacyAndFill_bFillValue:bool=~
    PrivacyAndFill_bFillAvailable:bool=~
   ) {is=title K2Node_CustomEvent}
   Bind Event to On Confirm Event(execute=@n13.then, self:WBP UIKit Dialog Island Details Object Reference=@n12, Delegate=@n14) {is=title K2Node_AddDelegate}
   Event PreConstruct()
   > DisableButtonWithReason!(
    execute
    self:Common Button Base Object Reference=Get Button_TimeLocked() -> (Button_TimeLocked:WBP UIKit Button Loud Object Reference)
    DisabledReason:"Text (by ref)"=Format Text(Format="LOCTEXT(\"The event is not live yet!\")")
   ) -> (then)
   > SetVisibility(self=Get CommunityRow() -> (CommunityRow:Fort Plugin Widget Injection Slot Object Reference), InVisibility=Collapsed)
   Event OnInitialized()
   > n15 = Construct Object from Class(Class=&/WBP_IslandDetails_MediaGallery.WBP_IslandDetails_MediaGallery_C) -> (ReturnValue:WBP Island Details Media Gallery Object Reference)
   > n16 = Set MediaGallery(MediaGallery:WBP Island Details Media Gallery Object Reference=@n15) -> (Output_Get:WBP Island Details Media Gallery Object Reference)
   > AddMediaGallery(self=Get WBP_IslandDetails_Carousel(), InMediaGallery=@n16)
   Event On Activated()
   > SetScrollWidget!(
    execute
    ToScroll:Widget Object Reference=Get ScrollBox!(self:WBP Island Details Stats Object Reference=Get WBP_IslandDetails_Stats() -> (WBP_IslandDetails_Stats:WBP Island Details Stats Object Reference)) -> (ScrollBox)
   ) -> (then)
   > Set MouseScrollIndex(MouseScrollIndex:int=0) -> (Output_Get:int=0)
   > SetBottomButtonVisibility!(
    execute
    self:WBP Bottom Bar Decoy Button Object Reference=Get Button_CreatorPage() -> (Button_CreatorPage:WBP Bottom Bar Decoy Button Object Reference)
    ShouldButtonBeVisible:bool=GetIsPlayTimeOut!() -> (IsOut:bool)
   ) -> (then)
   Event Destruct()
   n17 = GetOwningPlayer()
   n18 = Get UIStateTreeManager() -> (ReturnValue:UIState Tree Manager Object Reference) {is=title K2Node_GetSubsystem}
   > Clear Player Parameter!(
    execute
    self:UIState Tree Manager Object Reference=@n18
    ParameterName:Name=SelectedLinkMnemonic
    StateTreeOwner:Player Controller Object Reference=@n17
   ) -> (then)
   > Clear Player Parameter(self=@n18, ParameterName=ShouldShowLikePercent, StateTreeOwner=@n17)
   n19 = Get WBP_IslandDetails_Footer()
   > Unbind Event from On Clicked(
    self:Common Button Base Object Reference=Get FooterButton_ViewCreator!(self:WBP Island Details Footer Object Reference=@n19) -> (FooterButton_ViewCreator:WBP UIKit Button Generic Object Reference)
    Delegate=Create Event()
   ) {is=title K2Node_RemoveDelegate}
   > Unbind Event from On Clicked(
    self=Get FooterButton_ViewAttributions!(self:WBP Island Details Footer Object Reference=@n19) -> (FooterButton_ViewAttributions:WBP UIKit Button Generic Object Reference)
    Delegate=Create Event()
   )
   > Unbind Event from On Clicked(
    self=Get FooterButton_ReportIsland!(self:WBP Island Details Footer Object Reference=@n19) -> (FooterButton_ReportIsland:WBP UIKit Button Generic Object Reference)
    Delegate=Create Event()
   )
   > Set IsInitialized(IsInitialized:bool=false) -> (Output_Get:bool)
   > Set FortIslandDetailsVM(FortIslandDetailsVM:Fort Island Details VM Object Reference) -> (Output_Get:Fort Island Details VM Object Reference)
   Event Construct()
   > n20 = Construct Object from Class(Class=/Script/FortniteUI.FortIslandDetailsVM) -> (ReturnValue:Fort Island Details VM Object Reference)
   > Set FortIslandDetailsVM(FortIslandDetailsVM=@n20)
   > n21 = "Get Player Parameter (String)"!(
    execute
    self:UIState Tree Manager Object Reference=@n18
    ParameterName:Name=SelectedLinkMnemonic
    StateTreeOwner:Player Controller Object Reference=@n17
   ) -> (then, bSuccess:bool, ReturnValue:String) {is=title K2Node_CallFunction}
   > InitWithMnemonic!(
    execute
    self:Fort Discover Link Entry Tile Item VM Object Reference=Get FortIslandDetailsVM()
    InLinkMnemonic:String=@n21.ReturnValue
    InActivityIndex:int=0
   ) -> (then)
   > n22 = "Get Player Parameter (Bool)"!(
    execute
    self:UIState Tree Manager Object Reference=@n18
    ParameterName:Name=ShouldShowLikePercent
    StateTreeOwner:Player Controller Object Reference=@n17
   ) -> (then, bSuccess:bool, ReturnValue:bool) {is=title K2Node_CallFunction}
   n23 = Get FortIslandDetailsVM()
   > SetShouldShowLikePercent!(execute, self:Discover Activity View Model Object Reference=@n23, bInShowLikePercent:bool=@n22.ReturnValue) -> (then)
   > Set FortIslandDetailsVM!(execute, FortIslandDetailsVM=@n23, self:WBP Island Details Stats Object Reference=Get WBP_IslandDetails_Stats()) -> (then, Output_Get)
   > n24 = Set IsInitialized(IsInitialized=true)
   > Set Toggle!(
    execute
    self:WBP Game Details Toggle Object Reference=Get Button_Favorite()
    IsInit:bool=@n24
    IsToggled:bool=Get bIsFavorite(self=Get FortIslandDetailsVM())
   ) -> (then) {is=title K2Node_CallFunction}
   SetIslandOptionsButtonText(execute=@n25.then)
   n25 = UpdatePrivacyAndFill!(
    execute=@n14.then
    self:Fort Activity Details Modal V2 Object Reference
    bIsPrivate:bool=@n14.PrivacyAndFill_bPrivacyValue
    bIsFill:bool=@n14.PrivacyAndFill_bFillValue
   ) -> (then)
  ]}
 GetActiveSelectButton {nodes [
   GetActiveSelectButton() {is=title K2Node_FunctionEntry}
   n1 = Get WidgetSwitcher_Buttons()
   > Return Node!(execute, IsVisible:bool=IsVisible(self=@n1), ReturnValue:Widget Object Reference=GetActiveWidget(self=@n1))
  ]}
 GetCommunityRowFocusTarget {nodes [
   GetCommunityRowFocusTarget() {is=title K2Node_FunctionEntry}
   n1 = Get CommunityRow()
   n2 = IsVisible(self=@n1)
   > n3 = Branch(Condition=@n2)
   > n4 = GetDesiredFocusTarget(self=@n1)
   > Return Node!(execute, CommunityRowFocusTarget:Widget Object Reference=@n4, IsVisible=@n2)
   Return Node(execute=@n3.else, IsVisible=@n2)
  ]}
 GetContentChildrenCount {nodes [
   GetContentChildrenCount() {is=title K2Node_FunctionEntry}
   n1 = Get VBox_Content() -> (VBox_Content:Vertical Box Object Reference)
   > Return Node!(execute, ReturnValue:int=int - int(A=GetChildrenCount!(self:Panel Widget Object Reference=@n1) -> (ReturnValue:int=0)))
  ]}
 GetIsPlayTimeOut {nodes [
   GetIsPlayTimeOut() {is=title K2Node_FunctionEntry}
   n1 = Get PlaytimeState!(self:Player Playtime State View Model Object Reference=Get PlayerPlaytimeStateVM() -> (PlayerPlaytimeStateVM:Player Playtime State View Model Object Reference)) -> (PlaytimeState:EPlaytimeState Enum=Unblocked)
   > Return Node!(execute, IsOut:bool=NOT Boolean(A="Equal (Enum)"(A:EPlaytimeState Enum=@n1, B:EPlaytimeState Enum=PlayTime_Blocked)))
  ]}
 GetPrimaryContentSetup {nodes [
   GetPrimaryContentSetup() {is=title K2Node_FunctionEntry}
   > Return Node!(
    execute
    ReturnValue:Primary Content Setup Structure=Make Primary Content Setup!(
     bShowBottomBar:bool=false
     bShowChatWidget:bool=false
     bEnableNewTextChat:bool=False
     bShowTopBar:bool=false
     bTabSupportsInterface:bool=False
    ) -> (PrimaryContentSetup:Primary Content Setup Structure) {is=title K2Node_MakeStruct}
   )
  ]}
 GetSelectButtonFocusTarget {nodes [
   GetSelectButtonFocusTarget() {is=title K2Node_FunctionEntry}
   n1 = IsVisible(self=GetActiveWidget(self=Get WidgetSwitcher_Buttons()))
   > n2 = Branch(Condition=@n1)
   n3 = GetActiveWidget(self=Get WidgetSwitcher_Buttons())
   n4 = Get SqueegeeSlot_LockInfo() -> (SqueegeeSlot_LockInfo:Squeegee Injection Slot Object Reference)
   > n5 = Branch(Condition="Equal (Object)"(A=@n4, B=@n3))
   > n6 = GetDesiredFocusTarget(self=@n4)
   > Return Node!(execute, SelectFocusTarget:Widget Object Reference=@n6, IsVisible:bool=@n1)
   Return Node(execute=@n5.else, SelectFocusTarget=@n3, IsVisible=@n1)
   Return Node(execute=@n2.else, IsVisible=@n1)
  ]}
 HandleContextMenuAction {nodes [
   n1 = HandleContextMenuAction!() -> (then, InteractionNameID:Name) {is=title K2Node_FunctionEntry}
   > n2 = Branch(Condition="Equal (Name)"(A=@n1, B=Get ContextMenuID_ViewCreator() -> (ContextMenuID_ViewCreator:Name=None)))
   > n3 = HandleCreatorPageButtonClicked()
   Close(
    execute=[@n3.then, @n5.then, @n6.then]
    self:Menu Anchor Object Reference=Get Anchor!(self:UIKit Context Menu Anchor Object Reference=Get Button_ContextMenu() -> (Button_ContextMenu:WBP UIKit Context Menu Anchor Button Object Reference)) -> (Anchor:Menu Anchor Object Reference)
   )
   n4 = Branch(execute=@n2.else, Condition="Equal (Name)"(A=@n1, B=Get ContextMenuID_ViewAttributions() -> (ContextMenuID_ViewAttributions:Name=None)))
   > n5 = HandleAttributionsButtonClicked(self:Fort Activity Details Modal V2 Object Reference)
   Branch(execute=@n4.else, Condition="Equal (Name)"(A=@n1, B=Get ContextMenuID_ReportIsland() -> (ContextMenuID_ReportIsland:Name=None)))
   > n6 = OnFooterButtonClicked_ReportIsland!(execute, self, Button:Common Button Base Object Reference) -> (then)
  ]}
 Initialize {nodes [
   Initialize() {is=title K2Node_FunctionEntry}
   > UpdateCarousel()
   > SetupCommunityHighlights()
   > SetupContextMenu()
   > SetupFooterButtons()
   > SetIslandOptionsButtonText()
   > RequestRefreshFocus()
  ]}
 Nav_CommunityRow_Down {nodes [
   Nav_CommunityRow_Down!() -> (then, Navigation:EUINavigation Enum) {is=title K2Node_FunctionEntry}
   > Return Node!(
    execute
    ReturnValue:Widget Object Reference=Get WBP_IslandDetails_Description() -> (WBP_IslandDetails_Description:WBP Island Details Description Object Reference)
   )
  ]}
 Nav_CommunityRow_Up {nodes [
   Nav_CommunityRow_Up!() -> (then, Navigation:EUINavigation Enum) {is=title K2Node_FunctionEntry}
   n1 = GetSelectButtonFocusTarget!() -> (SelectFocusTarget:Widget Object Reference, IsVisible:bool) {is=title K2Node_CallFunction}
   n2 = Get Button_IslandOptions() -> (Button_IslandOptions:WBP UIKit Button Generic Object Reference)
   n3 = IsVisible(self=@n2)
   n4 = Get WBP_IslandDetails_Carousel()
   > Return Node(ReturnValue=Select(Option 0:Widget Object Reference=Select(Option 0=@n4, Option 1=@n2, Index=@n3), Option 1:Widget Object Reference=@n1, Index=@n1.IsVisible) -> (ReturnValue:Widget Object Reference))
  ]}
 Nav_Description_Down {nodes [
   Nav_Description_Down!() -> (then, Navigation:EUINavigation Enum) {is=title K2Node_FunctionEntry}
   n1 = Get FooterButton_ViewCreator(self=Get WBP_IslandDetails_Footer())
   > n2 = Branch(Condition=IsVisible(self=@n1))
   > Return Node(ReturnValue=@n1)
   n3 = Get FooterButton_ViewAttributions(self=Get WBP_IslandDetails_Footer())
   n4 = Branch(execute=@n2.else, Condition=IsVisible(self=@n3))
   > Return Node(ReturnValue=@n3)
   n5 = Get FooterButton_ReportIsland(self=Get WBP_IslandDetails_Footer())
   Branch(execute=@n4.else, Condition=IsVisible(self=@n5))
   > Return Node(ReturnValue=@n5)
  ]}
 Nav_Description_Up {nodes [
   Nav_Description_Up!() -> (then, Navigation:EUINavigation Enum) {is=title K2Node_FunctionEntry}
   n1 = GetCommunityRowFocusTarget!() -> (CommunityRowFocusTarget:Widget Object Reference, IsVisible:bool) {is=title K2Node_CallFunction}
   n2 = GetSelectButtonFocusTarget()
   n3 = Get Button_IslandOptions()
   n4 = Select(Option 0=Get WBP_IslandDetails_Carousel(), Option 1=@n3, Index=IsVisible(self=@n3))
   > Return Node(ReturnValue=Select(Option 0=Select(Option 0=@n4, Option 1=@n2, Index=@n2.IsVisible), Option 1=@n1, Index=@n1.IsVisible))
  ]}
 Nav_IslandOptions_Up {nodes [
   Nav_IslandOptions_Up!() -> (then, Navigation:EUINavigation Enum) {is=title K2Node_FunctionEntry}
   > Return Node(ReturnValue=Get WBP_IslandDetails_Carousel())
  ]}
 Nav_MainButtonsAndRating_Down {nodes [
   Nav_MainButtonsAndRating_Down!() -> (then, Navigation:EUINavigation Enum) {is=title K2Node_FunctionEntry}
   n1 = GetCommunityRowFocusTarget()
   > Return Node(ReturnValue=Select(Option 0=Get WBP_IslandDetails_Description(), Option 1=@n1, Index=@n1.IsVisible))
  ]}
 Nav_MainButtons_Up {nodes [
   Nav_MainButtons_Up!() -> (then, Navigation:EUINavigation Enum) {is=title K2Node_FunctionEntry}
   n1 = Get Button_IslandOptions()
   > Return Node(ReturnValue=Select(Option 0=Get WBP_IslandDetails_Carousel(), Option 1=@n1, Index=IsVisible(self=@n1)))
  ]}
 OnCommunityRowProviderSetupCompleted {nodes [
   n1 = OnCommunityRowProviderSetupCompleted!() -> (then, Provider Populated:bool=~) {is=title K2Node_FunctionEntry}
   n2 = GetCommunitiesExperimentValue!() -> (ReturnValue:bool)
   > SetVisibility(
    self=Get CommunityRow()
    InVisibility="To Visibility (Boolean)"!(
     bIsVisible:bool=AND Boolean(A=@n1, B=@n2)
     TrueVisibility:ESlateVisibility Enum=SelfHitTestInvisible
     FalseVisibility:ESlateVisibility Enum=Collapsed
    ) -> (ReturnValue:ESlateVisibility Enum=Visible)
   )
  ]}
 OnFooterButtonClicked_ReportIsland {nodes [
   OnFooterButtonClicked_ReportIsland!() -> (then, Button:Common Button Base Object Reference) {is=title K2Node_FunctionEntry}
   > ReportIsland!(execute, self:Fort Island Details VM Object Reference=Get FortIslandDetailsVM(), ContextWidget:User Widget Object Reference=Self-Reference()) -> (then)
  ]}
 OnFooterButtonClicked_ViewAttributions {nodes [
   OnFooterButtonClicked_ViewAttributions!() -> (then, Button:Common Button Base Object Reference) {is=title K2Node_FunctionEntry}
   > HandleAttributionsButtonClicked()
  ]}
 OnFooterButtonClicked_ViewCreator {nodes [
   OnFooterButtonClicked_ViewCreator!() -> (then, Button:Common Button Base Object Reference) {is=title K2Node_FunctionEntry}
   > HandleCreatorPageButtonClicked()
  ]}
 OnMouseWheel {nodes [
   n1 = OnMouseWheel!() -> (then, MyGeometry:Geometry Structure, MouseEvent:"Pointer Event Structure (by ref)") {is=title K2Node_FunctionEntry}
   n2 = Get Wheel Delta!(Input:"Pointer Event Structure (by ref)"=@n1.MouseEvent) -> (ReturnValue:float) {is=title K2Node_CallFunction}
   > n3 = Branch(Condition="float > float"(A=@n2))
   > n4 = Branch(Condition="integer <= integer"(A=Get MouseScrollIndex() -> (MouseScrollIndex:int=0)))
   > Return Node(ReturnValue:Event Reply Structure=Unhandled())
   n5 = Get MouseStartingIndex() -> (MouseStartingIndex:int=0)
   n6 = Get MouseScrollIndex() -> (MouseScrollIndex=0)
   n7 = GetContentChildrenCount!() -> (ReturnValue:int=0) {is=title K2Node_CallFunction}
   n8 = Get MouseScrollIndex() -> (MouseScrollIndex=0)
   n9 = Decrement Int(
    " "=@n4.else
    Value=SelectInt(A=int - int(A=@n6, B=@n5), B=Get MouseScrollIndex() -> (MouseScrollIndex=0), bPickA="Equal (Integer)"(A=@n7, B=@n8))
   )
   > n10 = Set MouseScrollIndex(MouseScrollIndex=@n9) -> (Output_Get=0)
   > ScrollWidgetIntoView(
    self=Get ScrollBox!() -> (ScrollBox)
    WidgetToFind=GetChildAt!(self:Panel Widget Object Reference=Get VBox_Content(), Index:int=@n10) -> (ReturnValue:Widget Object Reference)
    AnimateScroll=true
    ScrollDestination=IntoView
    Padding=120
   )
   > Return Node(ReturnValue=Handled())
   n11 = Branch(execute=@n3.else, Condition="float < float"(A=@n2))
   > n12 = Branch(Condition="Equal (Integer)"(A=GetContentChildrenCount() -> (ReturnValue=0), B=Get MouseScrollIndex() -> (MouseScrollIndex=0)))
   > Return Node(ReturnValue=Unhandled())
   n13 = Get MouseScrollIndex() -> (MouseScrollIndex=0)
   n14 = Increment Int(
    " "=@n12.else
    Value=SelectInt(A=Get MouseStartingIndex() -> (MouseStartingIndex=0), B=Get MouseScrollIndex() -> (MouseScrollIndex=0), bPickA="integer <= integer"(A=@n13))
   )
   > n15 = Set MouseScrollIndex(MouseScrollIndex=@n14) -> (Output_Get=0)
   > ScrollWidgetIntoView(
    self=Get ScrollBox()
    WidgetToFind=GetChildAt(self=Get VBox_Content(), Index=@n15)
    AnimateScroll=true
    ScrollDestination=IntoView
    Padding=120
   )
   > Return Node(ReturnValue=Handled())
   Return Node(execute=@n11.else, ReturnValue=Unhandled())
  ]}
 SetFavoriteCount {nodes [
   n1 = SetFavoriteCount!() -> (then, FavoriteCount:int) {is=title K2Node_FunctionEntry}
   > n2 = Set Favorite Count(Favorite Count:int=@n1) -> (Output_Get:int=0)
   n3 = Get Button_Favorite()
   > SetText!(execute, self:WBP UIKit Button Generic Base Object Reference=@n3, Text:Text=ConvertQuantityToShorthandText!(Quantity:int=@n2) -> (Text:Text)) -> (then)
   > SetVisibility(
    self=Get Block_Text!(self:WBP UIKit Button Generic Base Object Reference=@n3) -> (Block_Text:WBP UIKit Block Text Object Reference)
    InVisibility=Select(Option 0:ESlateVisibility Enum=Collapsed, Option 1:ESlateVisibility Enum=SelfHitTestInvisible, Index="integer > integer"(A=@n2)) -> (ReturnValue:ESlateVisibility Enum)
   )
  ]}
 SetIslandOptionsButtonEnabled {nodes [
   n1 = SetIslandOptionsButtonEnabled!() -> (then, InvalidActivityReason:EFortInvalidActivityReason Enum=None) {is=title K2Node_FunctionEntry}
   n2 = Get FortPartyContext!() -> (ReturnValue:Fort Party Context Object Reference) {is=title K2Node_GetBlueprintContextEx}
   > SetIsEnabled(
    self=Get Button_IslandOptions()
    bInIsEnabled=AND Boolean(
     A="Equal (Enum)"(A:EFortInvalidActivityReason Enum=@n1, B:EFortInvalidActivityReason Enum=None)
     B=IsLocalPlayerPartyLeader!(self:Fort Party Context Object Reference=@n2) -> (ReturnValue:bool)
    )
   )
  ]}
 SetIslandOptionsButtonText {nodes [
   SetIslandOptionsButtonText() {is=title K2Node_FunctionEntry}
   > n1 = GetPrivacyAndFillToggleSettings!(execute, self) -> (
    then
    ReturnValue_bPrivacyValue:bool=~
    ReturnValue_bPrivacyAvailable:bool=~
    ReturnValue_bFillValue:bool=~
    ReturnValue_bFillAvailable:bool=~
   )
   n2 = AND Boolean!(A=@n1.ReturnValue_bPrivacyAvailable, B=@n1.ReturnValue_bFillAvailable, C:bool=@n1.ReturnValue_bFillValue) -> (ReturnValue)
   n3 = SelectText(A="LOCTEXT(\"Private\")", B="LOCTEXT(\"Public\")", bPickA=@n1)
   n4 = AND Boolean!(A=@n1.ReturnValue_bFillValue, B=@n1.ReturnValue_bFillAvailable) -> (ReturnValue)
   > SetText(
    self=Get Button_IslandOptions()
    Text=Format Text!(
     Format="INVTEXT(\"{p}{comma}{f}\")"
     p:Text=SelectText(A=@n3, bPickA=@n1.ReturnValue_bPrivacyAvailable)
     comma:Text=SelectText(A="LOCTEXT(\", \")", bPickA=@n2)
     f:Text=SelectText(A="LOCTEXT(\"Fill\")", bPickA=@n4)
    ) -> (Result)
   )
  ]}
 SetIslandOptionsButtonVisibility {nodes [
   n1 = SetIslandOptionsButtonVisibility!() -> (then, AtLeastOne:bool=~) {is=title K2Node_FunctionEntry}
   n2 = CanShowPrivacyAndFillToggles!(self:Fort Activity Details Modal V2 Object Reference) -> (ReturnValue:bool)
   > SetVisibility(self=Get Button_IslandOptions(), InVisibility=Select(Option 0=Collapsed, Option 1=Visible, Index=AND Boolean(A=@n1, B=@n2)))
  ]}
 SetLoadingStateAndInitialize {nodes [
   n1 = SetLoadingStateAndInitialize!() -> (then, IsLinkEntrySet:bool=~) {is=title K2Node_FunctionEntry}
   > SetLoadingVisual!(execute, self, IsLoading:bool=NOT Boolean(A=@n1)) -> (then)
   > Branch(Condition=@n1)
   > Initialize() {is=title K2Node_CallFunction}
  ]}
 SetLoadingVisual {nodes [
   n1 = SetLoadingVisual!() -> (then, IsLoading:bool=~) {is=title K2Node_FunctionEntry}
   > SetVisibility(
    self=Get LoadingThrobber() -> (LoadingThrobber:WBP UIKit Throbber Object Reference)
    InVisibility=Select(Option 0=Hidden, Option 1=SelfHitTestInvisible, Index=@n1)
   )
   > SetVisibility(self=Get ScrollBox(), InVisibility=Select(Option 0=Visible, Option 1=Hidden, Index=@n1))
  ]}
 SetupCommunityHighlights {nodes [
   SetupCommunityHighlights() {is=title K2Node_FunctionEntry}
   > Branch(Condition=Get bIsLinkEntrySet!(self:Fort Discover Link Entry Tile Item VM Object Reference=Get FortIslandDetailsVM()) -> (bIsLinkEntrySet:bool))
   > n1 = Cast To UserWidget!(
    execute
    Object=GetInjectedWidget!(self:Fort Plugin Widget Injection Slot Object Reference=Get CommunityRow()) -> (ReturnValue:Widget Object Reference)
   ) -> (then, CastFailed, AsUser Widget:User Widget Object Reference) {is=title K2Node_DynamicCast}
   n2 = Get FortDiscoverProviderViewModel() -> (FortDiscoverProviderViewModel:Discover Provider View Model Object Reference)
   > SetViewModel!(
    execute
    self:MVVM View Object Reference=Get View From User Widget!(
     self:Viewmodel Engine Subsytem Object Reference=Get MVVMSubsystem() -> (ReturnValue:Viewmodel Engine Subsytem Object Reference) {is=title K2Node_GetEngineSubsystem}
     UserWidget:User Widget Object Reference=@n1
    ) -> (ReturnValue:MVVM View Object Reference) {is=title K2Node_CallFunction}
    ViewModelName:Name=FortDiscoverProviderViewModel
    ViewModel:Notify Field Value Changed Interface=@n2
   ) -> (then, ReturnValue:bool)
   n3 = Get FortIslandDetailsVM()
   > SetupCommunityIslandHighlightsRowFromBP!(
    execute
    self:Discover Provider View Model Object Reference=@n2
    InPlayer:Player Controller Object Reference=GetOwningPlayer()
    InLinkId:"Online Link Id Structure (by ref)"=Get LinkId!(self:Fort Game Activity Object Reference=GetGameActivity(self=@n3)) -> (LinkId:Online Link Id Structure)
    OnProviderSetupFromBPCompleted:Delegate=Create Event()
   ) -> (then)
  ]}
 SetupContextMenu {nodes [
   SetupContextMenu() {is=title K2Node_FunctionEntry}
   > n1 = Construct Object from Class(Class=/Script/FortniteUI.UIKitContextMenuActionContext) -> (ReturnValue:UIKit Context Menu Action Context Object Reference)
   > Set ContextMenuActionContext(ContextMenuActionContext:UIKit Context Menu Action Context Object Reference=@n1) -> (Output_Get:UIKit Context Menu Action Context Object Reference)
   > n2 = Sequence!(execute) -> (then_0, then_1, then_2:Exec, then_3:Exec)
   > Branch(Condition=GetIsPlayTimeOut() {is=title K2Node_CallFunction})
   > AddInteraction!(
    execute
    self:UIKit Context Menu Action Context Object Reference=Get ContextMenuActionContext() -> (ContextMenuActionContext:UIKit Context Menu Action Context Object Reference)
    InActionInfo_DisplayName:Text="LOCTEXT(\"View Developer\")"
    InActionInfo_IconTexture:Texture 2D Soft Object Reference
    InActionInfo_IconMaterial:Material Instance Soft Object Reference
    InActionInfo_SortPriority:int=1000
    InActionInfo_bIsAvailable:bool=True
    InActionInfo_bIsEnabled:bool=True
    InActionInfo_bIsSelected:bool=False
    InActionInfo_bAddDivider:bool=False
    InActionInfo_bPositionDividerAfter:bool=False
    InActionInfo_bCloseMenuOnSelect:bool=False
    InActionInfo_InteractionNameID:Name=Get ContextMenuID_ViewCreator() -> (ContextMenuID_ViewCreator=None)
    InActionInfo_Interaction:Data Table Row Handle Structure=(DataTable=None,RowName="")
    InActionInfo_PressedDelegate:Delegate="(null).None"
    InActionInfo_ReleasedDelegate:Delegate="(null).None"
    InInteractionDelegate:"Delegate (by ref)"=Create Event()
   ) -> (then)
   Branch(execute=@n2.then_2, Condition=CanBeReported!(self:Fort Island Details VM Object Reference=Get FortIslandDetailsVM()) -> (ReturnValue:bool))
   > AddInteraction(
    self=Get ContextMenuActionContext()
    InActionInfo_DisplayName="LOCTEXT(\"Report Island\")"
    InActionInfo_SortPriority=10
    InActionInfo_bIsAvailable=True
    InActionInfo_bIsEnabled=True
    InActionInfo_bIsSelected=False
    InActionInfo_bAddDivider=False
    InActionInfo_bPositionDividerAfter=False
    InActionInfo_bCloseMenuOnSelect=False
    InActionInfo_InteractionNameID=Get ContextMenuID_ReportIsland() -> (ContextMenuID_ReportIsland=None)
    InActionInfo_Interaction=(DataTable=None,RowName="")
    InActionInfo_PressedDelegate="(null).None"
    InActionInfo_ReleasedDelegate="(null).None"
    InInteractionDelegate=Create Event()
   )
   SetContext!(
    execute=@n2.then_3
    self:WBP UIKit Context Menu Anchor Button Base Object Reference=Get Button_ContextMenu()
    InteractionContext_Data:UIKit Context Menu Action Context Object Reference=Get ContextMenuActionContext()
   ) -> (then)
   n3 = Get ContextMenuActionContext()
   > SetVisibility(
    self=Get Button_ContextMenu()
    InVisibility=Select(
     Option 0=Collapsed
     Option 1=Visible
     Index=HasInteractionItems!(self:UIKit Context Menu Action Context Object Reference=@n3) -> (ReturnValue:bool)
    )
   )
   Branch(execute=@n2.then_1, Condition=HasAttributions!(self:Fort Island Details VM Object Reference=Get FortIslandDetailsVM()) -> (ReturnValue:bool))
   > AddInteraction(
    self=Get ContextMenuActionContext()
    InActionInfo_DisplayName="LOCTEXT(\"View Attributions\")"
    InActionInfo_SortPriority=10
    InActionInfo_bIsAvailable=True
    InActionInfo_bIsEnabled=True
    InActionInfo_bIsSelected=False
    InActionInfo_bAddDivider=False
    InActionInfo_bPositionDividerAfter=False
    InActionInfo_bCloseMenuOnSelect=False
    InActionInfo_InteractionNameID=Get ContextMenuID_ViewAttributions() -> (ContextMenuID_ViewAttributions=None)
    InActionInfo_Interaction=(DataTable=None,RowName="")
    InActionInfo_PressedDelegate="(null).None"
    InActionInfo_ReleasedDelegate="(null).None"
    InInteractionDelegate=Create Event()
   )
  ]}
 SetupFooterButtons {nodes [
   SetupFooterButtons() {is=title K2Node_FunctionEntry}
   > n1 = Sequence!(execute) -> (then_0, then_1, then_2)
   n2 = GetIsPlayTimeOut()
   n3 = Get FooterButton_ViewCreator(self=Get WBP_IslandDetails_Footer())
   > SetVisibility(self=@n3, InVisibility=Select(Option 0=Collapsed, Option 1=Visible, Index=@n2))
   > Branch(Condition=@n2)
   > Bind Event to On Clicked(self:Common Button Base Object Reference=@n3, Delegate=Create Event()) {is=title K2Node_AddDelegate}
   n4 = HasAttributions(self=Get FortIslandDetailsVM())
   n5 = Get FooterButton_ViewAttributions(self=Get WBP_IslandDetails_Footer())
   SetVisibility(execute=@n1.then_1, self=@n5, InVisibility=Select(Option 0=Collapsed, Option 1=Visible, Index=@n4))
   > Branch(Condition=@n4)
   > Bind Event to On Clicked(self=@n5, Delegate=Create Event())
   n6 = CanBeReported(self=Get FortIslandDetailsVM())
   n7 = Get FooterButton_ReportIsland(self=Get WBP_IslandDetails_Footer())
   SetVisibility(execute=@n1.then_2, self=@n7, InVisibility=Select(Option 0=Collapsed, Option 1=Visible, Index=@n6))
   > Branch(Condition=@n6)
   > Bind Event to On Clicked(self=@n7, Delegate=Create Event())
  ]}
 SetupTimerViewModel {nodes [
   n1 = SetupTimerViewModel!() -> (then, IsTimeLocked:bool=~) {is=title K2Node_FunctionEntry}
   > Branch(Condition=@n1)
   n2 = Get FortCountdownTimerVM() -> (FortCountdownTimerVM:Countdown Timer View Model Object Reference)
   > Is Valid(InputObject=@n2)
   > SetupViewModel!(
    execute
    self:Countdown Timer View Model Object Reference=@n2
    InFutureTime:"Date Time Structure (by ref)"=GetNextStartTime!(self:Fort Discover Link Entry Tile Item VM Object Reference=Get FortIslandDetailsVM()) -> (ReturnValue:Date Time Structure)
    InLiveText:"Text (by ref)"=Format Text!(Format="LOCTEXT(\"Live\")") -> (Result)
   ) -> (then)
   > SetTimerUnitDisplayType!(execute, self:Countdown Timer View Model Object Reference=@n2, InTimerUnitDisplayType:ETimespanUnitDisplayFormat Enum=Abbreviated) -> (then)
   > SetApproximateLongTimers!(execute, self:Countdown Timer View Model Object Reference=@n2, bInApproximateLongTimers:bool=false) -> (then)
   > StartTimer(self:Countdown Timer View Model Object Reference=@n2)
  ]}
 "Update Button " {nodes [
   n1 = "Update Button "!() -> (then, Target:Widget Object Reference, ShowWidget:bool=~) {is=title K2Node_FunctionEntry}
   > Is Valid(InputObject=@n1)
   > SetVisibility(self=@n1, InVisibility=Select(Option 0=Collapsed, Option 1=SelfHitTestInvisible, Index=@n1.ShowWidget))
  ]}
 UpdateCarousel {nodes [
   UpdateCarousel() {is=title K2Node_FunctionEntry}
   > Clear(TargetArray=Get ImageURLs() -> (ImageURLs:Array of Strings))
   > n1 = Sequence!(execute) -> (then_0, then_1)
   > n2 = GetExtraImageUrls!(execute, self:Fort Island Details VM Object Reference=Get FortIslandDetailsVM(), Size:ECreativeLinkPreviewSize Enum=Large) -> (then, ReturnValue:Array of Strings)
   > n3 = Set ImageURLs(ImageURLs:Array of Strings=@n2) -> (Output_Get:Array of Strings)
   > Insert!(
    execute
    TargetArray=@n3
    NewItem=GetImageURL!(self:Fort Discover Link Entry Tile Item VM Object Reference=Get FortIslandDetailsVM(), InSize:ECreativeLinkPreviewSize Enum=Large) -> (ReturnValue:String)
    Index:int=0
   ) -> (then) {is=title K2Node_CallArrayFunction}
   SetMedia!(
    execute=@n1.then_1
    self:WBP Island Details Media Gallery Object Reference=Get MediaGallery()
    ImageURLs:Array of Strings=Get ImageURLs()
    InViewModel:Discover Activity View Model Object Reference=Get FortIslandDetailsVM()
   ) -> (then)
  ]}
 UpdateFavoriteStatus {nodes [
   UpdateFavoriteStatus!() -> (then, IsFavorite:bool=~) {is=title K2Node_FunctionEntry}
   > Branch(Condition=Get IsInitialized() -> (IsInitialized:bool))
   > Set Toggle(self=Get Button_Favorite(), IsInit=false, IsToggled=Get IsFavorite() -> (IsFavorite:bool))
  ]}
 UpdateInvalidActivityWarningState {nodes [
   n1 = UpdateInvalidActivityWarningState!() -> (then, InvalidActivityReason:EFortInvalidActivityReason Enum=None) {is=title K2Node_FunctionEntry}
   n2 = GetGameActivity(self=Get FortIslandDetailsVM())
   > n3 = Branch(Condition=AND Boolean(
     A="Not Equal (Enum)"!(A:EFortInvalidActivityReason Enum=@n1, B:EFortInvalidActivityReason Enum=InvalidData) -> (ReturnValue:bool)
     B=IsValid(Object=@n2)
    ))
   n4 = IsContentGated!(self:Fort Game Activity Object Reference=GetGameActivity(self=Get FortIslandDetailsVM())) -> (ReturnValue:bool)
   n5 = "Equal (Enum)"(A=Get InvalidActivityReason() -> (InvalidActivityReason:EFortInvalidActivityReason Enum=None), B=None)
   > SetIsLocked!(execute, self:Common Button Base Object Reference=Get Button_Select(), bInIsLocked:bool=NOT Boolean(A=OR Boolean(A=@n5, B=@n4))) -> (then)
   n6 = GetBestLockedReason!(
    self:Fort Game Activity Object Reference=GetGameActivity(self=Get FortIslandDetailsVM())
    Context:Player Controller Object Reference=GetOwningPlayer()
   ) -> (ReturnValue:Per Player Activity Locked Reason Structure)
   n7 = AND Boolean(
    A="Not Equal (Enum)"(A=Get InvalidActivityReason() -> (InvalidActivityReason=None), B=None)
    B="Equal (Enum)"(
     A:EFortActivityIsLockedReason Enum=Break Per Player Activity Locked Reason!(PerPlayerActivityLockedReason:"Per Player Activity Locked Reason Structure (by ref)"=@n6) -> (LockedFor:Fort Player Controller Object Reference, Reason:EFortActivityIsLockedReason Enum) {is=title K2Node_BreakStruct}.Reason
     B:EFortActivityIsLockedReason Enum=None
    )
   )
   > "Update Button "!(
    execute
    self
    Target:Widget Object Reference=Get DisabledWarningHB() -> (DisabledWarningHB:Horizontal Box Object Reference)
    ShowWidget:bool=@n7
   ) -> (then) {is=title K2Node_CallFunction}
   > Branch(Condition=@n7)
   > "SetText (Text)"(
    self=Get DisabledWarningMessage() -> (DisabledWarningMessage:Common Text Object Reference)
    InText=Select!(
     None:Text
     PartyTooBig:Text="LOCTABLE(\"/Game/UI/DiscoveryShared/StringTable_ActivityBrowser\", \"Warning_PartyTooLarge_Long\")"
     NotPartyLeader:Text="LOCTABLE(\"&StringTable_ActivityBrowser\", \"Warning_NotPartyLeader\")"
     MatchmakingAlready:Text="LOCTABLE(\"&StringTable_ActivityBrowser\", \"Warning_CurrentlyMatchmaking\")"
     PlaylistDisabled:Text="LOCTABLE(\"&StringTable_ActivityBrowser\", \"Warning_ModeDisabled\")"
     PlayerIsLfg:Text="LOCTABLE(\"&StringTable_ActivityBrowser\", \"Warning_PlayerIsLfg\")"
     InvalidData:Text="LOCTABLE(\"&StringTable_ActivityBrowser\", \"Warning_InvalidMode\")"
     ExpiredTournamentLockCache:Text="LOCTABLE(\"&StringTable_ActivityBrowser\", \"Warning_ExpiredTournamentLockCache\")"
     Index:EFortInvalidActivityReason Enum=Get InvalidActivityReason() -> (InvalidActivityReason=None)
    ) -> (ReturnValue:Text)
   )
   SetVisibility(execute=@n3.else, self=Get DisabledWarningHB(), InVisibility=Collapsed)
  ]}
 UpdateSelectText {nodes [
   n1 = UpdateSelectText!() -> (then, IsInGameDevice:bool=~, DownloadRequired:bool=~) {is=title K2Node_FunctionEntry}
   n2 = Get Button_Select()
   > SetText!(
    execute
    self:Fort CTAButton Object Reference=@n2
    ButtonText:"Text (by ref)"=Select!(
     Option 0:Text="LOCTABLE(\"&StringTable_ActivityBrowser\", \"Activity_Select_Button\")"
     Option 1:Text="LOCTABLE(\"&StringTable_ActivityBrowser\", \"Activity_Play_Button\")"
     Index:bool=@n1
    ) -> (ReturnValue)
   ) -> (then)
   > SetSecondaryText!(
    execute
    self:Fort CTAButton Object Reference=@n2
    ButtonSecondaryText:"Text (by ref)"=Select(Option 1="LOCTABLE(\"&StringTable_ActivityBrowser\", \"Warning_DownloadRequired\")", Index=@n1.DownloadRequired)
   ) -> (then)
  ]}
]
variables [
 ContextMenuActionContext {type=UIKitContextMenuActionContext, default=null}
 ContextMenuID_ReportIsland {type=string, default=ReportIsland}
 ContextMenuID_ViewAttributions {type=string, default=ViewAttributions}
 ContextMenuID_ViewCreator {type=string, default=ViewCreator}
 Favorite Count {type=integer}
 Favorite_Sound {type=SoundBase, default=/Game/Sounds/Fort_UI_Sounds/FrontEnd/MultiProduct/Common/Buttons/MSS_UI_MP_Button_Click_Favorite}
 ImageURLs {type="array<string>", default []}
 IsInitialized {type=boolean}
 IsNextFocusReturningFromExpandedMediaGallery {type=boolean}
 MediaGallery {type=WBP_IslandDetails_MediaGallery_C, default=null}
 Modal_IslandOptions {type=WBP_UIKit_Dialog_IslandDetails_C, default=null}
 MouseScrollIndex {type=integer}
 Unfavorite_Sound {type=SoundBase, default=/Game/Sounds/Fort_UI_Sounds/FrontEnd/MultiProduct/Common/Buttons/MSS_UI_MP_Button_Click_Unfavorite}
]
widgets {
 info {
  inheritedWidgetCount=0
  namedSlotCount=1
  parentClass=/Script/FortniteUI.FortActivityDetailsModalV2
  rootWidgetClass=&Overlay
  widgetCount=49
 }
 widgets [
  Overlay_0 {widgetClassPath=&Overlay, widget=Overlay_0}
  Image_Background {
   widgetClassPath=&Image
   bIsVariable=true
   parent=Overlay_0
   widget=Image_Background
   properties {brush {
     bIsDynamicallyLoaded=false
     drawAs=Image
     imageType=FullColor
     mirroring=NoMirror
     resourceName=None
     resourceObject=/FortUILibrary/MaterialInstances/Background/Nebulous/MI_UI_Background_Nebulous
     tiling=NoTile
     imageSize=<32 32>
     margin=<ltrb 0.5 0 0.5 0>
     outlineSettings {
      bUseBrushTransparency=false
      roundingType=HalfHeightRadius
      width=0
      color {colorUseRule=UseColor_Specified, specifiedColor=<rgba 0 0 0 0>}
      cornerRadii=<0 0 0 0>
     }
     tintColor {colorUseRule=UseColor_Specified, specifiedColor=<rgba 1 1 1 1>}
     uVRegion {bIsValid=false, max=<0 0>, min=<0 0>}
    }}
   slot {class=OverlaySlot, properties {horizontalAlignment=HAlign_Fill, verticalAlignment=VAlign_Fill}}
  }
  WBP_CaptureForPostBufferUpdate {
   widgetClassPath=&WBP_CaptureForPostBufferUpdate
   bIsVariable=true
   parent=Overlay_0
   slot {class=OverlaySlot, properties {horizontalAlignment=HAlign_Fill, verticalAlignment=VAlign_Fill}}
  }
  LoadingThrobber {
   widgetClassPath=/FortUILibrary/UIKit/Throbber/WBP_UIKit_Throbber
   bIsVariable=true
   parent=Overlay_0
   widget=LoadingThrobber
   properties {displayBackground=false, throbberSize=Medium, visibility=Collapsed}
   slot {class=OverlaySlot, properties {horizontalAlignment=HAlign_Center, verticalAlignment=VAlign_Center}}
  }
  SafeZone {
   widgetClassPath=&SafeZone
   parent=Overlay_0
   widget=SafeZone
   slot {class=OverlaySlot, properties {horizontalAlignment=HAlign_Fill, verticalAlignment=VAlign_Fill}}
  }
  ScrollBox {
   widgetClassPath=&ScrollBox
   bIsVariable=true
   parent=SafeZone
   widget=ScrollBox
   properties {
    bAnimateWheelScrolling=true
    consumeMouseWheel=Never
    navigationScrollPadding=120
    scrollBarVisibility=Collapsed
    scrollWhenFocusChanges=AnimatedScroll
   }
   slot {class=SafeZoneSlot}
  }
  VBox_Content {
   widgetClassPath=&VerticalBox
   bIsVariable=true
   parent=ScrollBox
   widget=VBox_Content
   slot {class=ScrollBoxSlot, properties {padding=<ltrb 80 54 80 80>}}
  }
  WBP_IslandDetails_Header {
   widgetClassPath=&/WBP_IslandDetails_Header
   bIsVariable=true
   parent=VBox_Content
   widget=WBP_IslandDetails_Header
   slot {class=VerticalBoxSlot, properties {padding=<ltrb 0 0 0 18>}}
  }
  Hbox_CarouselAndStats {
   widgetClassPath=&HorizontalBox
   parent=VBox_Content
   widget=Hbox_CarouselAndStats
   slot {class=VerticalBoxSlot, properties {padding=<ltrb 0 0 0 16>}}
  }
  SizeBox_0 {
   widgetClassPath=&SizeBox
   parent=Hbox_CarouselAndStats
   widget=SizeBox_0
   properties {bOverride_HeightOverride=true, bOverride_WidthOverride=true, heightOverride=703, widthOverride=1250}
   slot {class=HorizontalBoxSlot}
  }
  WBP_IslandDetails_Carousel {
   widgetClassPath=&/WBP_IslandDetails_Carousel
   bIsVariable=true
   parent=SizeBox_0
   widget=WBP_IslandDetails_Carousel
   properties {bIsFocusable=true, visibility=Visible}
   slot {class=SizeBoxSlot}
  }
  Spacer_396 {
   widgetClassPath=&Spacer
   parent=Hbox_CarouselAndStats
   widget=Spacer_396
   properties {size=<24 1>}
   slot {class=HorizontalBoxSlot}
  }
  VerticalBox_0 {
   widgetClassPath=&VerticalBox
   parent=Hbox_CarouselAndStats
   widget=VerticalBox_0
   properties {navigation=&self:WidgetTree.VerticalBox_0.WidgetNavigation_0}
   slot {class=HorizontalBoxSlot, properties {size {sizeRule=Fill, value=1}}}
  }
  WBP_IslandDetails_Stats {
   widgetClassPath=&/WBP_IslandDetails_Stats
   bIsVariable=true
   parent=VerticalBox_0
   widget=WBP_IslandDetails_Stats
   properties {bIsFocusable=true}
   slot {class=VerticalBoxSlot, properties {size {sizeRule=Fill, value=1}}}
  }
  SqueegeeSlot_Card {
   widgetClassPath=/Game/UI/Plugin/Squeegee/SqueegeeInjectionSlot
   bIsVariable=true
   parent=VerticalBox_0
   widget=SqueegeeSlot_Card
   properties {
    bIsFocusable=true
    navigation=&self:WidgetTree.SqueegeeSlot_Card.WidgetNavigation_0
    pluginId=Squeegee
    pluginWidgetId=RatingCard
    visibility=Visible
   }
   slot {class=VerticalBoxSlot, properties {verticalAlignment=VAlign_Bottom, padding=<ltrb 0 12 0 0>}}
  }
  VBox_Buttons {
   widgetClassPath=&VerticalBox
   parent=VBox_Content
   widget=VBox_Buttons
   slot {class=VerticalBoxSlot, properties {verticalAlignment=VAlign_Top, padding=<ltrb 0 0 0 24>}}
  }
  WarningEntryBox {
   widgetClassPath=&DynamicEntryBox
   bIsVariable=true
   parent=VBox_Buttons
   widget=WarningEntryBox
   properties {
    entryBoxType=Vertical
    entryWidgetClass=/FortniteCoreUI/UI/DiscoveryShared/GameDetails/ActivityBrowserStatusWarning.ActivityBrowserStatusWarning_C
    numDesignerPreviewEntries=1
    visibility=Collapsed
    entrySpacing=<25 0>
   }
   slot {class=VerticalBoxSlot, properties {horizontalAlignment=HAlign_Left, padding=<ltrb 0 0 0 12>}}
  }
  DisabledWarningHB {
   widgetClassPath=&HorizontalBox
   bIsVariable=true
   parent=VBox_Buttons
   widget=DisabledWarningHB
   properties {visibility=Collapsed}
   slot {class=VerticalBoxSlot, properties {horizontalAlignment=HAlign_Left, padding=<ltrb 0 0 0 12>}}
  }
  Overlay {
   widgetClassPath=&Overlay
   parent=DisabledWarningHB
   widget=Overlay
   slot {class=HorizontalBoxSlot, properties {horizontalAlignment=HAlign_Left, verticalAlignment=VAlign_Center}}
  }
  WarningTriangle_Pulse {
   widgetClassPath=&Image
   parent=Overlay
   widget=WarningTriangle_Pulse
   properties {
    renderOpacity=0
    visibility=SelfHitTestInvisible
    brush {
     bIsDynamicallyLoaded=false
     drawAs=Image
     imageType=NoImage
     mirroring=NoMirror
     resourceName=None
     resourceObject=/Game/UI/DiscoveryShared/Art/M_UI_Discovery_WarningTriangle_Pulse
     tiling=NoTile
     imageSize=<40 40>
     margin=<ltrb 0 0 0 0>
     outlineSettings {
      bUseBrushTransparency=false
      roundingType=HalfHeightRadius
      width=0
      color {colorUseRule=UseColor_Specified, specifiedColor=<rgba 0 0 0 0>}
      cornerRadii=<0 0 0 0>
     }
     tintColor {colorUseRule=UseColor_Specified, specifiedColor=<rgba 1 1 1 1>}
     uVRegion {bIsValid=false, max=<0 0>, min=<0 0>}
    }
   }
   slot {class=OverlaySlot, properties {horizontalAlignment=HAlign_Fill, verticalAlignment=VAlign_Fill}}
  }
  IconOvr {
   widgetClassPath=&Overlay
   parent=Overlay
   widget=IconOvr
   properties {visibility=HitTestInvisible}
   slot {class=OverlaySlot, properties {horizontalAlignment=HAlign_Fill, verticalAlignment=VAlign_Fill}}
  }
  WarningTriangle {
   widgetClassPath=&Image
   parent=IconOvr
   widget=WarningTriangle
   properties {
    visibility=SelfHitTestInvisible
    brush {
     bIsDynamicallyLoaded=false
     drawAs=Image
     imageType=NoImage
     mirroring=NoMirror
     resourceName=None
     resourceObject=/Game/UI/DiscoveryShared/Art/M_UI_Discovery_WarningTriangle
     tiling=NoTile
     imageSize=<32 32>
     margin=<ltrb 0 0 0 0>
     outlineSettings {
      bUseBrushTransparency=false
      roundingType=HalfHeightRadius
      width=0
      color {colorUseRule=UseColor_Specified, specifiedColor=<rgba 0 0 0 0>}
      cornerRadii=<0 0 0 0>
     }
     tintColor {colorUseRule=UseColor_Specified, specifiedColor=<rgba 1 1 1 1>}
     uVRegion {bIsValid=false, max=<0 0>, min=<0 0>}
    }
   }
   slot {class=OverlaySlot, properties {horizontalAlignment=HAlign_Fill, verticalAlignment=VAlign_Fill}}
  }
  Exclamation {
   widgetClassPath=&CommonTextBlock
   parent=IconOvr
   widget=Exclamation
   properties {
    text="!"
    colorAndOpacity {colorUseRule=UseColor_Specified, specifiedColor=<rgba 0 0 0 1>}
    font {
     bForceMonospaced=false
     bMaterialIsStencil=false
     fontMaterial=null
     fontObject=&Burbank
     letterSpacing=0
     monospacedWidth=1
     size=18
     skewAmount=0
     typefaceFontName=BurbankRegular-Black
     outlineSettings {
      bApplyOutlineToDropShadows=false
      bMiteredCorners=false
      bSeparateFillAlpha=false
      outlineMaterial=null
      outlineSize=0
      outlineColor=<rgba 0 0 0 1>
     }
    }
    margin=<ltrb 0 -4 0 -4>
    renderTransform {angle=0, scale=<1 1>, shear=<4 0>, translation=<-1 5>}
    shadowOffset=<0 0>
   }
   slot {class=OverlaySlot, properties {horizontalAlignment=HAlign_Center, verticalAlignment=VAlign_Center}}
  }
  DisabledWarningMessage {
   widgetClassPath=&CommonTextBlock
   bIsVariable=true
   parent=DisabledWarningHB
   widget=DisabledWarningMessage
   properties {
    mobileFontSizeMultiplier=1.2
    text="{Disabled warning message. Lorem ipsum.}"
    wrapTextAt=544
    font {
     bForceMonospaced=false
     bMaterialIsStencil=false
     fontMaterial=null
     fontObject=&HeadingNow
     letterSpacing=10
     monospacedWidth=1
     size=15.75
     skewAmount=0
     typefaceFontName=74 Regular
     outlineSettings {
      bApplyOutlineToDropShadows=false
      bMiteredCorners=false
      bSeparateFillAlpha=false
      outlineMaterial=/FortUILibrary/MaterialInstances/Text/MI_UI_Block_Text_Outline
      outlineSize=2
      outlineColor=<rgba 0 0 0 1>
     }
    }
    shadowOffset=<0 0>
    shapedTextOptions {bOverride_TextFlowDirection=true, bOverride_TextShapingMethod=false, textFlowDirection=Culture, textShapingMethod=Auto}
   }
   slot {class=HorizontalBoxSlot, properties {verticalAlignment=VAlign_Center, padding=<ltrb 8 0 0 0>, size {sizeRule=Fill, value=1}}}
  }
  HBox_IslandOptions {
   widgetClassPath=&HorizontalBox
   parent=VBox_Buttons
   widget=HBox_IslandOptions
   properties {navigation=&self:WidgetTree.HBox_IslandOptions.WidgetNavigation_0}
   slot {class=VerticalBoxSlot}
  }
  Button_IslandOptions {
   widgetClassPath=&WBP_UIKit_Button_Generic
   bIsVariable=true
   parent=HBox_IslandOptions
   widget=Button_IslandOptions
   properties {
    iconMaterial=/FortUILibrary/Textures/MaterialInstances/MI_UIKit_Icon_Options
    minHeightOverride=56
    minWidthOverride=320
    mobileMinHeight=80
    navigation=&self:WidgetTree.Button_IslandOptions.WidgetNavigation_0
    text="Public, Fill"
    textTransformPolicy=None
    font {
     bForceMonospaced=false
     bMaterialIsStencil=false
     fontMaterial=/FortUILibrary/UIKit/Buttons/MaterialInstances/CTAButtons/MI_UIKit_TextQuiet
     fontObject=&HeadingNow
     letterSpacing=-40
     monospacedWidth=1
     size=15.75
     skewAmount=0
     typefaceFontName=75 Medium
     outlineSettings {
      bApplyOutlineToDropShadows=false
      bMiteredCorners=false
      bSeparateFillAlpha=false
      outlineMaterial=null
      outlineSize=0
      outlineColor=<rgba 0 0 0 1>
     }
    }
   }
   slot {class=HorizontalBoxSlot, properties {padding=<ltrb 0 0 0 10>}}
  }
  HB_Buttons {
   widgetClassPath=&HorizontalBox
   parent=VBox_Buttons
   widget=HB_Buttons
   properties {navigation=&self:WidgetTree.HB_Buttons.WidgetNavigation_0}
   slot {class=VerticalBoxSlot, properties {horizontalAlignment=HAlign_Left, verticalAlignment=VAlign_Bottom}}
  }
  WidgetSwitcher_Buttons {
   widgetClassPath=&CommonActivatableWidgetSwitcher
   bIsVariable=true
   parent=HB_Buttons
   widget=WidgetSwitcher_Buttons
   properties {bOverride_Cursor=true, transitionDuration=0}
   slot {class=HorizontalBoxSlot}
  }
  Button_Select {
   widgetClassPath=/FortUILibrary/UIKit/Buttons/WBP_UIKit_Button_Loud
   bIsVariable=true
   parent=WidgetSwitcher_Buttons
   widget=Button_Select
   properties {
    bHideInputAction=true
    minHeight=56
    minHeightOverride=56
    minWidth=320
    minWidthOverride=320
    mobileMinWidth=400
    text=Select
    useDisabledVisualsOnLocked=true
   }
   slot {class=WidgetSwitcherSlot, properties {horizontalAlignment=HAlign_Left}}
  }
  SqueegeeSlot_LockInfo {
   widgetClassPath=&SqueegeeInjectionSlot
   bIsVariable=true
   parent=WidgetSwitcher_Buttons
   widget=SqueegeeSlot_LockInfo
   properties {bIsFocusable=true, pluginId=Squeegee, pluginWidgetId=Island Details}
   slot {class=WidgetSwitcherSlot, properties {horizontalAlignment=HAlign_Left, verticalAlignment=VAlign_Bottom}}
  }
  Button_TimeLocked {
   widgetClassPath=&WBP_UIKit_Button_Loud
   bIsVariable=true
   parent=WidgetSwitcher_Buttons
   widget=Button_TimeLocked
   properties {
    bHideInputAction=true
    icon=/FortUILibrary/Textures/Icons/T_UI_Icon_UIKit_UpdatePending_SDF
    iconMaterial=/FortUILibrary/Textures/MaterialInstances/MI_UIKit_Icon_UpdatePending
    minHeight=88
    minWidth=400
    text="{Timer}"
   }
   slot {class=WidgetSwitcherSlot, properties {horizontalAlignment=HAlign_Left}}
  }
  HB_SideButtonsAndInfos {
   widgetClassPath=&HorizontalBox
   parent=HB_Buttons
   widget=HB_SideButtonsAndInfos
   slot {class=HorizontalBoxSlot, properties {horizontalAlignment=HAlign_Left, verticalAlignment=VAlign_Bottom, padding=<ltrb 16 0 0 0>}}
  }
  Button_Favorite {
   widgetClassPath=/Game/UI/DiscoveryShared/GameDetails/WBP_GameDetails_Toggle
   bIsVariable=true
   parent=HB_SideButtonsAndInfos
   widget=Button_Favorite
   properties {
    bHideInputAction=true
    icon_HorizontalAlignment=HAlign_Left
    minHeightOverride=48
    minWidthOverride=48
    mobileMinHeight=100
    mobileMinWidth=100
    text=999.9k
    useTextBlock=true
    buttonPadding=<ltrb 0 0 0 0>
    font {
     bForceMonospaced=false
     bMaterialIsStencil=false
     fontMaterial=/FortUILibrary/UIKit/Buttons/MaterialInstances/CTAButtons/MI_UIKit_TextLoud
     fontObject=&HeadingNow
     letterSpacing=10
     monospacedWidth=1
     size=18
     skewAmount=0
     typefaceFontName=76 Bold
     outlineSettings {
      bApplyOutlineToDropShadows=false
      bMiteredCorners=false
      bSeparateFillAlpha=false
      outlineMaterial=null
      outlineSize=0
      outlineColor=<rgba 0 0 0 1>
     }
    }
    textPadding=<ltrb 64 0 16 0>
   }
   slot {class=HorizontalBoxSlot, properties {horizontalAlignment=HAlign_Left, verticalAlignment=VAlign_Bottom, padding=<ltrb 0 0 16 0>}}
  }
  Button_ContextMenu {
   widgetClassPath=/FortUILibrary/UIKit/ContextMenu/WBP_UIKit_ContextMenuAnchorButton
   bIsVariable=true
   parent=HB_SideButtonsAndInfos
   widget=Button_ContextMenu
   properties {
    bIsFocusable=true
    buttonClass=/Game/UI/DiscoveryShared/GameDetails/WBP_GameDetails_Details.WBP_GameDetails_Details_C
    buttonIcon=/FortUILibrary/Textures/Icons/T_UI_Icon_UIKit_ContextMenu
    buttonText=""
    min Height=48
    min Width=48
    mobileMinHeight=100
    mobileMinWidth=100
    visibility=HitTestInvisible
    button Padding=<ltrb 8 8 8 8>
   }
   slot {class=HorizontalBoxSlot}
  }
  VBox_Community {
   widgetClassPath=&VerticalBox
   parent=VBox_Content
   widget=VBox_Community
   properties {navigation=&self:WidgetTree.VBox_Community.WidgetNavigation_0}
   slot {class=VerticalBoxSlot, properties {padding=<ltrb 0 0 0 24>}}
  }
  CommunityRow {
   widgetClassPath=/Game/UI/Plugin/FortPluginWidgetInjectionSlot
   bIsVariable=true
   parent=VBox_Community
   widget=CommunityRow
   properties {pluginId=DiscoveryBrowser, pluginWidgetId=CreatorEventRow, visibility=Collapsed}
   slot {class=VerticalBoxSlot, properties {padding=<ltrb -80 0 -80 0>}}
  }
  VBox_Description {
   widgetClassPath=&VerticalBox
   parent=VBox_Content
   widget=VBox_Description
   properties {navigation=&self:WidgetTree.VBox_Description.WidgetNavigation_0}
   slot {class=VerticalBoxSlot, properties {padding=<ltrb 0 0 0 24>}}
  }
  Text_Details {
   widgetClassPath=&CommonTextBlock
   bIsVariable=true
   parent=VBox_Description
   widget=Text_Details
   properties {
    applyLineHeightToBottomLine=false
    autoWrapText=true
    mobileFontSizeMultiplier=1.2
    text=Description
    wrappingPolicy=AllowPerCharacterWrapping
    font {
     bForceMonospaced=false
     bMaterialIsStencil=false
     fontMaterial=null
     fontObject=&HeadingNow
     letterSpacing=-40
     monospacedWidth=1
     size=15.75
     skewAmount=0
     typefaceFontName=75 Medium
     outlineSettings {
      bApplyOutlineToDropShadows=false
      bMiteredCorners=false
      bSeparateFillAlpha=false
      outlineMaterial=null
      outlineSize=0
      outlineColor=<rgba 0 0 0 1>
     }
    }
    shadowOffset=<0 0>
    shapedTextOptions {bOverride_TextFlowDirection=true, bOverride_TextShapingMethod=false, textFlowDirection=Culture, textShapingMethod=Auto}
    strikeBrush {
     bIsDynamicallyLoaded=false
     drawAs=NoDrawType
     imageType=NoImage
     mirroring=NoMirror
     resourceName=None
     resourceObject=null
     tiling=NoTile
     imageSize=<32 32>
     margin=<ltrb 0 0 0 0>
     outlineSettings {
      bUseBrushTransparency=false
      roundingType=HalfHeightRadius
      width=0
      color {colorUseRule=UseColor_Specified, specifiedColor=<rgba 0 0 0 0>}
      cornerRadii=<0 0 0 0>
     }
     tintColor {colorUseRule=UseColor_Specified, specifiedColor=<rgba 1 1 1 1>}
     uVRegion {bIsValid=false, max=<0 0>, min=<0 0>}
    }
   }
   slot {class=VerticalBoxSlot, properties {padding=<ltrb 0 0 0 12>}}
  }
  WBP_IslandDetails_Description {
   widgetClassPath=&/WBP_IslandDetails_Description
   bIsVariable=true
   parent=VBox_Description
   widget=WBP_IslandDetails_Description
   properties {bIsFocusable=true, navigation=&self:WidgetTree.WBP_IslandDetails_Description.WidgetNavigation_1}
   slot {class=VerticalBoxSlot}
  }
  Image_Divider {
   widgetClassPath=&Image
   bIsVariable=true
   parent=VBox_Content
   widget=Image_Divider
   properties {
    visibility=HitTestInvisible
    brush {
     bIsDynamicallyLoaded=false
     drawAs=Image
     imageType=FullColor
     mirroring=NoMirror
     resourceName=None
     resourceObject=/FortUILibrary/UIKit/Buttons/MaterialInstances/CTAButtons/MI_UIKit_BgQuiet_Fallback
     tiling=NoTile
     imageSize=<32 2>
     margin=<ltrb 0 0 0 0>
     outlineSettings {
      bUseBrushTransparency=false
      roundingType=HalfHeightRadius
      width=0
      color {colorUseRule=UseColor_Specified, specifiedColor=<rgba 0 0 0 0>}
      cornerRadii=<0 0 0 0>
     }
     tintColor {colorUseRule=UseColor_Specified, specifiedColor=<rgba 1 1 1 1>}
     uVRegion {bIsValid=false, max=<0 0>, min=<0 0>}
    }
    colorAndOpacity=<rgba 1 1 1 0.2>
   }
   slot {class=VerticalBoxSlot, properties {padding=<ltrb 0 24 0 48>}}
  }
  VBox_Footer {
   widgetClassPath=&VerticalBox
   parent=VBox_Content
   widget=VBox_Footer
   slot {class=VerticalBoxSlot, properties {padding=<ltrb 0 0 0 24>}}
  }
  WBP_IslandDetails_Footer {
   widgetClassPath=&/WBP_IslandDetails_Footer
   bIsVariable=true
   parent=VBox_Footer
   widget=WBP_IslandDetails_Footer
   slot {class=VerticalBoxSlot}
  }
  SafeZone_Prompts {
   widgetClassPath=&SafeZone
   parent=Overlay_0
   widget=SafeZone_Prompts
   properties {padTop=false}
   slot {class=OverlaySlot, properties {horizontalAlignment=HAlign_Right, verticalAlignment=VAlign_Bottom, padding=<ltrb 80 0 80 32>}}
  }
  WBP_ActionBarContainer {
   widgetClassPath=/Game/UI/Frontend/BottomBar/WBP_ActionBarContainer
   bIsVariable=true
   parent=SafeZone_Prompts
   widget=WBP_ActionBarContainer
   slot {class=SafeZoneSlot}
  }
  HorizontalBox_1 {
   widgetClassPath=&HorizontalBox
   namedSlotHost=WBP_ActionBarContainer
   widget=HorizontalBox_1
   properties {flowDirectionPreference=Culture}
  }
  Button_CreatorPage {
   widgetClassPath=/Game/UI/Frontend/BottomBar/WBP_BottomBarDecoyButton
   bIsVariable=true
   parent=HorizontalBox_1
   widget=Button_CreatorPage
   properties {
    holdData=&CommonUI_HoldData_Base_C
    padding=<ltrb 0 0 0 0>
    triggeringInputAction {dataTable=/FortUILibrary/Foundation/ActivityBrowserInputActionTable, rowName=ShowCreatorPage}
   }
   slot {class=HorizontalBoxSlot}
  }
  Button_Back {
   widgetClassPath=&WBP_BottomBarDecoyButton
   bIsVariable=true
   parent=HorizontalBox_1
   widget=Button_Back
   properties {
    holdData=&CommonUI_HoldData_Base_C
    padding=<ltrb 0 0 0 0>
    triggeringInputAction {dataTable=&InputActionDataTable, rowName=Back}
   }
   slot {class=HorizontalBoxSlot}
  }
  WBP_AthenaBottomBar_WithNotifications {
   widgetClassPath=/Game/UI/Frontend/BottomBar/WBP_AthenaBottomBar_WithNotifications
   bIsVariable=true
   parent=HorizontalBox_1
   widget=WBP_AthenaBottomBar_WithNotifications
   slot {class=HorizontalBoxSlot}
  }
  MediaGallery_Expanded {
   widgetClassPath=&/WBP_IslandDetails_MediaGallery_Modal
   bIsVariable=true
   parent=Overlay_0
   widget=MediaGallery_Expanded
   properties {bSetVisibilityOnActivated=true, bSetVisibilityOnDeactivated=true, visibility=Collapsed}
   slot {class=OverlaySlot, properties {horizontalAlignment=HAlign_Fill, verticalAlignment=VAlign_Fill}}
  }
 ]
}
